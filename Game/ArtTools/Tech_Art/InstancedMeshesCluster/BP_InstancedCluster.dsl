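; Blueprint actor for an instanced mesh cluster, with five instanced static mesh components and an adjustable bounds volume.
; Its construction script gathers the components into an array and, in the editor only, sets each one's visibility from its matching enable-cluster flag.
; It then scales the bounds component from the radius and height variables and raises the billboard to half the bounds height.
; This script only configures components and adds no mesh instances.
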(fn ConstructionScript
  (|CallParentFunction)
  ; Put Components in Array
  (Utilities|Array|Add (Variables|Cluster|GetInstancedStaticMeshCompoenets) (Variables|Default|GetInstancedStaticMesh0))
  (Utilities|Array|Add (Variables|Cluster|GetInstancedStaticMeshCompoenets) (Variables|Default|GetInstancedStaticMesh1))
  (Utilities|Array|Add (Variables|Cluster|GetInstancedStaticMeshCompoenets) (Variables|Default|GetInstancedStaticMesh2))
  (Utilities|Array|Add (Variables|Cluster|GetInstancedStaticMeshCompoenets) (Variables|Default|GetInstancedStaticMesh3))
  (Utilities|Array|Add (Variables|Cluster|GetInstancedStaticMeshCompoenets) (Variables|Default|GetInstancedStaticMesh4))
  ; Delete Unmarked Components
  (if (Editor|InEditor)
    (for (_array_element _array_index) (Variables|Cluster|GetInstancedStaticMeshCompoenets)
      (Rendering|SetVisibility _array_element (Utilities|Array|Get(aref) (Utilities|Array|MakeArray (Variables|Cluster|GetEnableCluster0) (Variables|Cluster|GetEnableCluster1) (Variables|Cluster|GetEnableCluster2) (Variables|Cluster|GetEnableCluster3) (Variables|Cluster|GetEnableCluster4)) _array_index) false)))
  ; Set Billboard Height
  (Transformation|SetRelativeScale3D (Variables|Default|GetBounds) (/ (Math|Vector|MakeVector (Variables|Cluster|GetBoundsRadius) (Variables|Cluster|GetBoundsRadius) (Variables|Cluster|GetBoundsHeight)) "100.000000,100.000000,100.000000"))
  (Transformation|SetRelativeLocation (Variables|Default|GetBillboard) (Math|Vector|MakeVector 0.0 0.0 (/ (Variables|Cluster|GetBoundsHeight) 2.0)) false true))
