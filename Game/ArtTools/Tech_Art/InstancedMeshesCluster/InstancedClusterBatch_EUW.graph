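# Editor utility widget for batch-managing placed actors of a chosen BP_InstancedCluster class.
# Load and Unload buttons match World Partition actor descriptors by class path, then pin or unpin their GUIDs.
# Reset All Scales reads class-default enable flags and spawns a transient cluster to capture default instance transforms from five ISM components.
# It then restores only scale on enabled components of every loaded cluster, keeping each instance's location and rotation.
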EditorUtilityWidgetBlueprint /Game/ArtTools/Tech_Art/InstancedMeshesCluster/InstancedClusterBatch_EUW
graphs [
 BatchResetTransforms {nodes [
   n1 = BatchResetTransforms!() -> (then, Scale:bool=~, Rotation:bool=~, Location:bool=~) {is=title K2Node_FunctionEntry}
   > Set bScaleL(bScaleL:bool=@n1) -> (Output_Get:bool)
   > Set bRotationL(bRotationL:bool=@n1.Rotation) -> (Output_Get:bool)
   > Set bLocationL(bLocationL:bool=@n1.Location) -> (Output_Get:bool)
   > Clear(TargetArray:Array of Instanced Static Mesh Component Object References=Get List Of ISM Components() -> (List Of ISM Components:Array of Instanced Static Mesh Component Object References))
   > n2 = GetAllActorsOfClass!(
    execute
    WorldContextObject:Object Reference
    ActorClass:Actor Class Reference=Get ClusterAsset() -> (ClusterAsset:BP Instanced Cluster Class Reference)
   ) -> (then, OutActors:Array of BP Instanced Cluster Object References)
   > n3 = For Each Loop(Array:Array of BP Instanced Cluster Object References=@n2) -> (Array Element:BP Instanced Cluster Object Reference)
   > n4 = For Each Loop(Array:Array of Instanced Static Mesh Component Object References=Make Array!(
     "[0]":Instanced Static Mesh Component Object Reference=Get InstancedStaticMesh0!(self:BP Instanced Cluster Object Reference=@n3) -> (InstancedStaticMesh0:Instanced Static Mesh Component Object Reference)
     "[1]":Instanced Static Mesh Component Object Reference=Get InstancedStaticMesh1!(self:BP Instanced Cluster Object Reference=@n3) -> (InstancedStaticMesh1:Instanced Static Mesh Component Object Reference)
     "[2]":Instanced Static Mesh Component Object Reference=Get InstancedStaticMesh2!(self:BP Instanced Cluster Object Reference=@n3) -> (InstancedStaticMesh2:Instanced Static Mesh Component Object Reference)
     "[3]":Instanced Static Mesh Component Object Reference=Get InstancedStaticMesh3!(self:BP Instanced Cluster Object Reference=@n3) -> (InstancedStaticMesh3:Instanced Static Mesh Component Object Reference)
     "[4]":Instanced Static Mesh Component Object Reference=Get InstancedStaticMesh4!(self:BP Instanced Cluster Object Reference=@n3) -> (InstancedStaticMesh4:Instanced Static Mesh Component Object Reference)
    ) -> (Array:Array of Instanced Static Mesh Component Object References)) -> (Array Element:Instanced Static Mesh Component Object Reference)
   n5 = Get Enable Cluster 1() -> (Enable Cluster 1:bool)
   n6 = Get Enable Cluster 2() -> (Enable Cluster 2:bool)
   n7 = Get Enable Cluster 3() -> (Enable Cluster 3:bool)
   n8 = Get Enable Cluster 4() -> (Enable Cluster 4:bool)
   n9 = Get Enable Cluster 0() -> (Enable Cluster 0:bool)
   > Branch(Condition="Get (a ref)"(
     Array:Array of Booleans=Make Array("[0]":bool=@n9, "[1]":bool=@n5, "[2]":bool=@n6, "[3]":bool=@n7, "[4]":bool=@n8) -> (Array:Array of Booleans)
     Dimension 1=@n4.Array Index
    ) -> (Output:"Boolean (by ref)"))
   > n10 = For Each Loop(Array:Array of Transforms=Select!(
     Option 0:Array of Transforms=Get Out Transforms_0() -> (Out Transforms_0:Array of Transforms)
     Option 1:Array of Transforms=Get Out Transforms_1() -> (Out Transforms_1:Array of Transforms)
     Option 2:Array of Transforms=Get Out Transforms_2() -> (Out Transforms_2:Array of Transforms)
     Option 3:Array of Transforms=Get Out Transforms_3() -> (Out Transforms_3:Array of Transforms)
     Option 4:Array of Transforms=Get Out Transforms_4() -> (Out Transforms_4:Array of Transforms)
     Index:int=@n4.Array Index
    ) -> (ReturnValue:Array of Transforms)) -> (Array Element:Transform)
   n11 = BreakTransform(InTransform=GetInstanceTransform!(self:Instanced Static Mesh Component Object Reference=@n4, InstanceIndex:int=@n10.Array Index, bWorldSpace:bool=false) -> (OutInstanceTransform:Transform, ReturnValue:bool))
   n12 = BreakTransform(InTransform=@n10)
   n13 = Get bRotationL() -> (bRotationL:bool)
   n14 = Get bLocationL() -> (bLocationL:bool)
   n15 = Get bScaleL() -> (bScaleL:bool)
   > UpdateInstanceTransform!(
    execute
    self:Instanced Static Mesh Component Object Reference=@n4
    InstanceIndex:int=@n10.Array Index
    NewInstanceTransform:"Transform (by ref)"=MakeTransform(
     Location=SelectVector(A=@n12, B=@n11, bPickA=@n14)
     Rotation=SelectRotator!(A:Rotator=@n12.Rotation, B:Rotator=@n11.Rotation, bPickA:bool=@n13) -> (ReturnValue:Rotator)
     Scale=SelectVector(A=@n12.Scale, B=@n11.Scale, bPickA=@n15)
    )
    bWorldSpace:bool=false
    bMarkRenderStateDirty:bool="integer <= integer"(A=GetInstanceCount!(self:Instanced Static Mesh Component Object Reference=@n4) -> (ReturnValue:int=0), B=int + int(A=@n10.Array Index))
    bTeleport:bool=false
   ) -> (then, ReturnValue:bool)
  ]}
 EventGraph {nodes [
   "On Clicked (EditorUtilityButton_37)"() {is=title K2Node_ComponentBoundEvent}
   > FindActorsOfClass()
   > LoadActorsOfClass()
   "On Clicked (EditorUtilityButton)"() {is=title K2Node_ComponentBoundEvent}
   > FindActorsOfClass()
   > UnloadActorsOfClass()
   "On Clicked (EditorUtilityButton_1)"() {is=title K2Node_ComponentBoundEvent}
   > Get Original Settings() {is=title K2Node_CallFunction}
   > BatchResetTransforms!(execute, self, Scale:bool=true, Rotation:bool=false, Location:bool=false) -> (then) {is=title K2Node_CallFunction}
   Event PreConstruct()
   > SetObject(self=Get DetailsView_47() -> (DetailsView_47:Details View Object Reference), NewObject=Self-Reference())
   Event Construct()
   Get EditorAssetSubsystem() -> (ReturnValue:Editor Asset Subsystem Object Reference) {is=title K2Node_GetEditorSubsystem}
  ]}
 FindActorsOfClass {nodes [
   FindActorsOfClass() {is=title K2Node_FunctionEntry}
   > Clear(TargetArray:Array of Guid Structures=Get ListofGuid() -> (ListofGuid:Array of Guid Structures))
   > n1 = GetActorDescs!(execute) -> (then, OutActorDescs:Array of Actor Desc Structures, ReturnValue:bool)
   > n2 = For Each Loop(Array:Array of Actor Desc Structures=@n1) -> (Array Element:Actor Desc Structure)
   n3 = Break Actor Desc!(ActorDesc:"Actor Desc Structure (by ref)"=@n2) -> (
    Guid:Guid Structure
    NativeClass:Object Class Reference
    Class:Soft Object Path Structure
    Name:Name
    Label:Name
    Bounds:Box Structure
    RuntimeGrid:Name
    bIsSpatiallyLoaded:bool=~
    bActorIsEditorOnly:bool=~
    ActorPackage:Name
    ActorPath:Name
    FolderPath:Name
    DataLayerAssets:Array of Soft Object Path Structures
   ) {is=title K2Node_BreakStruct}
   n4 = GetSoftClassPath!(Class:Object Class Reference=Get ClusterAsset()) -> (ReturnValue:Soft Class Path Structure)
   > Branch(Condition=Contains(SearchIn=BreakSoftObjectPath(InSoftObjectPath=@n4), Substring=BreakSoftObjectPath(InSoftObjectPath=@n3.Class)))
   > Add(TargetArray:Array of Guid Structures=Get ListofGuid(), NewItem:"Guid Structure (by ref)"=@n3) {is=title K2Node_CallArrayFunction}
  ]}
 Get Original Settings {nodes [
   n1 = GetBlueprintAsset()
   > n2 = Gather Subobject Data For Blueprint(self=Get SubobjectDataSubsystem(), Context=@n1)
   > n3 = For Each Loop(Array:Array of Subobject Data Handle Structures=@n2) -> (Array Element:Subobject Data Handle Structure)
   > n4 = GetObjectForBlueprint!(execute, Data:"Subobject Data Structure (by ref)"=GetData(DataHandle=@n3), Blueprint:Blueprint Object Reference=@n1) -> (then, ReturnValue:Object Reference)
   > n5 = Cast To InstancedStaticMeshComponent!(execute, Object=@n4) -> (then, CastFailed, AsInstanced Static Mesh Component:Instanced Static Mesh Component Object Reference) {is=title K2Node_DynamicCast}
   > n6 = GetInstanceTransforms!(execute, InInstancedStaticMeshComponent:Instanced Static Mesh Component Object Reference=@n5) -> (then, OutTransforms:Array of Transforms)
   > PrintString(InString="To String (Integer)"(InInt=Length(TargetArray:Array of Transforms=@n6)))
   n7 = Cast To BP_InstancedCluster!(execute, Object:Wildcard) -> (then, CastFailed, AsBP Instanced Cluster:BP Instanced Cluster Object Reference) {is=title K2Node_DynamicCast}
   > n8 = GetInstanceTransforms(InInstancedStaticMeshComponent=Get InstancedStaticMesh0(self=@n7))
   > n9 = For Each Loop(Array:Array of Transforms=@n8) -> (Array Element:Transform)
   > PrintString(InString="To String (Transform)"!(InTrans:"Transform (by ref)"=@n9) -> (ReturnValue:String))
   Get Original Settings() {is=title K2Node_FunctionEntry}
   n10 = Get ClusterAsset()
   > Branch(Condition=IsValidClass!(Class:Object Class Reference=@n10) -> (ReturnValue:bool))
   n11 = Get Class Defaults!(Class:Object Class Reference=@n10) -> (
    EnableCluster0:bool=~
    EnableCluster1:bool=~
    EnableCluster2:bool=~
    EnableCluster3:bool=~
    EnableCluster4:bool=~
    ScaleMin/Max0:Vector 2D Structure
    ScaleMin/Max1:Vector 2D Structure
    ScaleMin/Max2:Vector 2D Structure
    ScaleMin/Max3:Vector 2D Structure
    ScaleMin/Max4:Vector 2D Structure
    BoundsRadius:double=~
    BoundsHeight:double=~
    UpBias:double=~
    ForceBounds:bool=~
   ) {is=title K2Node_GetClassDefaults}
   > Set Enable Cluster 0(Enable Cluster 0:bool=@n11) -> (Output_Get:bool)
   > Set Enable Cluster 1(Enable Cluster 1:bool=@n11.EnableCluster1) -> (Output_Get:bool)
   > Set Enable Cluster 2(Enable Cluster 2:bool=@n11.EnableCluster2) -> (Output_Get:bool)
   > Set Enable Cluster 3(Enable Cluster 3:bool=@n11.EnableCluster3) -> (Output_Get:bool)
   > Set Enable Cluster 4(Enable Cluster 4:bool=@n11.EnableCluster4) -> (Output_Get:bool)
   > n12 = SpawnActorFromClass!(
    execute
    self:Editor Actor Subsystem Object Reference=Get EditorActorSubsystem()
    ActorClass:Actor Class Reference=Get ClusterAsset()
    Location:Vector="0, 0, 0"
    Rotation:Rotator
    bTransient:bool=true
   ) -> (then, ReturnValue:BP Instanced Cluster Object Reference)
   > n13 = For Each Loop(Array:Array of Instanced Static Mesh Component Object References=Make Array(
     "[0]":Instanced Static Mesh Component Object Reference=Get InstancedStaticMesh0(self=@n12)
     "[1]":Instanced Static Mesh Component Object Reference=Get InstancedStaticMesh1(self=@n12)
     "[2]":Instanced Static Mesh Component Object Reference=Get InstancedStaticMesh2(self=@n12)
     "[3]":Instanced Static Mesh Component Object Reference=Get InstancedStaticMesh3(self=@n12)
     "[4]":Instanced Static Mesh Component Object Reference=Get InstancedStaticMesh4(self=@n12)
    ) -> (Array:Array of Instanced Static Mesh Component Object References)) -> (Array Element:Instanced Static Mesh Component Object Reference)
   Destroy Actor(execute=@n13.Completed, self=@n12)
   n14 = GetInstanceTransforms(execute=@n13.LoopBody, InInstancedStaticMeshComponent=@n13)
   > n15 = Switch on Int!(execute:Exec, Selection:int=@n13.Array Index) -> (Default:Exec, 0:Exec, 1:Exec, 2:Exec, 3:Exec, 4:Exec)
   Set Out Transforms_0(execute=@n15.0, Out Transforms_0:Array of Transforms=@n14) -> (Output_Get:Array of Transforms)
   Set Out Transforms_1(execute=@n15.1, Out Transforms_1:Array of Transforms=@n14) -> (Output_Get:Array of Transforms)
   Set Out Transforms_2(execute=@n15.2, Out Transforms_2:Array of Transforms=@n14) -> (Output_Get:Array of Transforms)
   Set Out Transforms_3(execute=@n15.3, Out Transforms_3:Array of Transforms=@n14) -> (Output_Get:Array of Transforms)
   Set Out Transforms_4(execute=@n15.4, Out Transforms_4:Array of Transforms=@n14) -> (Output_Get:Array of Transforms)
   GetDisplayName()
   Get Components By Class!(self:Actor Object Reference, ComponentClass:Actor Component Class Reference=/Script/Engine.ActorComponent) -> (ReturnValue:Array of Actor Component Object References)
   CreateComponentListFromClass!(execute, AISpawnerDataClass:Fort Athena AISpawner Data Class Reference, OuterObject:Object Reference=Get EditorAssetSubsystem()) -> (then, ReturnValue:Fort Athena AISpawner Data Component List Object Reference)
   GetAllLevelActorsComponents!(execute, self:Editor Actor Subsystem Object Reference=Get EditorActorSubsystem()) -> (then, ReturnValue:Array of Actor Component Object References)
   CreateComponentListFromClass(OuterObject=Get SubobjectDataSubsystem())
   RenameSubobjectMemberVariable!(execute, BPContext:Blueprint Object Reference, InHandle:"Subobject Data Handle Structure (by ref)", NewName:Name=None) -> (then)
   Get EditorAssetSubsystem()
   Get AssetEditorSubsystem()
   Get ToolPresetAssetSubsystem() -> (ReturnValue:Tool Preset Asset Subsystem Object Reference) {is=title K2Node_GetEditorSubsystem}
   Get ActorEditorContextSubsystem() -> (ReturnValue:Actor Editor Context Subsystem Object Reference) {is=title K2Node_GetEditorSubsystem}
   Get LevelEditorSubsystem()
   Set members in Subobject Data Handle!(execute:Exec, StructRef:"Subobject Data Handle Structure (by ref)") -> (then:Exec, StructOut:"Subobject Data Handle Structure (by ref)") {is=title K2Node_SetFieldsInStruct}
   Get ClusterAsset()
  ]}
 GetISMs {nodes [
   n1 = GetISMs!() -> (then, Cluster Actor:BP Instanced Cluster Object Reference) {is=title K2Node_FunctionEntry}
   > n2 = For Each Loop(Array:Array of Booleans=Make Array(
     "[0]":bool=Get EnableCluster0!(self:BP Instanced Cluster Object Reference=@n1) -> (EnableCluster0:bool)
     "[1]":bool=Get EnableCluster1!(self:BP Instanced Cluster Object Reference=@n1) -> (EnableCluster1:bool)
     "[2]":bool=Get EnableCluster2!(self:BP Instanced Cluster Object Reference=@n1) -> (EnableCluster2:bool)
     "[3]":bool=Get EnableCluster3!(self:BP Instanced Cluster Object Reference=@n1) -> (EnableCluster3:bool)
     "[4]":bool=Get EnableCluster4!(self:BP Instanced Cluster Object Reference=@n1) -> (EnableCluster4:bool)
    ) -> (Array:Array of Booleans)) -> (Array Element:bool)
   > Branch(Condition=@n2)
   n3 = "Get (a copy)"(
    Array:Array of Instanced Static Mesh Component Object References=Make Array(
     "[0]":Instanced Static Mesh Component Object Reference=Get InstancedStaticMesh0(self=@n1)
     "[1]":Instanced Static Mesh Component Object Reference=Get InstancedStaticMesh1(self=@n1)
     "[2]":Instanced Static Mesh Component Object Reference=Get InstancedStaticMesh2(self=@n1)
     "[3]":Instanced Static Mesh Component Object Reference=Get InstancedStaticMesh3(self=@n1)
     "[4]":Instanced Static Mesh Component Object Reference=Get InstancedStaticMesh4(self=@n1)
    ) -> (Array:Array of Instanced Static Mesh Component Object References)
    Dimension 1=@n2.Array Index
   ) -> (Output:Instanced Static Mesh Component Object Reference)
   > Add(
    TargetArray:Array of Instanced Static Mesh Component Object References=Get List Of ISM Components()
    NewItem:"Instanced Static Mesh Component Object Reference (by ref)"=@n3
   )
   ResetEditorProperty!(execute, Object:Object Reference, PropertyName:Name=PerInstanceSMData, ChangeNotifyMode:EPropertyAccessChangeNotifyMode Enum) -> (then, ReturnValue:bool)
   Reroute Node(InputPin:Instanced Static Mesh Component Object Reference=@n3) -> (OutputPin:Instanced Static Mesh Component Object Reference)
  ]}
 LoadActorsOfClass {nodes [
   LoadActorsOfClass() {is=title K2Node_FunctionEntry}
   n1 = Get ListofGuid()
   > PinActors!(execute, InActorsToPin:Array of Guid Structures=@n1) -> (then)
   n2 = Length(TargetArray:Array of Guid Structures=@n1)
   > PrintString(InString=Append(A="To String (Integer)"(InInt=@n2), B=" Actor(s) loaded"))
  ]}
 UnloadActorsOfClass {nodes [
   UnloadActorsOfClass() {is=title K2Node_FunctionEntry}
   n1 = Get ListofGuid()
   > UnpinActors!(execute, InActorsToUnpin:Array of Guid Structures=@n1) -> (then)
   n2 = Length(TargetArray=@n1)
   > PrintString(InString=Append(A="To String (Integer)"(InInt=@n2), B=" Actor(s) unloaded"))
  ]}
]
variables [
 ClusterAsset {type=Class, default=null}
 Enable Cluster 0 {type=boolean}
 Enable Cluster 1 {type=boolean}
 Enable Cluster 2 {type=boolean}
 Enable Cluster 3 {type=boolean}
 Enable Cluster 4 {type=boolean}
 Instanced Static Mesh 4 {type=InstancedStaticMeshComponent, default=null}
 List Of ISM Components {type="array<InstancedStaticMeshComponent>", default []}
 ListofGuid {type="array<Guid>", default []}
 Out Transforms_0 {type="array<Transform>", default []}
 Out Transforms_1 {type="array<Transform>", default []}
 Out Transforms_2 {type="array<Transform>", default []}
 Out Transforms_3 {type="array<Transform>", default []}
 Out Transforms_4 {type="array<Transform>", default []}
]
widgets {
 info {rootWidgetClass=&VerticalBox, widgetCount=8}
 widgets [
  VerticalBox_0 {widgetClassPath=&VerticalBox, widget=VerticalBox_0}
  DetailsView_47 {
   widgetClassPath=&DetailsView
   parent=VerticalBox_0
   widget=DetailsView_47
   properties {propertiesToShow [ClusterAsset,Asset,Asset_0]}
   slot {class=VerticalBoxSlot}
  }
  EditorUtilityButton_37 {widgetClassPath=&EditorUtilityButton, parent=VerticalBox_0, widget=EditorUtilityButton_37, slot {class=VerticalBoxSlot}}
  TextBlock_14 {
   widgetClassPath=&TextBlock
   parent=EditorUtilityButton_37
   widget=TextBlock_14
   properties {text=Load Actors}
   slot {class=ButtonSlot}
  }
  EditorUtilityButton {widgetClassPath=&EditorUtilityButton, parent=VerticalBox_0, widget=EditorUtilityButton, slot {class=VerticalBoxSlot}}
  TextBlock {
   widgetClassPath=&TextBlock
   parent=EditorUtilityButton
   widget=TextBlock
   properties {text=Unload Actors}
   slot {class=ButtonSlot}
  }
  EditorUtilityButton_1 {
   widgetClassPath=&EditorUtilityButton
   parent=VerticalBox_0
   widget=EditorUtilityButton_1
   slot {class=VerticalBoxSlot, properties {padding=<ltrb 0 30 0 0>}}
  }
  TextBlock_1 {
   widgetClassPath=&TextBlock
   parent=EditorUtilityButton_1
   widget=TextBlock_1
   properties {text=Reset All Scales}
   slot {class=ButtonSlot}
  }
 ]
}
