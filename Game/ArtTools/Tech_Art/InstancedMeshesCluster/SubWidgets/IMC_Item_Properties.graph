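# Editor utility sub-widget in the InstancedMeshesCluster art tool that shows one item's mesh, scale min/max and detail mode as editable property fields.
# Three Single Property Views sit in a bordered row, with PreConstruct setting their property names from variables and the border brush alpha from DarkBG.
# The SetObjects event binds the Mesh and Detail views to a supplied component and the Scale view to a supplied actor.
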EditorUtilityWidgetBlueprint /Game/ArtTools/Tech_Art/InstancedMeshesCluster/SubWidgets/IMC_Item_Properties
graphs [EventGraph {nodes [
   n1 = SetObjects!() -> (OutputDelegate, then, NewComponent:Object Reference, NewActor:Object Reference) {is=title K2Node_CustomEvent}
   > SetObject(
    self=[Get Detail() -> (Detail:Single Property View Object Reference), Get Mesh!() -> (Mesh:Single Property View Object Reference)]
    NewObject=@n1.NewComponent
   )
   > SetObject(self=Get Scale() -> (Scale:Single Property View Object Reference), NewObject=@n1.NewActor)
   Event PreConstruct()
   > SetPropertyName!(execute, self:Single Property View Object Reference=Get Mesh(), NewPropertyName:Name=Get MeshProp() -> (MeshProp:Name=None)) -> (then)
   > SetPropertyName(self=Get Detail(), NewPropertyName=Get DetailProp() -> (DetailProp:Name=None))
   > SetPropertyName(self=Get Scale(), NewPropertyName=Get ScaleProp() -> (ScaleProp:Name=None))
   > SetBrushColor(
    self=Get CommonBorder_141() -> (CommonBorder_141:Common Border Object Reference)
    InBrushColor=MakeColor!(R:float=1, G:float=1, B:float=1, A:float=Get DarkBG() -> (DarkBG:float)) -> (ReturnValue:Linear Color Structure)
   )
   Event Construct()
   Event Tick()
  ]}]
variables [
 DarkBG {type=number, default=1}
 DetailProp {type=string, default=DetailMode}
 MeshProp {type=string, default=StaticMesh}
 ScaleProp {type=string, default=ScaleMin/Max}
]
widgets {
 info {rootWidgetClass=&CommonBorder, widgetCount=7}
 widgets [
  CommonBorder_141 {
   widgetClassPath=&CommonBorder
   bIsVariable=true
   widget=CommonBorder_141
   properties {
    style=/FortUILibrary/Styles/Border/Border-MainL-Black.Border-MainL-Black_C
    background {
     bIsDynamicallyLoaded=false
     drawAs=Box
     imageType=NoImage
     mirroring=NoMirror
     resourceName=None
     resourceObject=/FortUILibrary/Styles/Border/Materials/M-Border-MainL-Black
     tiling=NoTile
     imageSize=<128 32>
     margin=<ltrb 0.1 0.45 0.9 0.5>
     outlineSettings {
      bUseBrushTransparency=false
      roundingType=HalfHeightRadius
      width=0
      color {colorUseRule=UseColor_Specified, specifiedColor=<rgba 0 0 0 0>}
      cornerRadii=<0 0 0 0>
     }
     tintColor {colorUseRule=UseColor_Specified, specifiedColor=<rgba 1 1 1 1>}
     uVRegion {bIsValid=false, max=<0 0>, min=<0 0>}
    }
   }
  }
  HorizontalBox_130 {widgetClassPath=&HorizontalBox, parent=CommonBorder_141, widget=HorizontalBox_130, slot {class=BorderSlot}}
  Mesh {
   widgetClassPath=&SinglePropertyView
   parent=HorizontalBox_130
   widget=Mesh
   properties {nameOverride=" "}
   slot {class=HorizontalBoxSlot, properties {size {sizeRule=Fill, value=1}}}
  }
  VerticalBox_133 {
   widgetClassPath=&VerticalBox
   parent=HorizontalBox_130
   widget=VerticalBox_133
   slot {class=HorizontalBoxSlot, properties {size {sizeRule=Fill, value=1}}}
  }
  CommonTextBlock_25 {
   widgetClassPath=&CommonTextBlock
   parent=VerticalBox_133
   widget=CommonTextBlock_25
   properties {
    style=/Game/ArtTools/Font/TextStyle_RobotoEngineLike.TextStyle_RobotoEngineLike_C
    text=Scale Min/Max
    colorAndOpacity {colorUseRule=UseColor_Specified, specifiedColor=<rgba 0.5 0.5 0.5 1>}
    font {
     bForceMonospaced=false
     bMaterialIsStencil=false
     fontMaterial=null
     fontObject=&Roboto
     letterSpacing=0
     monospacedWidth=1
     size=8
     skewAmount=0
     typefaceFontName=Regular
     outlineSettings {
      bApplyOutlineToDropShadows=false
      bMiteredCorners=false
      bSeparateFillAlpha=false
      outlineMaterial=null
      outlineSize=0
      outlineColor=<rgba 0 0 0 1>
     }
    }
    shadowOffset=<0 0>
   }
   slot {class=VerticalBoxSlot}
  }
  Scale {
   widgetClassPath=&SinglePropertyView
   parent=VerticalBox_133
   widget=Scale
   properties {nameOverride=" "}
   slot {class=VerticalBoxSlot, properties {padding=<ltrb -15 0 0 0>, size {sizeRule=Fill, value=1}}}
  }
  Detail {
   widgetClassPath=&SinglePropertyView
   parent=VerticalBox_133
   widget=Detail
   properties {nameOverride=Detail Mode}
   slot {class=VerticalBoxSlot, properties {size {sizeRule=Fill, value=1}}}
  }
 ]
}
