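# Editor utility widget for placing and editing BP_InstancedCluster actors, which hold up to five instanced static mesh components.
# Tick tracks the selected cluster and loads its settings into the panel.
# Buttons spawn clusters at a camera trace, jitter or randomize instance transforms on enabled components, and open a batch widget.
# Snapping, one-shot or tick-gated continuous, multi-traces onto landscape or "Cliff"-named actors, optionally aligning to up-biased surface normals.
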EditorUtilityWidgetBlueprint /Game/ArtTools/Tech_Art/InstancedMeshesCluster/InstancedCluster_EUW
graphs [
 AddIfValid {nodes [
   n1 = Inputs!() -> (exec:Exec, ISM:"Instanced Static Mesh Component Object Reference (by ref)")
   > n2 = Is Valid(InputObject=@n1)
   > n3 = Add(
    TargetArray:Array of Instanced Static Mesh Component Object References=Get ItemsInstancedStaticMeshCompoenets() -> (ItemsInstancedStaticMeshCompoenets:Array of Instanced Static Mesh Component Object References)
    NewItem:"Instanced Static Mesh Component Object Reference (by ref)"=@n1
   ) {is=title K2Node_CallArrayFunction}
   Outputs!(then:Exec=[@n3.then, @n2.Is Not Valid])
  ]}
 CheckIfComponentVisible {nodes [
   n1 = Inputs!() -> (execute:Exec, Index:int)
   n2 = Get LastCluster() -> (LastCluster:BP Instanced Cluster Object Reference)
   > Is Valid(InputObject=@n2)
   n3 = Get EnableCluster0!(self:BP Instanced Cluster Object Reference=@n2) -> (EnableCluster0:bool)
   n4 = Get EnableCluster1!(self:BP Instanced Cluster Object Reference=@n2) -> (EnableCluster1:bool)
   n5 = Get EnableCluster2!(self:BP Instanced Cluster Object Reference=@n2) -> (EnableCluster2:bool)
   n6 = Get EnableCluster3!(self:BP Instanced Cluster Object Reference=@n2) -> (EnableCluster3:bool)
   n7 = Get EnableCluster4!(self:BP Instanced Cluster Object Reference=@n2) -> (EnableCluster4:bool)
   > n8 = Branch(Condition="Get (a ref)"(
     Array:Array of Booleans=Make Array!("[0]":bool=@n3, "[1]":bool=@n4, "[2]":bool=@n5, "[3]":bool=@n6, "[4]":bool=@n7) -> (Array:Array of Booleans)
     Dimension 1=@n1
    ) -> (Output:"Boolean (by ref)"))
   > Outputs!(true:Exec, false:Exec=@n8.else)
  ]}
 EventGraph {nodes [
   "On Clicked (Btn_Snap)"() {is=title K2Node_ComponentBoundEvent}
   > Snap To Ground()
   "On Clicked (Btn_Add)"() {is=title K2Node_ComponentBoundEvent}
   n1 = GetLevelViewportCameraInfo!(self:Unreal Editor Subsystem Object Reference=Get UnrealEditorSubsystem()) -> (CameraLocation:Vector, CameraRotation:Rotator, ReturnValue:bool)
   n2 = vector + vector(
    A=@n1
    B=vector * vector(A=GetForwardVector!(InRot:Rotator=@n1.CameraRotation) -> (ReturnValue:Vector), B:Vector="1000,1000,1000") -> (ReturnValue=~)
   )
   > n3 = Line Trace By Channel!(
    execute
    WorldContextObject:Object Reference
    Start:Vector=@n1
    End:Vector=@n2
    TraceChannel:ETraceTypeQuery Enum
    bTraceComplex:bool=false
    ActorsToIgnore:Array of Actor Object References
    DrawDebugType:EDrawDebugTrace Enum
    bIgnoreSelf:bool
    TraceColor:Linear Color Structure
    TraceHitColor:Linear Color Structure
    DrawTime:float
   ) -> (then, OutHit:Hit Result Structure, ReturnValue:bool)
   n4 = BreakHitResult!(Hit:"Hit Result Structure (by ref)"=@n3) -> (
    bBlockingHit:bool
    bInitialOverlap:bool
    Time:float
    Distance:float
    Location:Vector
    ImpactPoint:Vector
    Normal:Vector
    ImpactNormal:Vector
    PhysMat:Physical Material Object Reference
    HitActor:Actor Object Reference
    HitComponent:Primitive Component Object Reference
    HitBoneName:Name
    BoneName:Name
    HitItem:int
    ElementIndex:int
    FaceIndex:int
    TraceStart:Vector
    TraceEnd:Vector
   )
   > SpawnActor BP Instanced Cluster!(
    execute:Exec
    WorldContextObject:Object Reference
    Class:Actor Class Reference=&/BP_InstancedCluster.BP_InstancedCluster_C
    SpawnTransform:Transform=MakeTransform(Location=SelectVector(A=@n4.Location, B=@n2, bPickA=@n3.ReturnValue), Rotation="0, 0, 0")
    CollisionHandlingOverride:ESpawnActorCollisionHandlingMethod Enum=Undefined
    TransformScaleMethod:ESpawnActorScaleMethod Enum=MultiplyWithRoot
    Owner:Actor Object Reference
    DamagerOwner:Building Static Mesh Object Reference
    AssociatedMissionParam:Fort Mission Object Reference
    Instigator:Pawn Object Reference
   ) -> (then:Exec, ReturnValue:BP Instanced Cluster Object Reference) {is=title K2Node_SpawnActorFromClass}
   "On Clicked (Btn_Visibility)"() {is=title K2Node_ComponentBoundEvent}
   n5 = Get LastCluster()
   > Is Valid(InputObject=@n5)
   > n6 = Branch(Condition=Get BtnVisibilityChecked() -> (BtnVisibilityChecked:bool))
   > n7 = SetVisibilityButton!(execute, self, Is Visible:bool=false) -> (then, Out:bool)
   n8 = Get Bounds!(self:BP Instanced Cluster Object Reference=@n5) -> (Bounds:Static Mesh Component Object Reference)
   > SetVisibility!(execute, self:Scene Component Object Reference=@n8, bNewVisibility:bool=@n7, bPropagateToChildren:bool=false) -> (then)
   n9 = SetVisibilityButton(execute=@n6.else, Is Visible=true)
   > SetVisibility(self=@n8, bNewVisibility=@n9, bPropagateToChildren=false)
   "On Clicked (Btn_RandRot)"() {is=title K2Node_ComponentBoundEvent}
   > JitterRotation!(execute, self, Amount:double=Get RandomRotationBias() -> (RandomRotationBias:double)) -> (then)
   "On Clicked (Btn_RandLoc)"() {is=title K2Node_ComponentBoundEvent}
   > n10 = JitterPosition!(execute, self, Amount:double=Get MaxRandomDistance() -> (MaxRandomDistance:double)) -> (then)
   Branch(execute=[@n10.then, @n12.then], Condition=Get SnapToGroundPos() -> (SnapToGroundPos:bool))
   > Snap To Ground()
   "On Clicked (Btn_Focus)"() {is=title K2Node_ComponentBoundEvent}
   > SetSelectedLevelActors(
    self=Get EditorActorSubsystem()
    ActorsToSelect=Make Array!("[0]":Actor Object Reference=Get LastCluster()) -> (Array:Array of Actor Object References)
   )
   > ExecuteConsoleCommand(Command=CAMERA ALIGN ACTIVEVIEWPORTONLY)
   "On Clicked (Btn_SnapContinuous)"() {is=title K2Node_ComponentBoundEvent}
   > n11 = Flip Flop!(None:Exec) -> (A:Exec, B:Exec, IsA:bool=~) {is=title K2Node_MacroInstance}
   > K2Node_Composite_2 = Active!(execute) -> (then:Exec) {is=title K2Node_Composite}
   Gate!(Enter:Exec=@n13.Is Valid, Open:Exec=@K2Node_Composite_2.then, Close:Exec=@K2Node_Composite_3.then, Toggle:Exec, bStartClosed:bool=true) -> (Exit:Exec) {is=title K2Node_MacroInstance}
   > Snap To Ground()
   > K2Node_Composite_1 = Processing Icon Animation!(execute) -> (then:Exec) {is=title K2Node_Composite}
   K2Node_Composite_3 = Stop!(execute=@n11.B) -> (then:Exec) {is=title K2Node_Composite}
   "On Clicked (Btn_RandScale)"() {is=title K2Node_ComponentBoundEvent}
   > JitterScale()
   "On Clicked (Btn_ScaleMult)"() {is=title K2Node_ComponentBoundEvent}
   > MultiplyScale!(execute, self, Amount:double=Get ScaleMultiplier() -> (ScaleMultiplier:double)) -> (then)
   "On Clicked (Btn_RandLoc_1)"() {is=title K2Node_ComponentBoundEvent}
   > n12 = RandomPosition()
   "On Clicked (Btn_Batch)"() {is=title K2Node_ComponentBoundEvent}
   > Execute Python Script(PythonScript="import unreal\r\nasset_ref = unreal.load_asset(\"&/InstancedClusterBatch_EUW\")\r\nunreal.EditorUtilitySubsystem().spawn_and_register_tab(asset_ref)")
   Event Tick()
   > Delay!(execute, WorldContextObject:Object Reference, Duration=0.066) -> (then)
   > UpdateUI()
   > K2Node_Composite_0 = Monitor Selection!(execute) -> (isValid:Exec, notValid:Exec) {is=title K2Node_Composite}
   > n13 = Is Valid(InputObject=Get LastCluster())
   Event PreConstruct()
   > SetObject(
    self=[
     Get OrientUp() -> (OrientUp:Single Property View Object Reference)
     Get RandomRotationAmt() -> (RandomRotationAmt:Single Property View Object Reference)
     Get PropJitterDistance() -> (PropJitterDistance:Single Property View Object Reference)
     Get PropScaleMultiplier() -> (PropScaleMultiplier:Single Property View Object Reference)
     Get PropSnap() -> (PropSnap:Single Property View Object Reference)
    ]
    NewObject=Self-Reference()
   )
   > SetObject(
    self=[
     Get PropCheckBox0() -> (PropCheckBox0:Single Property View Object Reference)
     Get PropCheckBox1() -> (PropCheckBox1:Single Property View Object Reference)
     Get PropCheckBox2() -> (PropCheckBox2:Single Property View Object Reference)
     Get PropCheckBox3() -> (PropCheckBox3:Single Property View Object Reference)
     Get PropCheckBox4() -> (PropCheckBox4:Single Property View Object Reference)
    ]
    NewObject=Self-Reference()
   )
   Branch(Condition=Get ForceBounds() -> (ForceBounds:bool))
   > ForceBoundsFunction()
   UtilitiesFunction()
  ]}
 Monitor Selection {
  path=EventGraph.K2Node_Composite_0.Monitor Selection
  nodes [
   Inputs!() -> (execute)
   > n1 = GetSelectedActors()
   > n2 = Branch(Condition=Is Not Empty!(TargetArray:Array of Actor Object References=@n1) -> (ReturnValue:bool))
   n3 = "Get (a copy)"(Array:Array of Actor Object References=@n1, Dimension 1=Last Index!(TargetArray:Array of Actor Object References=@n1) -> (ReturnValue=0)) -> (Output:Actor Object Reference)
   > n4 = Is Valid(InputObject=@n3)
   > n5 = Cast To BP_InstancedCluster!(execute, Object=@n3) -> (then, CastFailed, AsBP Instanced Cluster:BP Instanced Cluster Object Reference) {is=title K2Node_DynamicCast}
   > n6 = Set LastCluster(LastCluster:BP Instanced Cluster Object Reference=@n5) -> (Output_Get:BP Instanced Cluster Object Reference)
   > GetClusterInstancedMeshComponents!(execute, self, Cluster:BP Instanced Cluster Object Reference=@n6) -> (then)
   n7 = Get LastCluster()
   > Set Text(Text="To Text (String)"(InString=GetDisplayName(Object=@n7)), self=Get ObjectName() -> (ObjectName:Common Text Object Reference))
   > n8 = Sequence()
   > Load Cluster Info()
   > Set LastClusterBuffer(LastClusterBuffer:BP Instanced Cluster Object Reference=Get LastCluster()) -> (Output_Get:BP Instanced Cluster Object Reference)
   Outputs!(isValid:Exec=[@n8.then_1, @n2.else], notValid:Exec=[@n5.CastFailed, @n4.Is Not Valid])
   Branch(Condition="Equal (Object)"(A=Get LastCluster(), B=Get LastClusterBuffer() -> (LastClusterBuffer:BP Instanced Cluster Object Reference)))
   Get LastCluster()
  ]
 }
 Processing Icon Animation {
  path=EventGraph.K2Node_Composite_1.Processing Icon Animation
  nodes [
   Inputs()
   n1 = Get processing_icon() -> (processing_icon:Image Object Reference)
   > SetRenderTransformAngle!(
    execute
    self:Widget Object Reference=@n1
    Angle:float=float + float(A=GetRenderTransformAngle!(self:Widget Object Reference=@n1) -> (ReturnValue:float), B=30)
   ) -> (then)
   > Outputs!(then:Exec)
  ]
 }
 Active {
  path=EventGraph.K2Node_Composite_2.Active
  nodes [
   Inputs()
   > SetBackgroundColor!(
    execute
    self:Button Object Reference=Get Btn_SnapContinuous() -> (Btn_SnapContinuous:Button Object Reference)
    InBackgroundColor:Linear Color Structure=(R=2.3922,G=5,B=0.6275,A=1)
   ) -> (then)
   > SetBrushTintColor!(
    execute
    self:Image Object Reference=Get processing_icon()
    TintColor:Slate Color Structure=Make Slate Color(SpecifiedColor=(R=0.99,G=0.948,B=1,A=1))
   ) -> (then)
   > Outputs()
  ]
 }
 Stop {
  path=EventGraph.K2Node_Composite_3.Stop
  nodes [
   Inputs()
   > SetBackgroundColor(self=Get Btn_SnapContinuous(), InBackgroundColor=(R=1,G=1,B=1,A=1))
   > SetBrushTintColor(self=Get processing_icon(), TintColor=Make Slate Color(SpecifiedColor=(R=0.2588,G=0.5412,B=0.0667,A=1)))
   > Outputs()
  ]
 }
 ForceBoundsFunction {nodes [
   ForceBoundsFunction() {is=title K2Node_FunctionEntry}
   > n1 = For Each Loop(Array:Array of Instanced Static Mesh Component Object References=Get ItemsInstancedStaticMeshCompoenets()) -> (Array Element:Instanced Static Mesh Component Object Reference)
   > n2 = For Loop(
    FirstIndex=@n1.Array Index
    LastIndex=int - int(A=GetInstanceCount!(self:Instanced Static Mesh Component Object Reference=@n1) -> (ReturnValue:int=0))
   )
   n3 = GetInstanceTransform!(self:Instanced Static Mesh Component Object Reference=@n1, InstanceIndex:int=@n2, bWorldSpace:bool=false) -> (OutInstanceTransform:Transform, ReturnValue:bool)
   n4 = BreakTransform(InTransform=@n3)
   n5 = BreakVector(InVec=BreakTransform(InTransform=@n3))
   n6 = MakeVector(X=@n5, Y=@n5.Y)
   n7 = Get BoundsRadius() -> (BoundsRadius:double)
   n8 = SelectVector!(
    A=@n6
    B=vector * vector(A=Normalize!(A:Vector=@n6, Tolerance:float=0.0001) -> (ReturnValue:Vector) {is=title K2Node_CallFunction}, B:double=@n7) -> (ReturnValue=~)
    bPickA="float > float"(A="Distance (Vector)"!(V1:Vector=@n6, V2:Vector="0, 0, 0") -> (ReturnValue:double), B=@n7)
   ) -> (ReturnValue_X:double, ReturnValue_Y:double, ReturnValue_Z:double)
   n9 = float * float(
    A="Min (Float)"(A:double="Absolute (Float)"(A=@n5.Z), B:double=Get BoundsHeight() -> (BoundsHeight:double)) -> (ReturnValue:double) {is=title K2Node_CommutativeAssociativeBinaryOperator}
    B="Sign (Float)"!(A:double=@n5.Z) -> (ReturnValue:double)
   ) {is=title K2Node_PromotableOperator}
   > UpdateInstanceTransform!(
    execute
    self:Instanced Static Mesh Component Object Reference=@n1
    InstanceIndex:int=@n2
    NewInstanceTransform:"Transform (by ref)"=MakeTransform(Location=MakeVector(X=@n8, Y=@n8.ReturnValue_Y, Z=@n9), Rotation=@n4.Rotation, Scale=@n4.Scale)
    bWorldSpace:bool=false
    bMarkRenderStateDirty:bool="integer >= integer"(A=@n2, B=@n2)
    bTeleport:bool=false
   ) -> (then, ReturnValue:bool)
  ]}
 GetClusterInstancedMeshComponents {nodes [
   n1 = GetClusterInstancedMeshComponents!() -> (then, Cluster:BP Instanced Cluster Object Reference) {is=title K2Node_FunctionEntry}
   > Clear(TargetArray:Array of Instanced Static Mesh Component Object References=Get ItemsInstancedStaticMeshCompoenets())
   > Add if Valid!(
    exec
    ISM:"Instanced Static Mesh Component Object Reference (by ref)"=Get InstancedStaticMesh0!(self:BP Instanced Cluster Object Reference=@n1) -> (InstancedStaticMesh0:Instanced Static Mesh Component Object Reference)
   ) -> (then:Exec) {is=title K2Node_MacroInstance}
   > Add if Valid(ISM=Get InstancedStaticMesh1!(self:BP Instanced Cluster Object Reference=@n1) -> (InstancedStaticMesh1:Instanced Static Mesh Component Object Reference))
   > Add if Valid(ISM=Get InstancedStaticMesh2!(self:BP Instanced Cluster Object Reference=@n1) -> (InstancedStaticMesh2:Instanced Static Mesh Component Object Reference))
   > Add if Valid(ISM=Get InstancedStaticMesh3!(self:BP Instanced Cluster Object Reference=@n1) -> (InstancedStaticMesh3:Instanced Static Mesh Component Object Reference))
   > Add if Valid(ISM=Get InstancedStaticMesh4!(self:BP Instanced Cluster Object Reference=@n1) -> (InstancedStaticMesh4:Instanced Static Mesh Component Object Reference))
  ]}
 JitterPosition {nodes [
   n1 = JitterPosition!() -> (then, Amount:double=10) {is=title K2Node_FunctionEntry}
   > n2 = For Each Loop(Array=Get ItemsInstancedStaticMeshCompoenets())
   > Check if Component Visible!(execute:Exec, Index:int=@n2.Array Index) -> (true:Exec, false:Exec) {is=title K2Node_MacroInstance}
   n3 = int - int(A=GetInstanceCount(self=@n2) -> (ReturnValue=0))
   > n4 = For Loop(LastIndex=@n3)
   n5 = BreakTransform(InTransform=GetInstanceTransform(self=@n2, InstanceIndex=@n4, bWorldSpace=true))
   n6 = vector * vector(
    A=RandomUnitVector!() -> (ReturnValue:Vector)
    B:Vector=vector * vector(A:double=@n1, B:double=RandomFloat!() -> (ReturnValue:double)) -> (ReturnValue=~)
   ) -> (ReturnValue=~)
   > UpdateInstanceTransform(
    self=@n2
    InstanceIndex=@n4
    NewInstanceTransform=MakeTransform(Location=vector + vector(A=@n5, B=@n6), Rotation=@n5.Rotation, Scale=@n5.Scale)
    bWorldSpace=true
    bMarkRenderStateDirty="Equal (Integer)"(A=@n4, B=@n3)
    bTeleport=true
   )
  ]}
 JitterRotation {nodes [
   n1 = JitterRotation!() -> (then, Amount:double=0.1) {is=title K2Node_FunctionEntry}
   > n2 = For Each Loop(Array=Get ItemsInstancedStaticMeshCompoenets())
   > Check if Component Visible(Index=@n2.Array Index)
   n3 = int - int(A=GetInstanceCount(self=@n2) -> (ReturnValue=0))
   > n4 = For Loop(LastIndex=@n3)
   n5 = BreakTransform(InTransform=GetInstanceTransform(self=@n2, InstanceIndex=@n4, bWorldSpace=true))
   n6 = GetForwardVector(InRot=@n5.Rotation)
   n7 = RandomFloatInRange(Min=0, Max=360)
   n8 = GetUpVector!(InRot:Rotator=@n5.Rotation) -> (ReturnValue:Vector)
   n9 = Select(
    Option 0:Rotator=RandomRotator!(bRoll:bool=false) -> (ReturnValue:Rotator)
    Option 1:Rotator=MakeRotFromZX!(
     Z:"Vector (by ref)"=@n8
     X:"Vector (by ref)"=Rotate Vector Around Axis!(InVect:Vector=@n6, AngleDeg:float=@n7, Axis:Vector=@n8) -> (ReturnValue:Vector)
    ) -> (ReturnValue:Rotator)
    Index=IsChecked(self=Get Check_RotZOnly() -> (Check_RotZOnly:Check Box Object Reference))
   ) -> (ReturnValue:Rotator)
   > UpdateInstanceTransform(
    self=@n2
    InstanceIndex=@n4
    NewInstanceTransform=MakeTransform(
     Location=@n5
     Rotation="Lerp (Rotator)"!(A:Rotator=@n5.Rotation, B:Rotator=@n9, Alpha:float=@n1, bShortestPath:bool=true) -> (ReturnValue:Rotator)
     Scale=@n5.Scale
    )
    bWorldSpace=true
    bMarkRenderStateDirty="Equal (Integer)"(A=@n4, B=@n3)
    bTeleport=true
   )
  ]}
 JitterScale {nodes [
   JitterScale() {is=title K2Node_FunctionEntry}
   > n1 = For Each Loop(Array=Get ItemsInstancedStaticMeshCompoenets())
   > Check if Component Visible(Index=@n1.Array Index)
   n2 = int - int(A=GetInstanceCount(self=@n1) -> (ReturnValue=0))
   > n3 = For Loop(LastIndex=@n2)
   > Set RandomScale(RandomScale:double=RandomFloat()) -> (Output_Get:double)
   n4 = BreakTransform(InTransform=GetInstanceTransform(self=@n1, InstanceIndex=@n3, bWorldSpace=true))
   n5 = Get LastCluster()
   n6 = Get ScaleMin/Max2!(self:BP Instanced Cluster Object Reference=@n5) -> (ScaleMin/Max2:Vector 2D Structure)
   n7 = Get ScaleMin/Max3!(self:BP Instanced Cluster Object Reference=@n5) -> (ScaleMin/Max3:Vector 2D Structure)
   n8 = Get ScaleMin/Max4!(self:BP Instanced Cluster Object Reference=@n5) -> (ScaleMin/Max4:Vector 2D Structure)
   n9 = Get ScaleMin/Max1!(self:BP Instanced Cluster Object Reference=@n5) -> (ScaleMin/Max1:Vector 2D Structure)
   n10 = Get ScaleMin/Max0!(self:BP Instanced Cluster Object Reference=@n5) -> (ScaleMin/Max0:Vector 2D Structure)
   n11 = BreakVector2D(InVec="Get (a ref)"(
     Array:Array of Vector 2D Structures=Make Array!(
      "[0]":Vector 2D Structure=@n10
      "[1]":Vector 2D Structure=@n9
      "[2]":Vector 2D Structure=@n6
      "[3]":Vector 2D Structure=@n7
      "[4]":Vector 2D Structure=@n8
     ) -> (Array:Array of Vector 2D Structures)
     Dimension 1=@n1.Array Index
    ) -> (Output:"Vector 2D Structure (by ref)"))
   n12 = Lerp(A=@n11, B=@n11.Y, Alpha=Get RandomScale() -> (RandomScale:double))
   > UpdateInstanceTransform(
    self=@n1
    InstanceIndex=@n3
    NewInstanceTransform=MakeTransform(Location=@n4, Rotation=@n4.Rotation, Scale="To Vector (Float)"!(InDouble:double=@n12) -> (ReturnValue:Vector))
    bWorldSpace=true
    bMarkRenderStateDirty="Equal (Integer)"(A=@n3, B=@n2)
    bTeleport=true
   )
  ]}
 Load Cluster Info {nodes [
   Load Cluster Info() {is=title K2Node_FunctionEntry}
   > Resize!(
    execute
    TargetArray:Array of Instanced Static Mesh Component Object References=Get ItemsInstanced Static Mesh Compoenets Buffer() -> (ItemsInstanced Static Mesh Compoenets Buffer:Array of Instanced Static Mesh Component Object References)
    Size:int=5
   ) -> (then) {is=title K2Node_CallArrayFunction}
   n1 = Get Actor Location(self=Get LastCluster())
   n2 = Get Last Cluster Buffer Location() -> (Last Cluster Buffer Location:Vector)
   n3 = Get UpBias!(self:BP Instanced Cluster Object Reference=Get LastCluster()) -> (UpBias:double)
   n4 = Get UpBias!() -> (UpBias)
   > n5 = Branch(Condition=AND Boolean(
     A="Equal (Vector)"!(A:Vector=@n1, B:Vector=@n2, ErrorTolerance:float=0) -> (ReturnValue:bool=~) {is=title K2Node_PromotableOperator}
     B="Equal (Float)"(A=@n3, B=@n4)
    ))
   > n6 = For Each Loop(Array=Get ItemsInstancedStaticMeshCompoenets())
   n7 = Get LastCluster()
   SetObject(
    execute=@n6.Completed
    self=[
     Get BoundsHeightInfo() -> (BoundsHeightInfo:Single Property View Object Reference)
     Get BoundsRadiusInfo() -> (BoundsRadiusInfo:Single Property View Object Reference)
     Get OrientUp()
    ]
    NewObject=@n7
   )
   > SetVisibilityButton(Is Visible=Get bVisible!(self:Scene Component Object Reference=Get Bounds(self=@n7)) -> (bVisible:bool))
   n8 = Get ItemsInstanced Static Mesh Compoenets Buffer()
   n9 = Branch(
    execute=@n6.LoopBody
    Condition="Not Equal (Object)"(
     A:Object Reference=@n6
     B:Object Reference="Get (a copy)"(Array:Array of Instanced Static Mesh Component Object References=@n8, Dimension 1=@n6.Array Index) -> (Output:Instanced Static Mesh Component Object Reference)
    ) -> (ReturnValue:bool=~) {is=title K2Node_PromotableOperator}
   )
   n10 = "Get (a copy)"(
    Array:Array of IMC Item Properties Object References=Make Array(
     "[0]":IMC Item Properties Object Reference=Get Element_0() -> (Element_0:IMC Item Properties Object Reference)
     "[1]":IMC Item Properties Object Reference=Get Element_1() -> (Element_1:IMC Item Properties Object Reference)
     "[2]":IMC Item Properties Object Reference=Get Element_2() -> (Element_2:IMC Item Properties Object Reference)
     "[3]":IMC Item Properties Object Reference=Get Element_3() -> (Element_3:IMC Item Properties Object Reference)
     "[4]":IMC Item Properties Object Reference=Get Element_4() -> (Element_4:IMC Item Properties Object Reference)
    ) -> (Array:Array of IMC Item Properties Object References)
    Dimension 1=@n6.Array Index
   ) -> (Output:IMC Item Properties Object Reference)
   n11 = Self-Reference()
   > n12 = SetObjects!(execute, self:IMC Item Properties Object Reference=@n10, NewComponent:Object Reference=@n11, NewActor:Object Reference=@n11) -> (then)
   n13 = Get ItemsInstanced Static Mesh Compoenets Buffer()
   n14 = Branch(
    execute=[@n12.then, @n9.else]
    Condition="Equal (Object)"(
     A=@n6
     B="Get (a copy)"(Array:Array of Instanced Static Mesh Component Object References=@n13, Dimension 1=@n6.Array Index) -> (Output:Instanced Static Mesh Component Object Reference)
    )
   )
   SetObjects(execute=@n14.else, self=@n10, NewComponent=@n6, NewActor=Get LastCluster())
   > Set Array Elem!(
    execute
    TargetArray:Array of Instanced Static Mesh Component Object References=Get ItemsInstanced Static Mesh Compoenets Buffer()
    Index:int=@n6.Array Index
    Item:"Instanced Static Mesh Component Object Reference (by ref)"=@n6
    bSizeToFit:bool=true
   ) -> (then)
   Set Last Cluster Buffer Location(execute=@n5.else, Last Cluster Buffer Location:Vector=@n1) -> (Output_Get:Vector)
   > Set UpBias(UpBias:double=@n3) -> (Output_Get:double)
   n15 = For Each Loop(Array=Get ItemsInstancedStaticMeshCompoenets())
   n16 = For Each Loop(
    Exec=@n15.Completed
    Array:Array of Check Box Object References=Make Array(
     "[0]":Check Box Object Reference=Get CheckBox0() -> (CheckBox0:Check Box Object Reference)
     "[1]":Check Box Object Reference=Get CheckBox1() -> (CheckBox1:Check Box Object Reference)
     "[2]":Check Box Object Reference=Get CheckBox2() -> (CheckBox2:Check Box Object Reference)
     "[3]":Check Box Object Reference=Get CheckBox3() -> (CheckBox3:Check Box Object Reference)
     "[4]":Check Box Object Reference=Get CheckBox4() -> (CheckBox4:Check Box Object Reference)
    ) -> (Array:Array of Check Box Object References)
   ) -> (Array Element:Check Box Object Reference)
   n17 = Get LastCluster()
   n18 = For Each Loop(
    Exec=@n16.Completed
    Array:Array of Vector 2D Structures=Make Array(
     "[0]":Vector 2D Structure=Get ScaleMin/Max0(self=@n17)
     "[1]":Vector 2D Structure=Get ScaleMin/Max1(self=@n17)
     "[2]":Vector 2D Structure=Get ScaleMin/Max2(self=@n17)
     "[3]":Vector 2D Structure=Get ScaleMin/Max3(self=@n17)
     "[4]":Vector 2D Structure=Get ScaleMin/Max4(self=@n17)
    ) -> (Array:Array of Vector 2D Structures)
   ) -> (Array Element:Vector 2D Structure)
   n19 = Get LastCluster()
   Set BoundsRadius(execute=@n18.Completed, BoundsRadius:double=Get BoundsRadius!(self:BP Instanced Cluster Object Reference=@n19) -> (BoundsRadius)) -> (Output_Get:double)
   > Set BoundsHeight(BoundsHeight:double=Get BoundsHeight!(self:BP Instanced Cluster Object Reference=@n19) -> (BoundsHeight)) -> (Output_Get:double)
   n20 = Switch on Int!(execute:Exec=@n18.LoopBody, Selection:int=@n18.Array Index) -> (Default:Exec, 0:Exec, 1:Exec, 2:Exec, 3:Exec, 4:Exec)
   Set ScaleInfo_0(execute=@n20.0, ScaleInfo_0:Vector 2D Structure=@n18) -> (Output_Get:Vector 2D Structure)
   Set ScaleInfo_1(execute=@n20.1, ScaleInfo_1:Vector 2D Structure=@n18) -> (Output_Get:Vector 2D Structure)
   Set ScaleInfo_2(execute=@n20.2, ScaleInfo_2:Vector 2D Structure=@n18) -> (Output_Get:Vector 2D Structure)
   Set ScaleInfo_3(execute=@n20.3, ScaleInfo_3:Vector 2D Structure=@n18) -> (Output_Get:Vector 2D Structure)
   Set ScaleInfo_4(execute=@n20.4, ScaleInfo_4:Vector 2D Structure=@n18) -> (Output_Get:Vector 2D Structure)
   n21 = Get LastCluster()
   n22 = Get EnableCluster0(self=@n21)
   n23 = Get EnableCluster1(self=@n21)
   n24 = Get EnableCluster2(self=@n21)
   n25 = Get EnableCluster3(self=@n21)
   n26 = Get EnableCluster4(self=@n21)
   SetIsChecked!(
    execute=@n16.LoopBody
    self:Check Box Object Reference=@n16
    InIsChecked:bool="Get (a ref)"(
     Array:Array of Booleans=Make Array("[0]":bool=@n22, "[1]":bool=@n23, "[2]":bool=@n24, "[3]":bool=@n25, "[4]":bool=@n26) -> (Array:Array of Booleans)
     Dimension 1=@n16.Array Index
    ) -> (Output:"Boolean (by ref)")
   ) -> (then)
   n27 = Sequence(execute=@n15.LoopBody)
   > n28 = Switch on Int(Selection=@n15.Array Index)
   n29 = Get StaticMesh(self=@n15)
   Set MeshDisplayInfo0(execute=@n28.0, MeshDisplayInfo0:Static Mesh Object Reference=@n29) -> (Output_Get:Static Mesh Object Reference)
   Set MeshDisplayInfo1(execute=@n28.1, MeshDisplayInfo1:Static Mesh Object Reference=@n29) -> (Output_Get:Static Mesh Object Reference)
   Set MeshDisplayInfo2(execute=@n28.2, MeshDisplayInfo2:Static Mesh Object Reference=@n29) -> (Output_Get:Static Mesh Object Reference)
   Set MeshDisplayInfo3(execute=@n28.3, MeshDisplayInfo3:Static Mesh Object Reference=@n29) -> (Output_Get:Static Mesh Object Reference)
   Set MeshDisplayInfo4(execute=@n28.4, MeshDisplayInfo4:Static Mesh Object Reference=@n29) -> (Output_Get:Static Mesh Object Reference)
   n30 = Switch on Int(execute=@n27.then_1, Selection=@n15.Array Index)
   n31 = Byte to Enum Enum_InstancedClusterDetailMode!(Byte:Byte=MakeLiteralByte!(Value:Byte=Get DetailMode!(self:Scene Component Object Reference=@n15) -> (DetailMode:EDetailMode Enum=DM_Low)) -> (ReturnValue:Byte=0)) -> (ReturnValue:Enum_InstancedClusterDetailMode Enum) {is=title K2Node_CastByteToEnum}
   Set DetailMode_0(execute=@n30.0, DetailMode_0:Enum_InstancedClusterDetailMode Enum=@n31) -> (Output_Get:Enum_InstancedClusterDetailMode Enum=NewEnumerator0)
   Set DetailMode_1(execute=@n30.1, DetailMode_1:Enum_InstancedClusterDetailMode Enum=@n31) -> (Output_Get:Enum_InstancedClusterDetailMode Enum=NewEnumerator0)
   Set DetailMode_2(execute=@n30.2, DetailMode_2:Enum_InstancedClusterDetailMode Enum=@n31) -> (Output_Get:Enum_InstancedClusterDetailMode Enum=NewEnumerator0)
   Set DetailMode_3(execute=@n30.3, DetailMode_3:Enum_InstancedClusterDetailMode Enum=@n31) -> (Output_Get:Enum_InstancedClusterDetailMode Enum=NewEnumerator0)
   Set DetailMode_4(execute=@n30.4, DetailMode_4:Enum_InstancedClusterDetailMode Enum=@n31) -> (Output_Get:Enum_InstancedClusterDetailMode Enum=NewEnumerator0)
  ]}
 MultiplyScale {nodes [
   n1 = MultiplyScale!() -> (then, Amount:double=0.1) {is=title K2Node_FunctionEntry}
   > Set RandomScale(RandomScale=RandomFloat())
   > n2 = For Each Loop(Array:Array of Instanced Static Mesh Component Object References=Get ItemsInstancedStaticMeshCompoenets()) -> (Array Element:Instanced Static Mesh Component Object Reference)
   > Check if Component Visible(Index=@n2.Array Index)
   n3 = int - int(A=GetInstanceCount(self=@n2) -> (ReturnValue=0))
   > n4 = For Loop(LastIndex=@n3)
   n5 = BreakTransform(InTransform=GetInstanceTransform(self=@n2, InstanceIndex=@n4, bWorldSpace=true))
   > UpdateInstanceTransform(
    self=@n2
    InstanceIndex=@n4
    NewInstanceTransform=MakeTransform(Location=@n5, Rotation=@n5.Rotation, Scale=vector * vector(A:Vector=@n5.Scale, B=@n1) -> (ReturnValue=~))
    bWorldSpace=true
    bMarkRenderStateDirty="Equal (Integer)"(A=@n4, B=@n3)
    bTeleport=true
   )
  ]}
 RandomPosition {nodes [
   RandomPosition() {is=title K2Node_FunctionEntry}
   > n1 = For Each Loop(Array=Get ItemsInstancedStaticMeshCompoenets())
   > Check if Component Visible(Index=@n1.Array Index)
   n2 = int - int(A=GetInstanceCount(self=@n1) -> (ReturnValue=0))
   > n3 = For Loop(LastIndex=@n2)
   n4 = BreakTransform(InTransform=GetInstanceTransform(self=@n1, InstanceIndex=@n3, bWorldSpace=false))
   n5 = Get LastCluster()
   n6 = Get BoundsHeight(self=@n5)
   n7 = float / float(A=@n6, B=2)
   n8 = RandomFloatInRange(Min=float * float(A=@n7, B=-1), Max=@n7)
   n9 = RandomUnitVector()
   n10 = Get BoundsRadius(self=@n5)
   n11 = vector * vector!(
    A=Normalize(A=vector * vector!(A=@n9, B="1,1,0") -> (ReturnValue=~), Tolerance=0.0001)
    B:Vector=vector / vector(A:Vector=MakeVector(X=@n10, Y=@n10, Z=@n6), B:Vector="2,2,2") -> (ReturnValue:Vector)
    C:double=RandomFloat()
   ) -> (ReturnValue=~)
   > UpdateInstanceTransform(
    self=@n1
    InstanceIndex=@n3
    NewInstanceTransform=MakeTransform(Location=vector + vector!(A=@n11, B_X:double=0, B_Y:double=0, B_Z:double=@n8) -> (ReturnValue), Rotation=@n4.Rotation, Scale=@n4.Scale)
    bWorldSpace=false
    bMarkRenderStateDirty="Equal (Integer)"(A=@n3, B=@n2)
    bTeleport=true
   )
  ]}
 SetVisibilityButton {nodes [
   n1 = SetVisibilityButton!() -> (then, Is Visible:bool=~) {is=title K2Node_FunctionEntry}
   > Set BtnVisibilityChecked(BtnVisibilityChecked:bool=@n1) -> (Output_Get:bool)
   n2 = Break Button Style!(ButtonStyle:"Button Style Structure (by ref)"=Get WidgetStyle!(self:Button Object Reference=Get Btn_Visibility() -> (Btn_Visibility:Button Object Reference)) -> (WidgetStyle:Button Style Structure)) -> (
    Normal:Slate Brush Structure
    Hovered:Slate Brush Structure
    Pressed:Slate Brush Structure
    Disabled:Slate Brush Structure
    NormalForeground:Slate Color Structure
    HoveredForeground:Slate Color Structure
    PressedForeground:Slate Color Structure
    DisabledForeground:Slate Color Structure
    NormalPadding:Margin Structure
    PressedPadding:Margin Structure
    PressedSlateSound:Slate Sound Structure
    ClickedSlateSound:Slate Sound Structure
    HoveredSlateSound:Slate Sound Structure
   ) {is=title K2Node_BreakStruct}
   n3 = Break Slate Brush!(SlateBrush:"Slate Brush Structure (by ref)"=@n2) -> (
    TintColor:Slate Color Structure
    DrawAs:ESlateBrushDrawType Enum
    Tiling:ESlateBrushTileType Enum
    Mirroring:ESlateBrushMirrorType Enum
    ImageSize:Vector 2D Structure
    Margin:Margin Structure
    ResourceObject:Object Reference
    OutlineSettings:Slate Brush Outline Settings Structure
   ) {is=title K2Node_BreakStruct}
   n4 = Make Slate Brush!(
    TintColor:Slate Color Structure=@n3
    DrawAs:ESlateBrushDrawType Enum=@n3.DrawAs
    Tiling:ESlateBrushTileType Enum=@n3.Tiling
    Mirroring:ESlateBrushMirrorType Enum=@n3.Mirroring
    ImageSize:Vector 2D Structure=@n3.ImageSize
    Margin:Margin Structure=@n3.Margin
    ResourceObject:Object Reference=Select(
     Option 0:Object Reference=/Game/ArtTools/Tech_Art/InstancedMeshesCluster/Icon/Icon_Hidden
     Option 1:Object Reference=/Game/ArtTools/Tech_Art/InstancedMeshesCluster/Icon/Icon_Visible
     Index=@n1
    ) -> (ReturnValue:Object Reference)
    OutlineSettings:Slate Brush Outline Settings Structure=@n3.OutlineSettings
   ) -> (SlateBrush:Slate Brush Structure) {is=title K2Node_MakeStruct}
   > Set WidgetStyle!(
    execute
    WidgetStyle:Button Style Structure=Make Button Style!(
     Normal:Slate Brush Structure=@n4
     Hovered:Slate Brush Structure=@n4
     Pressed:Slate Brush Structure=@n4
     Disabled:Slate Brush Structure=@n4
     NormalForeground:Slate Color Structure=@n2.NormalForeground
     HoveredForeground:Slate Color Structure=@n2.HoveredForeground
     PressedForeground:Slate Color Structure=@n2.PressedForeground
     DisabledForeground:Slate Color Structure=@n2.DisabledForeground
     NormalPadding:Margin Structure=@n2.NormalPadding
     PressedPadding:Margin Structure=@n2.PressedPadding
     PressedSlateSound:Slate Sound Structure=@n2.PressedSlateSound
     ClickedSlateSound:Slate Sound Structure=(ResourceObject=None)
     HoveredSlateSound:Slate Sound Structure=@n2.HoveredSlateSound
    ) -> (ButtonStyle:Button Style Structure) {is=title K2Node_MakeStruct}
    self:Button Object Reference=Get Btn_Visibility()
   ) -> (then, Output_Get:Button Style Structure)
   > Return Node!(execute, Out:bool=@n1)
  ]}
 Snap To Ground {nodes [
   Snap To Ground() {is=title K2Node_FunctionEntry}
   > n1 = For Each Loop(Array=Get ItemsInstancedStaticMeshCompoenets())
   > Check if Component Visible(Index=@n1.Array Index)
   n2 = int - int(A=GetInstanceCount(self=@n1) -> (ReturnValue=0))
   > n3 = For Loop(LastIndex=@n2)
   n4 = GetInstanceTransform(self=@n1, InstanceIndex=@n3, bWorldSpace=true)
   > n5 = TerrainMultiTrace!(execute, self, Transform:Transform=@n4) -> (then, Result:Transform, Hit:bool)
   > Branch(Condition=@n5.Hit)
   n6 = BreakTransform(InTransform=@n5)
   n7 = BreakTransform(InTransform=@n4)
   n8 = IsChecked(self=Get Check_AlignToSurface() -> (Check_AlignToSurface:Check Box Object Reference))
   > UpdateInstanceTransform(
    self=@n1
    InstanceIndex=@n3
    NewInstanceTransform=MakeTransform(
     Location=@n6
     Rotation=Select(Option 0:Rotator=@n7.Rotation, Option 1:Rotator=@n6.Rotation, Index=@n8) -> (ReturnValue:Rotator)
     Scale=@n6.Scale
    )
    bWorldSpace=true
    bMarkRenderStateDirty="Equal (Integer)"(A=@n3, B=@n2)
    bTeleport=true
   )
  ]}
 Snap To Ground_0 {nodes [
   Snap To Ground_0() {is=title K2Node_FunctionEntry}
   > n1 = For Each Loop(Array=Get ItemsInstancedStaticMeshCompoenets())
   n2 = int - int(A=GetInstanceCount(self=@n1) -> (ReturnValue=0))
   > n3 = For Loop(LastIndex=@n2)
   n4 = GetInstanceTransform(self=@n1, InstanceIndex=@n3, bWorldSpace=true)
   > n5 = TerrainMultiTrace(Transform=@n4)
   > Branch(Condition=@n5.Hit)
   n6 = BreakTransform(InTransform=@n5)
   n7 = BreakTransform(InTransform=@n4)
   n8 = IsChecked(self=Get Check_AlignToSurface())
   > UpdateInstanceTransform(
    self=@n1
    InstanceIndex=@n3
    NewInstanceTransform=MakeTransform(Location=@n6, Rotation=Select(Option 0=@n7.Rotation, Option 1=@n6.Rotation, Index=@n8), Scale=@n6.Scale)
    bWorldSpace=true
    bMarkRenderStateDirty="Equal (Integer)"(A=@n3, B=@n2)
    bTeleport=true
   )
  ]}
 TerrainMultiTrace {nodes [
   n1 = TerrainMultiTrace!() -> (then, Transform:Transform) {is=title K2Node_FunctionEntry}
   n2 = BreakTransform(InTransform=@n1)
   n3 = BreakVector(InVec=@n2)
   n4 = Get LastCluster()
   n5 = MakeVector(X=@n3, Y=@n3.Y, Z=BreakVector(InVec=Get Actor Location(self=@n4)).Z)
   n6 = Get LastCluster()
   n7 = MakeVector(Y=0, Z=float / float(A=Get BoundsHeight(self=@n6), B=2))
   > n8 = Multi Line Trace By Channel!(
    execute
    WorldContextObject:Object Reference
    Start:Vector=vector + vector!(A=@n5, B:Vector=@n7) -> (ReturnValue)
    End:Vector=vector - vector(A=@n5, B=@n7)
    TraceChannel:ETraceTypeQuery Enum
    bTraceComplex:bool=false
    ActorsToIgnore:Array of Actor Object References=Make Array!("[0]":Actor Object Reference=Get LastCluster()) -> (Array:Array of Actor Object References)
    DrawDebugType:EDrawDebugTrace Enum
    bIgnoreSelf:bool=true
    TraceColor:Linear Color Structure=(R=1,G=0,B=0,A=1)
    TraceHitColor:Linear Color Structure=(R=0,G=1,B=0,A=1)
    DrawTime:float=5
   ) -> (then, OutHits:Array of Hit Result Structures, ReturnValue:bool)
   > n9 = For Each Loop with Break!(Exec:Exec, Array:Array of Hit Result Structures=@n8, Break:Exec=@n15.then) -> (LoopBody:Exec, Array Element:Hit Result Structure, Array Index:int, Completed:Exec)
   n10 = BreakHitResult(Hit=@n9)
   > Branch(Condition=@n10)
   > n11 = Branch(Condition=Contains(SearchIn=GetDisplayName(Object=@n10.HitActor), Substring=Cliff))
   n12 = Get LastCluster()
   n13 = Normalize(
    A="Lerp (Vector)"!(A:Vector=@n10.ImpactNormal, B:Vector="0,0,1", Alpha:float=Get UpBias!(self:BP Instanced Cluster Object Reference=@n12) -> (UpBias)) -> (ReturnValue:Vector)
    Tolerance=0.0001
   )
   n14 = GetForwardVector(InRot=@n2.Rotation)
   Set HitTransform(
    execute=[@n16.then, @n11.then]
    HitTransform:Transform=MakeTransform(Location=@n10.Location, Rotation=MakeRotFromZX(Z=@n13, X=@n14), Scale=@n2.Scale)
   ) -> (Output_Get:Transform)
   > n15 = Set HitValue(HitValue:bool=@n10) -> (Output_Get:bool)
   n16 = Cast To LandscapeProxy!(execute=@n11.else, Object=@n10.HitActor) -> (then, CastFailed, AsLandscape Proxy:Landscape Proxy Object Reference) {is=title K2Node_DynamicCast}
   Return Node!(execute=@n9.Completed, Result:Transform=Get HitTransform() -> (HitTransform:Transform), Hit:bool=Get HitValue() -> (HitValue:bool))
  ]}
 UpdateClusterProperties {nodes [
   ClearListItems(self=Get ListView_46() -> (ListView_46:List View Object Reference))
   > n1 = For Each Loop(Array=Get ItemsInstancedStaticMeshCompoenets())
   > AddItem!(execute, self:List View Object Reference=Get ListView_46(), Item:Object Reference=@n1) -> (then)
   Branch(Condition="Equal (Object)"(A=Get LastCluster(), B=Get LastClusterBuffer()))
   > n2 = Increment Int(Value=Get UpdateClusterPropertiesSlowLoop() -> (UpdateClusterPropertiesSlowLoop:int=0))
   > Set UpdateClusterPropertiesSlowLoop(UpdateClusterPropertiesSlowLoop:int="% (Integer)"(A=@n2, B=5) -> (ReturnValue=0)) -> (Output_Get:int=0)
   > n3 = Sequence!(execute) -> (then_0, then_1, then_2:Exec, then_3:Exec, then_4:Exec)
   n4 = Get DetailMode_0() -> (DetailMode_0:Enum_InstancedClusterDetailMode Enum=NewEnumerator0)
   n5 = Get DetailMode_1() -> (DetailMode_1:Enum_InstancedClusterDetailMode Enum=NewEnumerator0)
   n6 = Get DetailMode_2() -> (DetailMode_2:Enum_InstancedClusterDetailMode Enum=NewEnumerator0)
   n7 = Get DetailMode_3() -> (DetailMode_3:Enum_InstancedClusterDetailMode Enum=NewEnumerator0)
   n8 = Get DetailMode_4() -> (DetailMode_4:Enum_InstancedClusterDetailMode Enum=NewEnumerator0)
   n9 = Get UpdateClusterPropertiesSlowLoop() -> (UpdateClusterPropertiesSlowLoop=0)
   SetEditorProperty(
    execute=@n3.then_1
    Object="Get (a copy)"(Array=Get ItemsInstancedStaticMeshCompoenets(), Dimension 1=@n9)
    PropertyName=DetailMode
    PropertyValue:"Enum_InstancedClusterDetailMode Enum (by ref)"="Get (a ref)"(
     Array:Array of Enum_InstancedClusterDetailMode Enums=Make Array!(
      "[0]":Enum_InstancedClusterDetailMode Enum=@n4
      "[1]":Enum_InstancedClusterDetailMode Enum=@n5
      "[2]":Enum_InstancedClusterDetailMode Enum=@n6
      "[3]":Enum_InstancedClusterDetailMode Enum=@n7
      "[4]":Enum_InstancedClusterDetailMode Enum=@n8
     ) -> (Array:Array of Enum_InstancedClusterDetailMode Enums)
     Dimension 1=@n9
    ) -> (Output:"Enum_InstancedClusterDetailMode Enum (by ref)")
   )
   n10 = Get MeshDisplayInfo0() -> (MeshDisplayInfo0:Static Mesh Object Reference)
   n11 = Get MeshDisplayInfo1() -> (MeshDisplayInfo1:Static Mesh Object Reference)
   n12 = Get MeshDisplayInfo2() -> (MeshDisplayInfo2:Static Mesh Object Reference)
   n13 = Get MeshDisplayInfo3() -> (MeshDisplayInfo3:Static Mesh Object Reference)
   n14 = Get MeshDisplayInfo4() -> (MeshDisplayInfo4:Static Mesh Object Reference)
   n15 = Get UpdateClusterPropertiesSlowLoop() -> (UpdateClusterPropertiesSlowLoop=0)
   SetStaticMesh(
    execute=@n3.then_2
    self="Get (a copy)"(Array=Get ItemsInstancedStaticMeshCompoenets(), Dimension 1=@n15)
    NewMesh="Get (a copy)"(
     Array:Array of Static Mesh Object References=Make Array(
      "[0]":Static Mesh Object Reference=@n10
      "[1]":Static Mesh Object Reference=@n11
      "[2]":Static Mesh Object Reference=@n12
      "[3]":Static Mesh Object Reference=@n13
      "[4]":Static Mesh Object Reference=@n14
     ) -> (Array:Array of Static Mesh Object References)
     Dimension 1=@n15
    ) -> (Output:Static Mesh Object Reference)
   )
   n16 = For Each Loop(
    Exec=@n3.then_3
    Array:Array of Vector 2D Structures=Make Array(
     "[0]":Vector 2D Structure=Get ScaleInfo_0() -> (ScaleInfo_0:Vector 2D Structure)
     "[1]":Vector 2D Structure=Get ScaleInfo_1() -> (ScaleInfo_1:Vector 2D Structure)
     "[2]":Vector 2D Structure=Get ScaleInfo_2() -> (ScaleInfo_2:Vector 2D Structure)
     "[3]":Vector 2D Structure=Get ScaleInfo_3() -> (ScaleInfo_3:Vector 2D Structure)
     "[4]":Vector 2D Structure=Get ScaleInfo_4() -> (ScaleInfo_4:Vector 2D Structure)
    ) -> (Array:Array of Vector 2D Structures)
   ) -> (Array Element:Vector 2D Structure)
   > n17 = Switch on Int(Selection=@n16.Array Index)
   n18 = Get LastCluster()
   Set ScaleMin/Max0!(execute=@n17.0, ScaleMin/Max0:Vector 2D Structure=@n16, self:BP Instanced Cluster Object Reference=@n18) -> (then, Output_Get:Vector 2D Structure)
   Set ScaleMin/Max1!(execute=@n17.1, ScaleMin/Max1:Vector 2D Structure=@n16, self:BP Instanced Cluster Object Reference=@n18) -> (then, Output_Get:Vector 2D Structure)
   Set ScaleMin/Max2!(execute=@n17.2, ScaleMin/Max2:Vector 2D Structure=@n16, self:BP Instanced Cluster Object Reference=@n18) -> (then, Output_Get:Vector 2D Structure)
   Set ScaleMin/Max3!(execute=@n17.3, ScaleMin/Max3:Vector 2D Structure=@n16, self:BP Instanced Cluster Object Reference=@n18) -> (then, Output_Get:Vector 2D Structure)
   Set ScaleMin/Max4!(execute=@n17.4, ScaleMin/Max4:Vector 2D Structure=@n16, self:BP Instanced Cluster Object Reference=@n18) -> (then, Output_Get:Vector 2D Structure)
   n19 = Get UpdateClusterPropertiesSlowLoop() -> (UpdateClusterPropertiesSlowLoop=0)
   n20 = Switch on Int(execute=@n3.then_0, Selection=@n19)
   n21 = IsChecked(self="Get (a copy)"(
     Array:Array of Check Box Object References=Make Array(
      "[0]":Check Box Object Reference
      "[1]":Check Box Object Reference
      "[2]":Check Box Object Reference
      "[3]":Check Box Object Reference
      "[4]":Check Box Object Reference
     ) -> (Array:Array of Check Box Object References)
     Dimension 1=@n19
    ) -> (Output:Check Box Object Reference))
   n22 = Get LastCluster()
   n23 = Set EnableCluster0!(execute=@n20.0, EnableCluster0:bool=@n21, self:BP Instanced Cluster Object Reference=@n22) -> (then, Output_Get:bool)
   SetVisibility(
    execute=[@n23.then, @n24.then, @n26.then, @n25.then, @n27.then]
    self="Get (a copy)"(
     Array:Array of Instanced Static Mesh Component Object References=Get ItemsInstancedStaticMeshCompoenets()
     Dimension 1=Get UpdateClusterPropertiesSlowLoop() -> (UpdateClusterPropertiesSlowLoop=0)
    ) -> (Output:Instanced Static Mesh Component Object Reference)
    bNewVisibility=@n21
    bPropagateToChildren=false
   )
   n24 = Set EnableCluster1!(execute=@n20.1, EnableCluster1:bool=@n21, self:BP Instanced Cluster Object Reference=@n22) -> (then, Output_Get:bool)
   n25 = Set EnableCluster2!(execute=@n20.2, EnableCluster2:bool=@n21, self:BP Instanced Cluster Object Reference=@n22) -> (then, Output_Get:bool)
   n26 = Set EnableCluster3!(execute=@n20.3, EnableCluster3:bool=@n21, self:BP Instanced Cluster Object Reference=@n22) -> (then, Output_Get:bool)
   n27 = Set EnableCluster4!(execute=@n20.4, EnableCluster4:bool=@n21, self:BP Instanced Cluster Object Reference=@n22) -> (then, Output_Get:bool)
   n28 = Get LastCluster()
   Set BoundsRadius!(execute=@n3.then_4, BoundsRadius=Get BoundsRadius!() -> (BoundsRadius), self:BP Instanced Cluster Object Reference=@n28) -> (then, Output_Get)
   > Set BoundsHeight!(execute, BoundsHeight=Get BoundsHeight!() -> (BoundsHeight), self:BP Instanced Cluster Object Reference=@n28) -> (then, Output_Get)
   n29 = Get BoundsRadius()
   n30 = Get BoundsHeight()
   > SetRelativeScale3D!(
    execute
    self:Scene Component Object Reference=Get Bounds(self=@n28)
    NewScale3D:Vector=vector / vector(A=MakeVector(X=@n29, Y=@n29, Z=@n30), B="100,100,100")
   ) -> (then)
   n31 = Get BoundsHeight()
   > Set World Location!(
    execute
    self:Scene Component Object Reference=Get Billboard!(self:BP Instanced Cluster Object Reference=@n28) -> (Billboard:Billboard Component Object Reference)
    NewLocation:Vector=vector + vector!(A=Get Actor Location(self=@n28), B_X:double=0, B_Y:double=0, B_Z:double=float / float(A=@n31, B=2)) -> (ReturnValue)
    bSweep:bool=false
    bTeleport:bool=false
   ) -> (then, SweepHitResult:Hit Result Structure) {is=title K2Node_CallFunction}
   n32 = Branch(Condition="Equal (Object)"(A=Get LastCluster(), B=Get LastClusterBuffer()))
   Set LastClusterBuffer(execute=@n32.else, LastClusterBuffer=Get LastCluster())
   > n33 = For Each Loop(Array:Array of Instanced Static Mesh Component Object References=Get ItemsInstancedStaticMeshCompoenets()) -> (Array Element:Instanced Static Mesh Component Object Reference)
   GetOwner(self="Get (a copy)"(Array:Array of Actor Component Object References) -> (Output:Actor Component Object Reference))
   UpdateClusterProperties() {is=title K2Node_FunctionEntry}
   "Get (a copy)"(Array:Array of Wildcards, Dimension 1=@n33.Array Index) -> (Output:Wildcard)
   n34 = Get MeshDisplayInfo0()
   n35 = Get MeshDisplayInfo1()
   n36 = Get MeshDisplayInfo2()
   n37 = Get MeshDisplayInfo3()
   n38 = Get MeshDisplayInfo4()
   Is Valid(InputObject="Get (a copy)"(
     Array:Array of Static Mesh Object References=Make Array(
      "[0]":Static Mesh Object Reference=@n34
      "[1]":Static Mesh Object Reference=@n35
      "[2]":Static Mesh Object Reference=@n36
      "[3]":Static Mesh Object Reference=@n37
      "[4]":Static Mesh Object Reference=@n38
     ) -> (Array:Array of Static Mesh Object References)
     Dimension 1=Get UpdateClusterPropertiesSlowLoop() -> (UpdateClusterPropertiesSlowLoop=0)
    ) -> (Output:Static Mesh Object Reference))
  ]}
 UpdateUI {nodes [
   UpdateUI() {is=title K2Node_FunctionEntry}
   n1 = Get LastCluster()
   > SetObject(
    self=[
     Get BoundsHeightInfo()
     Get BoundsRadiusInfo()
     Get PropCheckBox0()
     Get PropCheckBox1()
     Get PropCheckBox3()
     Get PropCheckBox4()
     Get PropCheckBox2()
    ]
    NewObject=Select(Option 0:Object Reference=Self-Reference(), Option 1:Object Reference=@n1, Index=IsValid(Object=@n1)) -> (ReturnValue:Object Reference)
   )
   n2 = IsValid(Object=Get LastCluster())
   > Set Visibility(
    Visibility=Select(Option 0:ESlateVisibility Enum=Collapsed, Option 1:ESlateVisibility Enum=SelfHitTestInvisible, Index=@n2) -> (ReturnValue:ESlateVisibility Enum)
    self=Get MeshPanel() -> (MeshPanel:Common Border Object Reference)
   )
   > Set Visibility(
    Visibility=Select(Option 0=SelfHitTestInvisible, Option 1=Collapsed, Index=@n2)
    self=Get MessagePanel() -> (MessagePanel:Common Border Object Reference)
   )
   > n3 = For Each Loop(Array:Array of IMC Item Properties Object References=Make Array(
     "[0]":IMC Item Properties Object Reference=Get Element_0()
     "[1]":IMC Item Properties Object Reference=Get Element_1()
     "[2]":IMC Item Properties Object Reference=Get Element_2()
     "[3]":IMC Item Properties Object Reference=Get Element_3()
     "[4]":IMC Item Properties Object Reference=Get Element_4()
    ) -> (Array:Array of IMC Item Properties Object References)) -> (Array Element:IMC Item Properties Object Reference)
   n4 = Get LastCluster()
   n5 = Select(
    Option 0:Array of Booleans=Make Array(
     "[0]":bool=Get EnableCluster0!() -> (EnableCluster0)
     "[1]":bool=Get EnableCluster1!() -> (EnableCluster1)
     "[2]":bool=Get EnableCluster2!() -> (EnableCluster2)
     "[3]":bool=Get EnableCluster3!() -> (EnableCluster3)
     "[4]":bool=Get EnableCluster4!() -> (EnableCluster4)
    ) -> (Array:Array of Booleans)
    Option 1:Array of Booleans=Make Array(
     "[0]"=Get EnableCluster0!(self:BP Instanced Cluster Object Reference=@n4) -> (EnableCluster0)
     "[1]"=Get EnableCluster1!(self:BP Instanced Cluster Object Reference=@n4) -> (EnableCluster1)
     "[2]"=Get EnableCluster2!(self:BP Instanced Cluster Object Reference=@n4) -> (EnableCluster2)
     "[3]"=Get EnableCluster3!(self:BP Instanced Cluster Object Reference=@n4) -> (EnableCluster3)
     "[4]"=Get EnableCluster4!(self:BP Instanced Cluster Object Reference=@n4) -> (EnableCluster4)
    )
    Index=IsValid(Object=Get LastCluster())
   ) -> (ReturnValue:Array of Booleans)
   > SetVisibility!(
    execute
    self:Widget Object Reference=@n3
    InVisibility:ESlateVisibility Enum=Select(
     Option 0:ESlateVisibility Enum=Collapsed
     Option 1:ESlateVisibility Enum=SelfHitTestInvisible
     Index="Get (a ref)"(Array:Array of Booleans=@n5, Dimension 1=@n3.Array Index) -> (Output:"Boolean (by ref)")
    ) -> (ReturnValue:ESlateVisibility Enum)
   ) -> (then)
  ]}
 UtilitiesFunction {nodes [
   UtilitiesFunction() {is=title K2Node_FunctionEntry}
   > Sequence!(execute) -> (then_0, then_1)
   > n1 = For Each Loop(Array:Array of Instanced Static Mesh Component Object References=Get ItemsInstancedStaticMeshCompoenets()) -> (Array Element:Instanced Static Mesh Component Object Reference)
   > Clear(TargetArray:Array of Transforms=Get Transforms() -> (Transforms:Array of Transforms))
   > n2 = For Loop(LastIndex=int - int(A=GetInstanceCount(self=@n1) -> (ReturnValue=0)))
   n3 = Get StoredData() -> (StoredData:Array of Struct Instanced Cluster Structures)
   n4 = "Get (a ref)"!(Array:Array of Struct Instanced Cluster Structures=@n3, Dimension 1=@n1.Array Index) -> (
    Output_Transforms_3_#1:Array of Transforms
    Output_Mesh_7_#2:Static Mesh Object Reference
    Output_Visibile_17_#3:bool=~
    Output_WIPDetailMode_19_#4:Enum_InstancedClusterDetailMode Enum
    Output_MinSize_22_#5:double=~
    Output_MaxSize_24_#6:double=~
   )
   Set Array Elem(
    execute=@n2.Completed
    TargetArray:Array of Struct Instanced Cluster Structures=@n3
    Index=@n1.Array Index
    Item:"Struct Instanced Cluster Structure (by ref)"=Make Struct Instanced Cluster!(
     Transforms_3_#1:Array of Transforms=Get Transforms()
     Mesh_7_#2:Static Mesh Object Reference=@n4.Output_Mesh_7_#2
     Visibile_17_#3:bool=@n4.Output_Visibile_17_#3
     WIPDetailMode_19_#4:Enum_InstancedClusterDetailMode Enum=NewEnumerator0
     MinSize_22_#5:double=0
     MaxSize_24_#6:double=0
    ) -> (Struct_InstancedCluster:Struct Instanced Cluster Structure) {is=title K2Node_MakeStruct}
    bSizeToFit=false
   )
   Add(
    execute=@n2.LoopBody
    TargetArray:Array of Transforms=Get Transforms()
    NewItem:"Transform (by ref)"=GetInstanceTransform(self=@n1, InstanceIndex=@n2, bWorldSpace=true)
   )
  ]}
]
variables [
 BoundsHeight {type=number, default=1000}
 BoundsRadius {type=number, default=500}
 BtnStyleActive {
  type=ButtonStyle
  default {
   clickedSlateSound {resourceObject=null}
   disabled {
    bIsDynamicallyLoaded=false
    drawAs=NoDrawType
    imageType=NoImage
    mirroring=NoMirror
    resourceName=None
    resourceObject=null
    tiling=NoTile
    imageSize=<0 0>
    margin=<ltrb 0 0 0 0>
    outlineSettings {
     bUseBrushTransparency=false
     roundingType=HalfHeightRadius
     width=0
     color {colorUseRule=UseColor_Specified, specifiedColor=<rgba 0 0 0 0>}
     cornerRadii=<0 0 0 1>
    }
    tintColor {colorUseRule=UseColor_Specified, specifiedColor=<rgba 1 1 1 1>}
    uVRegion {bIsValid=false, max=<0 0>, min=<0 0>}
   }
   disabledForeground {colorUseRule=UseColor_Foreground, specifiedColor=<rgba 1 0 1 1>}
   hovered {
    bIsDynamicallyLoaded=false
    drawAs=Image
    imageType=NoImage
    mirroring=NoMirror
    resourceName=None
    resourceObject=null
    tiling=NoTile
    imageSize=<32 32>
    margin=<ltrb 0 0 0 0>
    outlineSettings {
     bUseBrushTransparency=false
     roundingType=HalfHeightRadius
     width=0
     color {colorUseRule=UseColor_Specified, specifiedColor=<rgba 0 0 0 0>}
     cornerRadii=<0 0 0 1>
    }
    tintColor {colorUseRule=UseColor_Specified, specifiedColor=<rgba 1 1 1 1>}
    uVRegion {bIsValid=false, max=<0 0>, min=<0 0>}
   }
   hoveredForeground {colorUseRule=UseColor_Foreground, specifiedColor=<rgba 1 0 1 1>}
   hoveredSlateSound {resourceObject=null}
   normal {
    bIsDynamicallyLoaded=false
    drawAs=Image
    imageType=NoImage
    mirroring=NoMirror
    resourceName=None
    resourceObject=null
    tiling=NoTile
    imageSize=<32 32>
    margin=<ltrb 0 0 0 0>
    outlineSettings {
     bUseBrushTransparency=false
     roundingType=HalfHeightRadius
     width=0
     color {colorUseRule=UseColor_Specified, specifiedColor=<rgba 0 0 0 0>}
     cornerRadii=<0 0 0 1>
    }
    tintColor {colorUseRule=UseColor_Specified, specifiedColor=<rgba 1 1 1 1>}
    uVRegion {bIsValid=false, max=<0 0>, min=<0 0>}
   }
   normalForeground {colorUseRule=UseColor_Foreground, specifiedColor=<rgba 1 0 1 1>}
   normalPadding=<ltrb 0 0 0 0>
   pressed {
    bIsDynamicallyLoaded=false
    drawAs=Image
    imageType=NoImage
    mirroring=NoMirror
    resourceName=None
    resourceObject=null
    tiling=NoTile
    imageSize=<32 32>
    margin=<ltrb 0 0 0 0>
    outlineSettings {
     bUseBrushTransparency=false
     roundingType=HalfHeightRadius
     width=0
     color {colorUseRule=UseColor_Specified, specifiedColor=<rgba 0 0 0 0>}
     cornerRadii=<0 0 0 1>
    }
    tintColor {colorUseRule=UseColor_Specified, specifiedColor=<rgba 1 1 1 1>}
    uVRegion {bIsValid=false, max=<0 0>, min=<0 0>}
   }
   pressedForeground {colorUseRule=UseColor_Foreground, specifiedColor=<rgba 1 0 1 1>}
   pressedPadding=<ltrb 0 0 0 0>
   pressedSlateSound {resourceObject=null}
  }
 }
 BtnStyleDefault {
  type=ButtonStyle
  default {
   clickedSlateSound {resourceObject=null}
   disabled {
    bIsDynamicallyLoaded=false
    drawAs=NoDrawType
    imageType=NoImage
    mirroring=NoMirror
    resourceName=None
    resourceObject=null
    tiling=NoTile
    imageSize=<0 0>
    margin=<ltrb 0 0 0 0>
    outlineSettings {
     bUseBrushTransparency=false
     roundingType=HalfHeightRadius
     width=0
     color {colorUseRule=UseColor_Specified, specifiedColor=<rgba 0 0 0 0>}
     cornerRadii=<0 0 0 1>
    }
    tintColor {colorUseRule=UseColor_Specified, specifiedColor=<rgba 1 1 1 1>}
    uVRegion {bIsValid=false, max=<0 0>, min=<0 0>}
   }
   disabledForeground {colorUseRule=UseColor_Foreground, specifiedColor=<rgba 1 0 1 1>}
   hovered {
    bIsDynamicallyLoaded=false
    drawAs=Image
    imageType=NoImage
    mirroring=NoMirror
    resourceName=None
    resourceObject=null
    tiling=NoTile
    imageSize=<32 32>
    margin=<ltrb 0 0 0 0>
    outlineSettings {
     bUseBrushTransparency=false
     roundingType=HalfHeightRadius
     width=0
     color {colorUseRule=UseColor_Specified, specifiedColor=<rgba 0 0 0 0>}
     cornerRadii=<0 0 0 1>
    }
    tintColor {colorUseRule=UseColor_Specified, specifiedColor=<rgba 1 1 1 1>}
    uVRegion {bIsValid=false, max=<0 0>, min=<0 0>}
   }
   hoveredForeground {colorUseRule=UseColor_Foreground, specifiedColor=<rgba 1 0 1 1>}
   hoveredSlateSound {resourceObject=null}
   normal {
    bIsDynamicallyLoaded=false
    drawAs=Image
    imageType=NoImage
    mirroring=NoMirror
    resourceName=None
    resourceObject=null
    tiling=NoTile
    imageSize=<32 32>
    margin=<ltrb 0 0 0 0>
    outlineSettings {
     bUseBrushTransparency=false
     roundingType=HalfHeightRadius
     width=0
     color {colorUseRule=UseColor_Specified, specifiedColor=<rgba 0 0 0 0>}
     cornerRadii=<0 0 0 1>
    }
    tintColor {colorUseRule=UseColor_Specified, specifiedColor=<rgba 0.0395 0.0395 0.0395 1>}
    uVRegion {bIsValid=false, max=<0 0>, min=<0 0>}
   }
   normalForeground {colorUseRule=UseColor_Foreground, specifiedColor=<rgba 1 0 1 1>}
   normalPadding=<ltrb 0 0 0 0>
   pressed {
    bIsDynamicallyLoaded=false
    drawAs=Image
    imageType=NoImage
    mirroring=NoMirror
    resourceName=None
    resourceObject=null
    tiling=NoTile
    imageSize=<32 32>
    margin=<ltrb 0 0 0 0>
    outlineSettings {
     bUseBrushTransparency=false
     roundingType=HalfHeightRadius
     width=0
     color {colorUseRule=UseColor_Specified, specifiedColor=<rgba 0 0 0 0>}
     cornerRadii=<0 0 0 1>
    }
    tintColor {colorUseRule=UseColor_Specified, specifiedColor=<rgba 1 1 1 1>}
    uVRegion {bIsValid=false, max=<0 0>, min=<0 0>}
   }
   pressedForeground {colorUseRule=UseColor_Foreground, specifiedColor=<rgba 1 0 1 1>}
   pressedPadding=<ltrb 0 0 0 0>
   pressedSlateSound {resourceObject=null}
  }
 }
 BtnVisibilityChecked {type=boolean, default=true}
 DetailMode_0 {type=Enum_InstancedClusterDetailMode, default=Low}
 DetailMode_1 {type=Enum_InstancedClusterDetailMode, default=Medium}
 DetailMode_2 {type=Enum_InstancedClusterDetailMode, default=High}
 DetailMode_3 {type=Enum_InstancedClusterDetailMode, default=High}
 DetailMode_4 {type=Enum_InstancedClusterDetailMode, default=Epic}
 EnableCluster0 {type=boolean}
 EnableCluster1 {type=boolean}
 EnableCluster2 {type=boolean}
 EnableCluster3 {type=boolean}
 EnableCluster4 {type=boolean}
 ForceBounds {type=boolean}
 ItemsInstanced Static Mesh Compoenets Buffer {type="array<InstancedStaticMeshComponent>", default []}
 ItemsInstancedStaticMeshCompoenets {type="array<InstancedStaticMeshComponent>", default []}
 Last Cluster Buffer Location {type=Vector}
 LastCluster {type=BP_InstancedCluster_C, default=null}
 LastClusterBuffer {type=BP_InstancedCluster_C, default=null}
 LastSelectedActors {type="array<Actor>"}
 MaxRandomDistance {type=number, default=100}
 MeshDisplayInfo0 {type=StaticMesh, default=/Engine/BasicShapes/Cone}
 MeshDisplayInfo1 {type=StaticMesh, default=/Engine/BasicShapes/Cylinder}
 MeshDisplayInfo2 {type=StaticMesh, default=/Engine/BasicShapes/Sphere}
 MeshDisplayInfo3 {type=StaticMesh, default=/Engine/BasicShapes/Plane}
 MeshDisplayInfo4 {type=StaticMesh, default=/Engine/BasicShapes/Cube1}
 RandomRotationBias {type=number, default=0.1}
 ScaleInfo_0 {type=Vector2D, default=<0.9 1.1>}
 ScaleInfo_1 {type=Vector2D, default=<0.9 1.1>}
 ScaleInfo_2 {type=Vector2D, default=<0.9 1.1>}
 ScaleInfo_3 {type=Vector2D, default=<0.9 1.1>}
 ScaleInfo_4 {type=Vector2D, default=<0.9 1.1>}
 ScaleMultiplier {type=number, default=1}
 SnapToGroundPos {type=boolean, default=true}
 StoredData {type="array<Struct_InstancedCluster>", default []}
 Text Debug {type=string, default=""}
 UpBias {type=number}
 UpdateClusterPropertiesSlowLoop {type=integer, default=-1}
]
widgets {
 info {rootWidgetClass=&CanvasPanel, widgetCount=111}
 widgets [
  CanvasPanel_0 {widgetClassPath=&CanvasPanel, widget=CanvasPanel_0}
  Header {
   widgetClassPath=&HorizontalBox
   parent=CanvasPanel_0
   widget=Header
   slot {class=CanvasPanelSlot, properties {layoutData {alignment=<0 0>, anchors {maximum=<1 0>, minimum=<0 0>}, offsets=<ltrb 0 0 0 46>}}}
  }
  Btn_Focus {
   widgetClassPath=&Button
   parent=Header
   widget=Btn_Focus
   properties {widgetStyle {
     clickedSlateSound {resourceObject=null}
     disabled {
      bIsDynamicallyLoaded=false
      drawAs=RoundedBox
      imageType=NoImage
      mirroring=NoMirror
      resourceName=None
      resourceObject=null
      tiling=NoTile
      imageSize=<0 0>
      margin=<ltrb 0 0 0 0>
      outlineSettings {
       bUseBrushTransparency=false
       roundingType=FixedRadius
       width=1
       color {colorUseRule=UseColor_ColorTable, specifiedColor=<rgba 0 0 0 0>}
       cornerRadii=<4 4 4 4>
      }
      tintColor {colorUseRule=UseColor_Specified, specifiedColor=<rgba 0.0395 0.0395 0.0395 1>}
      uVRegion {bIsValid=false, max=<0 0>, min=<0 0>}
     }
     disabledForeground {colorUseRule=UseColor_ColorTable, specifiedColor=<rgba 0 0 0 0>}
     hovered {
      bIsDynamicallyLoaded=false
      drawAs=RoundedBox
      imageType=NoImage
      mirroring=NoMirror
      resourceName=None
      resourceObject=null
      tiling=NoTile
      imageSize=<32 32>
      margin=<ltrb 0 0 0 0>
      outlineSettings {
       bUseBrushTransparency=false
       roundingType=HalfHeightRadius
       width=1
       color {colorUseRule=UseColor_ColorTable, specifiedColor=<rgba 0 0 0 0>}
       cornerRadii=<4 4 4 4>
      }
      tintColor {colorUseRule=UseColor_Specified, specifiedColor=<rgba 0.0953 0.0953 0.0953 1>}
      uVRegion {bIsValid=false, max=<0 0>, min=<0 0>}
     }
     hoveredForeground {colorUseRule=UseColor_ColorTable, specifiedColor=<rgba 0 0 0 0>}
     hoveredSlateSound {resourceObject=null}
     normal {
      bIsDynamicallyLoaded=false
      drawAs=RoundedBox
      imageType=NoImage
      mirroring=NoMirror
      resourceName=None
      resourceObject=null
      tiling=NoTile
      imageSize=<32 32>
      margin=<ltrb 0 0 0 0>
      outlineSettings {
       bUseBrushTransparency=false
       roundingType=HalfHeightRadius
       width=1
       color {colorUseRule=UseColor_ColorTable, specifiedColor=<rgba 0 0 0 0>}
       cornerRadii=<4 4 4 4>
      }
      tintColor {colorUseRule=UseColor_Specified, specifiedColor=<rgba 0.0395 0.0395 0.0395 1>}
      uVRegion {bIsValid=false, max=<0 0>, min=<0 0>}
     }
     normalForeground {colorUseRule=UseColor_ColorTable, specifiedColor=<rgba 0 0 0 0>}
     normalPadding=<ltrb 0 0 0 0>
     pressed {
      bIsDynamicallyLoaded=false
      drawAs=RoundedBox
      imageType=NoImage
      mirroring=NoMirror
      resourceName=None
      resourceObject=null
      tiling=NoTile
      imageSize=<32 32>
      margin=<ltrb 0 0 0 0>
      outlineSettings {
       bUseBrushTransparency=false
       roundingType=HalfHeightRadius
       width=1
       color {colorUseRule=UseColor_ColorTable, specifiedColor=<rgba 0 0 0 0>}
       cornerRadii=<4 4 4 4>
      }
      tintColor {colorUseRule=UseColor_Specified, specifiedColor=<rgba 0.0284 0.0284 0.0284 1>}
      uVRegion {bIsValid=false, max=<0 0>, min=<0 0>}
     }
     pressedForeground {colorUseRule=UseColor_ColorTable, specifiedColor=<rgba 0 0 0 0>}
     pressedPadding=<ltrb 0 0 0 0>
     pressedSlateSound {resourceObject=null}
    }}
   slot {class=HorizontalBoxSlot, properties {horizontalAlignment=HAlign_Center, verticalAlignment=VAlign_Center, padding=<ltrb 8 0 0 0>}}
  }
  Image_226 {
   widgetClassPath=&Image
   bIsVariable=true
   parent=Btn_Focus
   widget=Image_226
   properties {brush {
     bIsDynamicallyLoaded=false
     drawAs=Image
     imageType=NoImage
     mirroring=NoMirror
     resourceName=None
     resourceObject=/Engine/EditorResources/S_TargetPoint
     tiling=NoTile
     imageSize=<16 16>
     margin=<ltrb 0 0 0 0>
     outlineSettings {
      bUseBrushTransparency=false
      roundingType=HalfHeightRadius
      width=0
      color {colorUseRule=UseColor_Specified, specifiedColor=<rgba 0 0 0 0>}
      cornerRadii=<0 0 0 0>
     }
     tintColor {colorUseRule=UseColor_Specified, specifiedColor=<rgba 1 1 1 1>}
     uVRegion {bIsValid=false, max=<0 0>, min=<0 0>}
    }}
   slot {class=ButtonSlot, properties {padding=<ltrb 4 4 4 4>}}
  }
  ObjectName {
   widgetClassPath=&CommonTextBlock
   bIsVariable=true
   parent=Header
   widget=ObjectName
   properties {
    font {
     bForceMonospaced=false
     bMaterialIsStencil=false
     fontMaterial=null
     fontObject=&Roboto
     letterSpacing=0
     monospacedWidth=1
     size=11
     skewAmount=0
     typefaceFontName=Regular
     outlineSettings {
      bApplyOutlineToDropShadows=false
      bMiteredCorners=false
      bSeparateFillAlpha=false
      outlineMaterial=null
      outlineSize=0
      outlineColor=<rgba 0 0 0 1>
     }
    }
    shadowOffset=<0 0>
   }
   slot {class=HorizontalBoxSlot, properties {verticalAlignment=VAlign_Center, padding=<ltrb 8 0 0 0>}}
  }
  HorizontalBox_0 {
   widgetClassPath=&HorizontalBox
   parent=Header
   widget=HorizontalBox_0
   slot {class=HorizontalBoxSlot, properties {horizontalAlignment=HAlign_Right, padding=<ltrb 0 9 8 9>, size {sizeRule=Fill, value=1}}}
  }
  Btn_Batch {
   widgetClassPath=&Button
   parent=HorizontalBox_0
   widget=Btn_Batch
   properties {widgetStyle {
     clickedSlateSound {resourceObject=null}
     disabled {
      bIsDynamicallyLoaded=false
      drawAs=RoundedBox
      imageType=NoImage
      mirroring=NoMirror
      resourceName=None
      resourceObject=null
      tiling=NoTile
      imageSize=<0 0>
      margin=<ltrb 0 0 0 0>
      outlineSettings {
       bUseBrushTransparency=false
       roundingType=FixedRadius
       width=1
       color {colorUseRule=UseColor_ColorTable, specifiedColor=<rgba 0 0 0 0>}
       cornerRadii=<4 4 4 4>
      }
      tintColor {colorUseRule=UseColor_Specified, specifiedColor=<rgba 0.0395 0.0395 0.0395 1>}
      uVRegion {bIsValid=false, max=<0 0>, min=<0 0>}
     }
     disabledForeground {colorUseRule=UseColor_ColorTable, specifiedColor=<rgba 0 0 0 0>}
     hovered {
      bIsDynamicallyLoaded=false
      drawAs=RoundedBox
      imageType=NoImage
      mirroring=NoMirror
      resourceName=None
      resourceObject=null
      tiling=NoTile
      imageSize=<32 32>
      margin=<ltrb 0 0 0 0>
      outlineSettings {
       bUseBrushTransparency=false
       roundingType=FixedRadius
       width=1
       color {colorUseRule=UseColor_ColorTable, specifiedColor=<rgba 0 0 0 0>}
       cornerRadii=<4 4 4 4>
      }
      tintColor {colorUseRule=UseColor_Specified, specifiedColor=<rgba 0.0953 0.0953 0.0953 1>}
      uVRegion {bIsValid=false, max=<0 0>, min=<0 0>}
     }
     hoveredForeground {colorUseRule=UseColor_ColorTable, specifiedColor=<rgba 0 0 0 0>}
     hoveredSlateSound {resourceObject=null}
     normal {
      bIsDynamicallyLoaded=false
      drawAs=RoundedBox
      imageType=NoImage
      mirroring=NoMirror
      resourceName=None
      resourceObject=null
      tiling=NoTile
      imageSize=<32 32>
      margin=<ltrb 0 0 0 0>
      outlineSettings {
       bUseBrushTransparency=false
       roundingType=FixedRadius
       width=1
       color {colorUseRule=UseColor_ColorTable, specifiedColor=<rgba 0 0 0 0>}
       cornerRadii=<4 4 4 4>
      }
      tintColor {colorUseRule=UseColor_Specified, specifiedColor=<rgba 0.0395 0.0395 0.0395 1>}
      uVRegion {bIsValid=false, max=<0 0>, min=<0 0>}
     }
     normalForeground {colorUseRule=UseColor_ColorTable, specifiedColor=<rgba 0 0 0 0>}
     normalPadding=<ltrb 12 1.5 12 1.5>
     pressed {
      bIsDynamicallyLoaded=false
      drawAs=RoundedBox
      imageType=NoImage
      mirroring=NoMirror
      resourceName=None
      resourceObject=null
      tiling=NoTile
      imageSize=<32 32>
      margin=<ltrb 0 0 0 0>
      outlineSettings {
       bUseBrushTransparency=false
       roundingType=FixedRadius
       width=1
       color {colorUseRule=UseColor_ColorTable, specifiedColor=<rgba 0 0 0 0>}
       cornerRadii=<4 4 4 4>
      }
      tintColor {colorUseRule=UseColor_Specified, specifiedColor=<rgba 0.0284 0.0284 0.0284 1>}
      uVRegion {bIsValid=false, max=<0 0>, min=<0 0>}
     }
     pressedForeground {colorUseRule=UseColor_ColorTable, specifiedColor=<rgba 0 0 0 0>}
     pressedPadding=<ltrb 12 2.5 12 0.5>
     pressedSlateSound {resourceObject=null}
    }}
   slot {class=HorizontalBoxSlot, properties {padding=<ltrb 0 0 10 0>}}
  }
  HorizontalBox {
   widgetClassPath=&HorizontalBox
   parent=Btn_Batch
   widget=HorizontalBox
   slot {class=ButtonSlot, properties {padding=<ltrb 4 5 4 5>}}
  }
  CommonTextBlock_9 {
   widgetClassPath=&CommonTextBlock
   parent=HorizontalBox
   widget=CommonTextBlock_9
   properties {
    text=Batch Tool
    font {
     bForceMonospaced=false
     bMaterialIsStencil=false
     fontMaterial=null
     fontObject=&Roboto
     letterSpacing=0
     monospacedWidth=1
     size=9
     skewAmount=0
     typefaceFontName=Regular
     outlineSettings {
      bApplyOutlineToDropShadows=false
      bMiteredCorners=false
      bSeparateFillAlpha=false
      outlineMaterial=null
      outlineSize=0
      outlineColor=<rgba 0 0 0 1>
     }
    }
    shadowOffset=<0 0>
   }
   slot {class=HorizontalBoxSlot, properties {horizontalAlignment=HAlign_Center, verticalAlignment=VAlign_Center}}
  }
  Btn_Add {
   widgetClassPath=&Button
   parent=HorizontalBox_0
   widget=Btn_Add
   properties {widgetStyle {
     clickedSlateSound {resourceObject=null}
     disabled {
      bIsDynamicallyLoaded=false
      drawAs=RoundedBox
      imageType=NoImage
      mirroring=NoMirror
      resourceName=None
      resourceObject=null
      tiling=NoTile
      imageSize=<0 0>
      margin=<ltrb 0 0 0 0>
      outlineSettings {
       bUseBrushTransparency=false
       roundingType=FixedRadius
       width=1
       color {colorUseRule=UseColor_ColorTable, specifiedColor=<rgba 0 0 0 0>}
       cornerRadii=<4 4 4 4>
      }
      tintColor {colorUseRule=UseColor_Specified, specifiedColor=<rgba 0.0395 0.0395 0.0395 1>}
      uVRegion {bIsValid=false, max=<0 0>, min=<0 0>}
     }
     disabledForeground {colorUseRule=UseColor_ColorTable, specifiedColor=<rgba 0 0 0 0>}
     hovered {
      bIsDynamicallyLoaded=false
      drawAs=RoundedBox
      imageType=NoImage
      mirroring=NoMirror
      resourceName=None
      resourceObject=null
      tiling=NoTile
      imageSize=<32 32>
      margin=<ltrb 0 0 0 0>
      outlineSettings {
       bUseBrushTransparency=false
       roundingType=FixedRadius
       width=1
       color {colorUseRule=UseColor_ColorTable, specifiedColor=<rgba 0 0 0 0>}
       cornerRadii=<4 4 4 4>
      }
      tintColor {colorUseRule=UseColor_Specified, specifiedColor=<rgba 0.0953 0.0953 0.0953 1>}
      uVRegion {bIsValid=false, max=<0 0>, min=<0 0>}
     }
     hoveredForeground {colorUseRule=UseColor_ColorTable, specifiedColor=<rgba 0 0 0 0>}
     hoveredSlateSound {resourceObject=null}
     normal {
      bIsDynamicallyLoaded=false
      drawAs=RoundedBox
      imageType=NoImage
      mirroring=NoMirror
      resourceName=None
      resourceObject=null
      tiling=NoTile
      imageSize=<32 32>
      margin=<ltrb 0 0 0 0>
      outlineSettings {
       bUseBrushTransparency=false
       roundingType=FixedRadius
       width=1
       color {colorUseRule=UseColor_ColorTable, specifiedColor=<rgba 0 0 0 0>}
       cornerRadii=<4 4 4 4>
      }
      tintColor {colorUseRule=UseColor_Specified, specifiedColor=<rgba 0.0395 0.0395 0.0395 1>}
      uVRegion {bIsValid=false, max=<0 0>, min=<0 0>}
     }
     normalForeground {colorUseRule=UseColor_ColorTable, specifiedColor=<rgba 0 0 0 0>}
     normalPadding=<ltrb 12 1.5 12 1.5>
     pressed {
      bIsDynamicallyLoaded=false
      drawAs=RoundedBox
      imageType=NoImage
      mirroring=NoMirror
      resourceName=None
      resourceObject=null
      tiling=NoTile
      imageSize=<32 32>
      margin=<ltrb 0 0 0 0>
      outlineSettings {
       bUseBrushTransparency=false
       roundingType=FixedRadius
       width=1
       color {colorUseRule=UseColor_ColorTable, specifiedColor=<rgba 0 0 0 0>}
       cornerRadii=<4 4 4 4>
      }
      tintColor {colorUseRule=UseColor_Specified, specifiedColor=<rgba 0.0284 0.0284 0.0284 1>}
      uVRegion {bIsValid=false, max=<0 0>, min=<0 0>}
     }
     pressedForeground {colorUseRule=UseColor_ColorTable, specifiedColor=<rgba 0 0 0 0>}
     pressedPadding=<ltrb 12 2.5 12 0.5>
     pressedSlateSound {resourceObject=null}
    }}
   slot {class=HorizontalBoxSlot}
  }
  HorizontalBox_37 {
   widgetClassPath=&HorizontalBox
   parent=Btn_Add
   widget=HorizontalBox_37
   slot {class=ButtonSlot, properties {padding=<ltrb 4 5 4 5>}}
  }
  Image_71 {
   widgetClassPath=&Image
   bIsVariable=true
   parent=HorizontalBox_37
   widget=Image_71
   properties {brush {
     bIsDynamicallyLoaded=false
     drawAs=Image
     imageType=NoImage
     mirroring=NoMirror
     resourceName=None
     resourceObject=/Game/ArtTools/Tech_Art/InstancedMeshesCluster/Icon/ClusterAdd
     tiling=NoTile
     imageSize=<14 14>
     margin=<ltrb 0 0 0 0>
     outlineSettings {
      bUseBrushTransparency=false
      roundingType=HalfHeightRadius
      width=0
      color {colorUseRule=UseColor_Specified, specifiedColor=<rgba 0 0 0 0>}
      cornerRadii=<0 0 0 0>
     }
     tintColor {colorUseRule=UseColor_Specified, specifiedColor=<rgba 1 1 1 1>}
     uVRegion {bIsValid=false, max=<0 0>, min=<0 0>}
    }}
   slot {class=HorizontalBoxSlot, properties {horizontalAlignment=HAlign_Center, padding=<ltrb 0 0 3 0>}}
  }
  CommonTextBlock {
   widgetClassPath=&CommonTextBlock
   parent=HorizontalBox_37
   widget=CommonTextBlock
   properties {
    text=New
    font {
     bForceMonospaced=false
     bMaterialIsStencil=false
     fontMaterial=null
     fontObject=&Roboto
     letterSpacing=0
     monospacedWidth=1
     size=9
     skewAmount=0
     typefaceFontName=Regular
     outlineSettings {
      bApplyOutlineToDropShadows=false
      bMiteredCorners=false
      bSeparateFillAlpha=false
      outlineMaterial=null
      outlineSize=0
      outlineColor=<rgba 0 0 0 1>
     }
    }
    shadowOffset=<0 0>
   }
   slot {class=HorizontalBoxSlot, properties {horizontalAlignment=HAlign_Center, verticalAlignment=VAlign_Center}}
  }
  MeshPanel {
   widgetClassPath=&CommonBorder
   bIsVariable=true
   parent=CanvasPanel_0
   widget=MeshPanel
   properties {
    background {
     bIsDynamicallyLoaded=false
     drawAs=Image
     imageType=NoImage
     mirroring=NoMirror
     resourceName=None
     resourceObject=&T_WhiteMask
     tiling=NoTile
     imageSize=<32 32>
     margin=<ltrb 0 0 0 0>
     outlineSettings {
      bUseBrushTransparency=false
      roundingType=HalfHeightRadius
      width=0
      color {colorUseRule=UseColor_Specified, specifiedColor=<rgba 0 0 0 0>}
      cornerRadii=<0 0 0 0>
     }
     tintColor {colorUseRule=UseColor_Specified, specifiedColor=<rgba 1 1 1 1>}
     uVRegion {bIsValid=false, max=<0 0>, min=<0 0>}
    }
    brushColor=<rgba 0.0103 0.0103 0.0103 1>
    padding=<ltrb 0 0 0 0>
   }
   slot {class=CanvasPanelSlot, properties {layoutData {alignment=<0 0>, anchors {maximum=<1 0>, minimum=<0 0>}, offsets=<ltrb 4 46 0 192>}}}
  }
  Instances {widgetClassPath=&HorizontalBox, parent=MeshPanel, widget=Instances, slot {class=BorderSlot, properties {padding=<ltrb 0 0 0 0>}}}
  ScrollBox_50 {
   widgetClassPath=&ScrollBox
   parent=Instances
   widget=ScrollBox_50
   properties {
    widgetBarStyle {
     thickness=8
     draggedThumbImage {
      bIsDynamicallyLoaded=false
      drawAs=RoundedBox
      imageType=NoImage
      mirroring=NoMirror
      resourceName=None
      resourceObject=null
      tiling=NoTile
      imageSize=<0 0>
      margin=<ltrb 0 0 0 0>
      outlineSettings {
       bUseBrushTransparency=false
       roundingType=FixedRadius
       width=0
       color {colorUseRule=UseColor_Specified, specifiedColor=<rgba 0 0 0 0>}
       cornerRadii=<4 4 4 4>
      }
      tintColor {colorUseRule=UseColor_Specified, specifiedColor=<rgba 0.2159 0.2159 0.2159 1>}
      uVRegion {bIsValid=false, max=<0 0>, min=<0 0>}
     }
     horizontalBackgroundImage {
      bIsDynamicallyLoaded=false
      drawAs=NoDrawType
      imageType=NoImage
      mirroring=NoMirror
      resourceName=None
      resourceObject=null
      tiling=NoTile
      imageSize=<0 0>
      margin=<ltrb 0 0 0 0>
      outlineSettings {
       bUseBrushTransparency=false
       roundingType=HalfHeightRadius
       width=0
       color {colorUseRule=UseColor_Specified, specifiedColor=<rgba 0 0 0 0>}
       cornerRadii=<0 0 0 0>
      }
      tintColor {colorUseRule=UseColor_Specified, specifiedColor=<rgba 1 1 1 1>}
      uVRegion {bIsValid=false, max=<0 0>, min=<0 0>}
     }
     horizontalBottomSlotImage {
      bIsDynamicallyLoaded=false
      drawAs=NoDrawType
      imageType=NoImage
      mirroring=NoMirror
      resourceName=None
      resourceObject=null
      tiling=NoTile
      imageSize=<0 0>
      margin=<ltrb 0 0 0 0>
      outlineSettings {
       bUseBrushTransparency=false
       roundingType=HalfHeightRadius
       width=0
       color {colorUseRule=UseColor_Specified, specifiedColor=<rgba 0 0 0 0>}
       cornerRadii=<0 0 0 0>
      }
      tintColor {colorUseRule=UseColor_Specified, specifiedColor=<rgba 1 1 1 1>}
      uVRegion {bIsValid=false, max=<0 0>, min=<0 0>}
     }
     horizontalTopSlotImage {
      bIsDynamicallyLoaded=false
      drawAs=NoDrawType
      imageType=NoImage
      mirroring=NoMirror
      resourceName=None
      resourceObject=null
      tiling=NoTile
      imageSize=<0 0>
      margin=<ltrb 0 0 0 0>
      outlineSettings {
       bUseBrushTransparency=false
       roundingType=HalfHeightRadius
       width=0
       color {colorUseRule=UseColor_Specified, specifiedColor=<rgba 0 0 0 0>}
       cornerRadii=<0 0 0 0>
      }
      tintColor {colorUseRule=UseColor_Specified, specifiedColor=<rgba 1 1 1 1>}
      uVRegion {bIsValid=false, max=<0 0>, min=<0 0>}
     }
     hoveredThumbImage {
      bIsDynamicallyLoaded=false
      drawAs=RoundedBox
      imageType=NoImage
      mirroring=NoMirror
      resourceName=None
      resourceObject=null
      tiling=NoTile
      imageSize=<0 0>
      margin=<ltrb 0 0 0 0>
      outlineSettings {
       bUseBrushTransparency=false
       roundingType=FixedRadius
       width=0
       color {colorUseRule=UseColor_Specified, specifiedColor=<rgba 0 0 0 0>}
       cornerRadii=<4 4 4 4>
      }
      tintColor {colorUseRule=UseColor_Specified, specifiedColor=<rgba 0.2159 0.2159 0.2159 1>}
      uVRegion {bIsValid=false, max=<0 0>, min=<0 0>}
     }
     normalThumbImage {
      bIsDynamicallyLoaded=false
      drawAs=RoundedBox
      imageType=NoImage
      mirroring=NoMirror
      resourceName=None
      resourceObject=null
      tiling=NoTile
      imageSize=<0 0>
      margin=<ltrb 0 0 0 0>
      outlineSettings {
       bUseBrushTransparency=false
       roundingType=FixedRadius
       width=0
       color {colorUseRule=UseColor_Specified, specifiedColor=<rgba 0 0 0 0>}
       cornerRadii=<4 4 4 4>
      }
      tintColor {colorUseRule=UseColor_Specified, specifiedColor=<rgba 0.0953 0.0953 0.0953 1>}
      uVRegion {bIsValid=false, max=<0 0>, min=<0 0>}
     }
     verticalBackgroundImage {
      bIsDynamicallyLoaded=false
      drawAs=NoDrawType
      imageType=NoImage
      mirroring=NoMirror
      resourceName=None
      resourceObject=null
      tiling=NoTile
      imageSize=<0 0>
      margin=<ltrb 0 0 0 0>
      outlineSettings {
       bUseBrushTransparency=false
       roundingType=HalfHeightRadius
       width=0
       color {colorUseRule=UseColor_Specified, specifiedColor=<rgba 0 0 0 0>}
       cornerRadii=<0 0 0 0>
      }
      tintColor {colorUseRule=UseColor_Specified, specifiedColor=<rgba 1 1 1 1>}
      uVRegion {bIsValid=false, max=<0 0>, min=<0 0>}
     }
     verticalBottomSlotImage {
      bIsDynamicallyLoaded=false
      drawAs=NoDrawType
      imageType=NoImage
      mirroring=NoMirror
      resourceName=None
      resourceObject=null
      tiling=NoTile
      imageSize=<0 0>
      margin=<ltrb 0 0 0 0>
      outlineSettings {
       bUseBrushTransparency=false
       roundingType=HalfHeightRadius
       width=0
       color {colorUseRule=UseColor_Specified, specifiedColor=<rgba 0 0 0 0>}
       cornerRadii=<0 0 0 0>
      }
      tintColor {colorUseRule=UseColor_Specified, specifiedColor=<rgba 1 1 1 1>}
      uVRegion {bIsValid=false, max=<0 0>, min=<0 0>}
     }
     verticalTopSlotImage {
      bIsDynamicallyLoaded=false
      drawAs=NoDrawType
      imageType=NoImage
      mirroring=NoMirror
      resourceName=None
      resourceObject=null
      tiling=NoTile
      imageSize=<0 0>
      margin=<ltrb 0 0 0 0>
      outlineSettings {
       bUseBrushTransparency=false
       roundingType=HalfHeightRadius
       width=0
       color {colorUseRule=UseColor_Specified, specifiedColor=<rgba 0 0 0 0>}
       cornerRadii=<0 0 0 0>
      }
      tintColor {colorUseRule=UseColor_Specified, specifiedColor=<rgba 1 1 1 1>}
      uVRegion {bIsValid=false, max=<0 0>, min=<0 0>}
     }
    }
    widgetStyle {
     barThickness=8
     bottomShadowBrush {
      bIsDynamicallyLoaded=false
      drawAs=NoDrawType
      imageType=FullColor
      mirroring=NoMirror
      resourceName=../../../Engine/Content/Slate/Common/ScrollBoxShadowBottom.png
      resourceObject=null
      tiling=NoTile
      imageSize=<16 8>
      margin=<ltrb 0.5 0 0.5 1>
      outlineSettings {
       bUseBrushTransparency=false
       roundingType=HalfHeightRadius
       width=0
       color {colorUseRule=UseColor_Specified, specifiedColor=<rgba 0 0 0 0>}
       cornerRadii=<0 0 0 0>
      }
      tintColor {colorUseRule=UseColor_Specified, specifiedColor=<rgba 1 1 1 1>}
      uVRegion {bIsValid=false, max=<0 0>, min=<0 0>}
     }
     horizontalScrolledContentPadding=<ltrb 0 0 1 0>
     leftShadowBrush {
      bIsDynamicallyLoaded=false
      drawAs=NoDrawType
      imageType=FullColor
      mirroring=NoMirror
      resourceName=../../../Engine/Content/Slate/Common/ScrollBoxShadowLeft.png
      resourceObject=null
      tiling=NoTile
      imageSize=<8 16>
      margin=<ltrb 1 0.5 0 0.5>
      outlineSettings {
       bUseBrushTransparency=false
       roundingType=HalfHeightRadius
       width=0
       color {colorUseRule=UseColor_Specified, specifiedColor=<rgba 0 0 0 0>}
       cornerRadii=<0 0 0 0>
      }
      tintColor {colorUseRule=UseColor_Specified, specifiedColor=<rgba 1 1 1 1>}
      uVRegion {bIsValid=false, max=<0 0>, min=<0 0>}
     }
     rightShadowBrush {
      bIsDynamicallyLoaded=false
      drawAs=NoDrawType
      imageType=FullColor
      mirroring=NoMirror
      resourceName=../../../Engine/Content/Slate/Common/ScrollBoxShadowRight.png
      resourceObject=null
      tiling=NoTile
      imageSize=<8 16>
      margin=<ltrb 0 0.5 1 0.5>
      outlineSettings {
       bUseBrushTransparency=false
       roundingType=HalfHeightRadius
       width=0
       color {colorUseRule=UseColor_Specified, specifiedColor=<rgba 0 0 0 0>}
       cornerRadii=<0 0 0 0>
      }
      tintColor {colorUseRule=UseColor_Specified, specifiedColor=<rgba 1 1 1 1>}
      uVRegion {bIsValid=false, max=<0 0>, min=<0 0>}
     }
     topShadowBrush {
      bIsDynamicallyLoaded=false
      drawAs=NoDrawType
      imageType=FullColor
      mirroring=NoMirror
      resourceName=../../../Engine/Content/Slate/Common/ScrollBoxShadowTop.png
      resourceObject=null
      tiling=NoTile
      imageSize=<16 8>
      margin=<ltrb 0.5 1 0.5 0>
      outlineSettings {
       bUseBrushTransparency=false
       roundingType=HalfHeightRadius
       width=0
       color {colorUseRule=UseColor_Specified, specifiedColor=<rgba 0 0 0 0>}
       cornerRadii=<0 0 0 0>
      }
      tintColor {colorUseRule=UseColor_Specified, specifiedColor=<rgba 1 1 1 1>}
      uVRegion {bIsValid=false, max=<0 0>, min=<0 0>}
     }
     verticalScrolledContentPadding=<ltrb 0 0 0 1>
    }
   }
   slot {class=HorizontalBoxSlot, properties {size {sizeRule=Fill, value=1}}}
  }
  InstanceDetails {widgetClassPath=&VerticalBox, parent=ScrollBox_50, widget=InstanceDetails, slot {class=ScrollBoxSlot}}
  Element_0 {
   widgetClassPath=/Game/ArtTools/Tech_Art/InstancedMeshesCluster/SubWidgets/IMC_Item_Properties
   bIsVariable=true
   parent=InstanceDetails
   widget=Element_0
   properties {scaleProp=ScaleMin/Max0}
   slot {class=VerticalBoxSlot}
  }
  Element_1 {
   widgetClassPath=&IMC_Item_Properties
   bIsVariable=true
   parent=InstanceDetails
   widget=Element_1
   properties {scaleProp=ScaleMin/Max1}
   slot {class=VerticalBoxSlot}
  }
  Element_2 {
   widgetClassPath=&IMC_Item_Properties
   bIsVariable=true
   parent=InstanceDetails
   widget=Element_2
   properties {scaleProp=ScaleMin/Max2}
   slot {class=VerticalBoxSlot}
  }
  Element_3 {
   widgetClassPath=&IMC_Item_Properties
   bIsVariable=true
   parent=InstanceDetails
   widget=Element_3
   properties {scaleProp=ScaleMin/Max3}
   slot {class=VerticalBoxSlot}
  }
  Element_4 {
   widgetClassPath=&IMC_Item_Properties
   bIsVariable=true
   parent=InstanceDetails
   widget=Element_4
   properties {scaleProp=ScaleMin/Max4}
   slot {class=VerticalBoxSlot}
  }
  HorizontalBox_216 {widgetClassPath=&HorizontalBox, parent=Instances, widget=HorizontalBox_216, slot {class=HorizontalBoxSlot}}
  VerticalBox_137 {widgetClassPath=&VerticalBox, parent=HorizontalBox_216, widget=VerticalBox_137, slot {class=HorizontalBoxSlot}}
  PropCheckBox0 {
   widgetClassPath=&SinglePropertyView
   parent=VerticalBox_137
   widget=PropCheckBox0
   properties {nameOverride=" ", propertyName=EnableCluster0}
   slot {class=VerticalBoxSlot, properties {padding=<ltrb -10 8 -20 8>}}
  }
  PropCheckBox1 {
   widgetClassPath=&SinglePropertyView
   parent=VerticalBox_137
   widget=PropCheckBox1
   properties {nameOverride=" ", propertyName=EnableCluster1}
   slot {class=VerticalBoxSlot, properties {padding=<ltrb -10 8 -20 8>}}
  }
  PropCheckBox2 {
   widgetClassPath=&SinglePropertyView
   parent=VerticalBox_137
   widget=PropCheckBox2
   properties {nameOverride=" ", propertyName=EnableCluster2}
   slot {class=VerticalBoxSlot, properties {padding=<ltrb -10 8 -20 8>}}
  }
  PropCheckBox3 {
   widgetClassPath=&SinglePropertyView
   parent=VerticalBox_137
   widget=PropCheckBox3
   properties {nameOverride=" ", propertyName=EnableCluster3}
   slot {class=VerticalBoxSlot, properties {padding=<ltrb -10 8 -20 8>}}
  }
  PropCheckBox4 {
   widgetClassPath=&SinglePropertyView
   parent=VerticalBox_137
   widget=PropCheckBox4
   properties {nameOverride=" ", propertyName=EnableCluster4}
   slot {class=VerticalBoxSlot, properties {padding=<ltrb -10 8 -20 8>}}
  }
  MessagePanel {
   widgetClassPath=&CommonBorder
   bIsVariable=true
   parent=CanvasPanel_0
   widget=MessagePanel
   properties {
    horizontalAlignment=HAlign_Center
    verticalAlignment=VAlign_Center
    background {
     bIsDynamicallyLoaded=false
     drawAs=Image
     imageType=NoImage
     mirroring=NoMirror
     resourceName=None
     resourceObject=&T_WhiteMask
     tiling=NoTile
     imageSize=<32 32>
     margin=<ltrb 0 0 0 0>
     outlineSettings {
      bUseBrushTransparency=false
      roundingType=HalfHeightRadius
      width=0
      color {colorUseRule=UseColor_Specified, specifiedColor=<rgba 0 0 0 0>}
      cornerRadii=<0 0 0 0>
     }
     tintColor {colorUseRule=UseColor_Specified, specifiedColor=<rgba 1 1 1 1>}
     uVRegion {bIsValid=false, max=<0 0>, min=<0 0>}
    }
    brushColor=<rgba 0.0103 0.0103 0.0103 1>
    padding=<ltrb 0 0 0 0>
   }
   slot {class=CanvasPanelSlot, properties {layoutData {alignment=<0 0>, anchors {maximum=<1 0>, minimum=<0 0>}, offsets=<ltrb 4 46 4 192>}}}
  }
  CommonTextBlock_36 {
   widgetClassPath=&CommonTextBlock
   parent=MessagePanel
   widget=CommonTextBlock_36
   properties {
    text=Select Cluster Mesh
    font {
     bForceMonospaced=false
     bMaterialIsStencil=false
     fontMaterial=null
     fontObject=&Roboto
     letterSpacing=0
     monospacedWidth=1
     size=16
     skewAmount=0
     typefaceFontName=Regular
     outlineSettings {
      bApplyOutlineToDropShadows=false
      bMiteredCorners=false
      bSeparateFillAlpha=false
      outlineMaterial=null
      outlineSize=0
      outlineColor=<rgba 0 0 0 1>
     }
    }
    shadowOffset=<0 0>
   }
   slot {class=BorderSlot, properties {horizontalAlignment=HAlign_Center, verticalAlignment=VAlign_Center, padding=<ltrb 0 0 0 0>}}
  }
  ScrollBox_39 {
   widgetClassPath=&ScrollBox
   parent=CanvasPanel_0
   widget=ScrollBox_39
   properties {
    widgetBarStyle {
     thickness=8
     draggedThumbImage {
      bIsDynamicallyLoaded=false
      drawAs=RoundedBox
      imageType=NoImage
      mirroring=NoMirror
      resourceName=None
      resourceObject=null
      tiling=NoTile
      imageSize=<0 0>
      margin=<ltrb 0 0 0 0>
      outlineSettings {
       bUseBrushTransparency=false
       roundingType=FixedRadius
       width=0
       color {colorUseRule=UseColor_Specified, specifiedColor=<rgba 0 0 0 0>}
       cornerRadii=<4 4 4 4>
      }
      tintColor {colorUseRule=UseColor_Specified, specifiedColor=<rgba 0.2159 0.2159 0.2159 1>}
      uVRegion {bIsValid=false, max=<0 0>, min=<0 0>}
     }
     horizontalBackgroundImage {
      bIsDynamicallyLoaded=false
      drawAs=NoDrawType
      imageType=NoImage
      mirroring=NoMirror
      resourceName=None
      resourceObject=null
      tiling=NoTile
      imageSize=<0 0>
      margin=<ltrb 0 0 0 0>
      outlineSettings {
       bUseBrushTransparency=false
       roundingType=HalfHeightRadius
       width=0
       color {colorUseRule=UseColor_Specified, specifiedColor=<rgba 0 0 0 0>}
       cornerRadii=<0 0 0 0>
      }
      tintColor {colorUseRule=UseColor_Specified, specifiedColor=<rgba 1 1 1 1>}
      uVRegion {bIsValid=false, max=<0 0>, min=<0 0>}
     }
     horizontalBottomSlotImage {
      bIsDynamicallyLoaded=false
      drawAs=NoDrawType
      imageType=NoImage
      mirroring=NoMirror
      resourceName=None
      resourceObject=null
      tiling=NoTile
      imageSize=<0 0>
      margin=<ltrb 0 0 0 0>
      outlineSettings {
       bUseBrushTransparency=false
       roundingType=HalfHeightRadius
       width=0
       color {colorUseRule=UseColor_Specified, specifiedColor=<rgba 0 0 0 0>}
       cornerRadii=<0 0 0 0>
      }
      tintColor {colorUseRule=UseColor_Specified, specifiedColor=<rgba 1 1 1 1>}
      uVRegion {bIsValid=false, max=<0 0>, min=<0 0>}
     }
     horizontalTopSlotImage {
      bIsDynamicallyLoaded=false
      drawAs=NoDrawType
      imageType=NoImage
      mirroring=NoMirror
      resourceName=None
      resourceObject=null
      tiling=NoTile
      imageSize=<0 0>
      margin=<ltrb 0 0 0 0>
      outlineSettings {
       bUseBrushTransparency=false
       roundingType=HalfHeightRadius
       width=0
       color {colorUseRule=UseColor_Specified, specifiedColor=<rgba 0 0 0 0>}
       cornerRadii=<0 0 0 0>
      }
      tintColor {colorUseRule=UseColor_Specified, specifiedColor=<rgba 1 1 1 1>}
      uVRegion {bIsValid=false, max=<0 0>, min=<0 0>}
     }
     hoveredThumbImage {
      bIsDynamicallyLoaded=false
      drawAs=RoundedBox
      imageType=NoImage
      mirroring=NoMirror
      resourceName=None
      resourceObject=null
      tiling=NoTile
      imageSize=<0 0>
      margin=<ltrb 0 0 0 0>
      outlineSettings {
       bUseBrushTransparency=false
       roundingType=FixedRadius
       width=0
       color {colorUseRule=UseColor_Specified, specifiedColor=<rgba 0 0 0 0>}
       cornerRadii=<4 4 4 4>
      }
      tintColor {colorUseRule=UseColor_Specified, specifiedColor=<rgba 0.2159 0.2159 0.2159 1>}
      uVRegion {bIsValid=false, max=<0 0>, min=<0 0>}
     }
     normalThumbImage {
      bIsDynamicallyLoaded=false
      drawAs=RoundedBox
      imageType=NoImage
      mirroring=NoMirror
      resourceName=None
      resourceObject=null
      tiling=NoTile
      imageSize=<0 0>
      margin=<ltrb 0 0 0 0>
      outlineSettings {
       bUseBrushTransparency=false
       roundingType=FixedRadius
       width=0
       color {colorUseRule=UseColor_Specified, specifiedColor=<rgba 0 0 0 0>}
       cornerRadii=<4 4 4 4>
      }
      tintColor {colorUseRule=UseColor_Specified, specifiedColor=<rgba 0.0953 0.0953 0.0953 1>}
      uVRegion {bIsValid=false, max=<0 0>, min=<0 0>}
     }
     verticalBackgroundImage {
      bIsDynamicallyLoaded=false
      drawAs=NoDrawType
      imageType=NoImage
      mirroring=NoMirror
      resourceName=None
      resourceObject=null
      tiling=NoTile
      imageSize=<0 0>
      margin=<ltrb 0 0 0 0>
      outlineSettings {
       bUseBrushTransparency=false
       roundingType=HalfHeightRadius
       width=0
       color {colorUseRule=UseColor_Specified, specifiedColor=<rgba 0 0 0 0>}
       cornerRadii=<0 0 0 0>
      }
      tintColor {colorUseRule=UseColor_Specified, specifiedColor=<rgba 1 1 1 1>}
      uVRegion {bIsValid=false, max=<0 0>, min=<0 0>}
     }
     verticalBottomSlotImage {
      bIsDynamicallyLoaded=false
      drawAs=NoDrawType
      imageType=NoImage
      mirroring=NoMirror
      resourceName=None
      resourceObject=null
      tiling=NoTile
      imageSize=<0 0>
      margin=<ltrb 0 0 0 0>
      outlineSettings {
       bUseBrushTransparency=false
       roundingType=HalfHeightRadius
       width=0
       color {colorUseRule=UseColor_Specified, specifiedColor=<rgba 0 0 0 0>}
       cornerRadii=<0 0 0 0>
      }
      tintColor {colorUseRule=UseColor_Specified, specifiedColor=<rgba 1 1 1 1>}
      uVRegion {bIsValid=false, max=<0 0>, min=<0 0>}
     }
     verticalTopSlotImage {
      bIsDynamicallyLoaded=false
      drawAs=NoDrawType
      imageType=NoImage
      mirroring=NoMirror
      resourceName=None
      resourceObject=null
      tiling=NoTile
      imageSize=<0 0>
      margin=<ltrb 0 0 0 0>
      outlineSettings {
       bUseBrushTransparency=false
       roundingType=HalfHeightRadius
       width=0
       color {colorUseRule=UseColor_Specified, specifiedColor=<rgba 0 0 0 0>}
       cornerRadii=<0 0 0 0>
      }
      tintColor {colorUseRule=UseColor_Specified, specifiedColor=<rgba 1 1 1 1>}
      uVRegion {bIsValid=false, max=<0 0>, min=<0 0>}
     }
    }
    widgetStyle {
     barThickness=8
     bottomShadowBrush {
      bIsDynamicallyLoaded=false
      drawAs=Box
      imageType=FullColor
      mirroring=NoMirror
      resourceName=../../../Engine/Content/Slate/Common/ScrollBoxShadowBottom.png
      resourceObject=null
      tiling=NoTile
      imageSize=<16 8>
      margin=<ltrb 0.5 0 0.5 1>
      outlineSettings {
       bUseBrushTransparency=false
       roundingType=HalfHeightRadius
       width=0
       color {colorUseRule=UseColor_Specified, specifiedColor=<rgba 0 0 0 0>}
       cornerRadii=<0 0 0 0>
      }
      tintColor {colorUseRule=UseColor_Specified, specifiedColor=<rgba 1 1 1 1>}
      uVRegion {bIsValid=false, max=<0 0>, min=<0 0>}
     }
     horizontalScrolledContentPadding=<ltrb 0 0 1 0>
     leftShadowBrush {
      bIsDynamicallyLoaded=false
      drawAs=Box
      imageType=FullColor
      mirroring=NoMirror
      resourceName=../../../Engine/Content/Slate/Common/ScrollBoxShadowLeft.png
      resourceObject=null
      tiling=NoTile
      imageSize=<8 16>
      margin=<ltrb 1 0.5 0 0.5>
      outlineSettings {
       bUseBrushTransparency=false
       roundingType=HalfHeightRadius
       width=0
       color {colorUseRule=UseColor_Specified, specifiedColor=<rgba 0 0 0 0>}
       cornerRadii=<0 0 0 0>
      }
      tintColor {colorUseRule=UseColor_Specified, specifiedColor=<rgba 1 1 1 1>}
      uVRegion {bIsValid=false, max=<0 0>, min=<0 0>}
     }
     rightShadowBrush {
      bIsDynamicallyLoaded=false
      drawAs=Box
      imageType=FullColor
      mirroring=NoMirror
      resourceName=../../../Engine/Content/Slate/Common/ScrollBoxShadowRight.png
      resourceObject=null
      tiling=NoTile
      imageSize=<8 16>
      margin=<ltrb 0 0.5 1 0.5>
      outlineSettings {
       bUseBrushTransparency=false
       roundingType=HalfHeightRadius
       width=0
       color {colorUseRule=UseColor_Specified, specifiedColor=<rgba 0 0 0 0>}
       cornerRadii=<0 0 0 0>
      }
      tintColor {colorUseRule=UseColor_Specified, specifiedColor=<rgba 1 1 1 1>}
      uVRegion {bIsValid=false, max=<0 0>, min=<0 0>}
     }
     topShadowBrush {
      bIsDynamicallyLoaded=false
      drawAs=Box
      imageType=FullColor
      mirroring=NoMirror
      resourceName=../../../Engine/Content/Slate/Common/ScrollBoxShadowTop.png
      resourceObject=null
      tiling=NoTile
      imageSize=<16 8>
      margin=<ltrb 0.5 1 0.5 0>
      outlineSettings {
       bUseBrushTransparency=false
       roundingType=HalfHeightRadius
       width=0
       color {colorUseRule=UseColor_Specified, specifiedColor=<rgba 0 0 0 0>}
       cornerRadii=<0 0 0 0>
      }
      tintColor {colorUseRule=UseColor_Specified, specifiedColor=<rgba 1 1 1 1>}
      uVRegion {bIsValid=false, max=<0 0>, min=<0 0>}
     }
     verticalScrolledContentPadding=<ltrb 0 0 0 1>
    }
   }
   slot {class=CanvasPanelSlot, properties {layoutData {alignment=<0 0>, anchors {maximum=<1 1>, minimum=<0 0>}, offsets=<ltrb 0 246 0 0>}}}
  }
  Functionalities {widgetClassPath=&HorizontalBox, parent=ScrollBox_39, widget=Functionalities, slot {class=ScrollBoxSlot}}
  VerticalBox_76 {
   widgetClassPath=&VerticalBox
   parent=Functionalities
   widget=VerticalBox_76
   slot {class=HorizontalBoxSlot, properties {size {sizeRule=Fill, value=0.25}}}
  }
  F_Bounds {
   widgetClassPath=&HorizontalBox
   parent=VerticalBox_76
   widget=F_Bounds
   slot {class=VerticalBoxSlot, properties {padding=<ltrb 0 0 0 8>}}
  }
  CommonBorder {
   widgetClassPath=&CommonBorder
   parent=F_Bounds
   widget=CommonBorder
   properties {
    background {
     bIsDynamicallyLoaded=false
     drawAs=Box
     imageType=NoImage
     mirroring=NoMirror
     resourceName=None
     resourceObject=&T_WhiteMask
     tiling=NoTile
     imageSize=<128 32>
     margin=<ltrb 0.1 0.45 0.9 0.5>
     outlineSettings {
      bUseBrushTransparency=false
      roundingType=HalfHeightRadius
      width=0
      color {colorUseRule=UseColor_Specified, specifiedColor=<rgba 0 0 0 0>}
      cornerRadii=<0 0 0 0>
     }
     tintColor {colorUseRule=UseColor_Specified, specifiedColor=<rgba 1 1 1 1>}
     uVRegion {bIsValid=false, max=<0 0>, min=<0 0>}
    }
    brushColor=<rgba 0.0103 0.0103 0.0103 1>
    padding=<ltrb 8 8 8 8>
   }
   slot {class=HorizontalBoxSlot, properties {padding=<ltrb 4 0 4 0>, size {sizeRule=Fill, value=1}}}
  }
  VerticalBox {
   widgetClassPath=&VerticalBox
   parent=CommonBorder
   widget=VerticalBox
   slot {class=BorderSlot, properties {padding=<ltrb 8 8 8 8>}}
  }
  Snap_1 {
   widgetClassPath=&HorizontalBox
   parent=VerticalBox
   widget=Snap_1
   slot {class=VerticalBoxSlot, properties {padding=<ltrb 0 0 0 8>}}
  }
  CommonTextBlock_7 {
   widgetClassPath=&CommonTextBlock
   parent=Snap_1
   widget=CommonTextBlock_7
   properties {
    style=/Game/ArtTools/Font/TextStyle_RobotoEngineLike.TextStyle_RobotoEngineLike_C
    text=Bounds
    colorAndOpacity {colorUseRule=UseColor_Specified, specifiedColor=<rgba 0.5 0.5 0.5 1>}
    font {
     bForceMonospaced=false
     bMaterialIsStencil=false
     fontMaterial=null
     fontObject=&Roboto
     letterSpacing=0
     monospacedWidth=1
     size=8
     skewAmount=0
     typefaceFontName=Regular
     outlineSettings {
      bApplyOutlineToDropShadows=false
      bMiteredCorners=false
      bSeparateFillAlpha=false
      outlineMaterial=null
      outlineSize=0
      outlineColor=<rgba 0 0 0 1>
     }
    }
    shadowOffset=<0 0>
   }
   slot {class=HorizontalBoxSlot, properties {verticalAlignment=VAlign_Center, padding=<ltrb 4 0 8 0>}}
  }
  Btn_Visibility {
   widgetClassPath=&Button
   parent=Snap_1
   widget=Btn_Visibility
   properties {widgetStyle {
     clickedSlateSound {resourceObject=null}
     disabled {
      bIsDynamicallyLoaded=false
      drawAs=Image
      imageType=NoImage
      mirroring=NoMirror
      resourceName=None
      resourceObject=&Icon_Visible
      tiling=NoTile
      imageSize=<16 16>
      margin=<ltrb 0 0 0 0>
      outlineSettings {
       bUseBrushTransparency=false
       roundingType=FixedRadius
       width=1
       color {colorUseRule=UseColor_ColorTable, specifiedColor=<rgba 0 0 0 0>}
       cornerRadii=<4 4 4 4>
      }
      tintColor {colorUseRule=UseColor_Specified, specifiedColor=<rgba 0.0395 0.0395 0.0395 1>}
      uVRegion {bIsValid=false, max=<0 0>, min=<0 0>}
     }
     disabledForeground {colorUseRule=UseColor_ColorTable, specifiedColor=<rgba 0 0 0 0>}
     hovered {
      bIsDynamicallyLoaded=false
      drawAs=Image
      imageType=NoImage
      mirroring=NoMirror
      resourceName=None
      resourceObject=&Icon_Visible
      tiling=NoTile
      imageSize=<16 16>
      margin=<ltrb 0 0 0 0>
      outlineSettings {
       bUseBrushTransparency=false
       roundingType=FixedRadius
       width=0
       color {colorUseRule=UseColor_ColorTable, specifiedColor=<rgba 0 0 0 0>}
       cornerRadii=<4 4 4 4>
      }
      tintColor {colorUseRule=UseColor_Specified, specifiedColor=<rgba 1 1 1 1>}
      uVRegion {bIsValid=false, max=<0 0>, min=<0 0>}
     }
     hoveredForeground {colorUseRule=UseColor_ColorTable, specifiedColor=<rgba 0 0 0 0>}
     hoveredSlateSound {resourceObject=null}
     normal {
      bIsDynamicallyLoaded=false
      drawAs=Image
      imageType=NoImage
      mirroring=NoMirror
      resourceName=None
      resourceObject=&Icon_Visible
      tiling=NoTile
      imageSize=<16 16>
      margin=<ltrb 0 0 0 0>
      outlineSettings {
       bUseBrushTransparency=false
       roundingType=FixedRadius
       width=0
       color {colorUseRule=UseColor_ColorTable, specifiedColor=<rgba 0 0 0 0>}
       cornerRadii=<4 4 4 4>
      }
      tintColor {colorUseRule=UseColor_Specified, specifiedColor=<rgba 1 1 1 1>}
      uVRegion {bIsValid=false, max=<0 0>, min=<0 0>}
     }
     normalForeground {colorUseRule=UseColor_ColorTable, specifiedColor=<rgba 0 0 0 0>}
     normalPadding=<ltrb 12 1.5 12 1.5>
     pressed {
      bIsDynamicallyLoaded=false
      drawAs=Image
      imageType=NoImage
      mirroring=NoMirror
      resourceName=None
      resourceObject=&Icon_Visible
      tiling=NoTile
      imageSize=<16 16>
      margin=<ltrb 0 0 0 0>
      outlineSettings {
       bUseBrushTransparency=false
       roundingType=FixedRadius
       width=0
       color {colorUseRule=UseColor_ColorTable, specifiedColor=<rgba 0 0 0 0>}
       cornerRadii=<4 4 4 4>
      }
      tintColor {colorUseRule=UseColor_Specified, specifiedColor=<rgba 1 1 1 1>}
      uVRegion {bIsValid=false, max=<0 0>, min=<0 0>}
     }
     pressedForeground {colorUseRule=UseColor_ColorTable, specifiedColor=<rgba 0 0 0 0>}
     pressedPadding=<ltrb 12 2.5 12 0.5>
     pressedSlateSound {resourceObject=null}
    }}
   slot {class=HorizontalBoxSlot}
  }
  SnapSettings_1 {widgetClassPath=&HorizontalBox, parent=VerticalBox, widget=SnapSettings_1, slot {class=VerticalBoxSlot}}
  BoundsRadiusInfo {
   widgetClassPath=&SinglePropertyView
   parent=SnapSettings_1
   widget=BoundsRadiusInfo
   properties {nameOverride=Radius, propertyName=BoundsRadius}
   slot {class=HorizontalBoxSlot, properties {size {sizeRule=Fill, value=1}}}
  }
  BoundsHeightInfo {
   widgetClassPath=&SinglePropertyView
   parent=SnapSettings_1
   widget=BoundsHeightInfo
   properties {nameOverride=Height, propertyName=BoundsHeight}
   slot {class=HorizontalBoxSlot, properties {size {sizeRule=Fill, value=1}}}
  }
  F_Snapping {
   widgetClassPath=&HorizontalBox
   parent=VerticalBox_76
   widget=F_Snapping
   slot {class=VerticalBoxSlot, properties {padding=<ltrb 0 0 0 8>}}
  }
  CommonBorder_56 {
   widgetClassPath=&CommonBorder
   parent=F_Snapping
   widget=CommonBorder_56
   properties {
    background {
     bIsDynamicallyLoaded=false
     drawAs=Box
     imageType=NoImage
     mirroring=NoMirror
     resourceName=None
     resourceObject=&T_WhiteMask
     tiling=NoTile
     imageSize=<128 32>
     margin=<ltrb 0.1 0.45 0.9 0.5>
     outlineSettings {
      bUseBrushTransparency=false
      roundingType=HalfHeightRadius
      width=0
      color {colorUseRule=UseColor_Specified, specifiedColor=<rgba 0 0 0 0>}
      cornerRadii=<0 0 0 0>
     }
     tintColor {colorUseRule=UseColor_Specified, specifiedColor=<rgba 1 1 1 1>}
     uVRegion {bIsValid=false, max=<0 0>, min=<0 0>}
    }
    brushColor=<rgba 0.0103 0.0103 0.0103 1>
    padding=<ltrb 8 8 8 8>
   }
   slot {class=HorizontalBoxSlot, properties {padding=<ltrb 4 0 4 0>, size {sizeRule=Fill, value=1}}}
  }
  VerticalBox_64 {
   widgetClassPath=&VerticalBox
   parent=CommonBorder_56
   widget=VerticalBox_64
   slot {class=BorderSlot, properties {padding=<ltrb 8 8 8 8>}}
  }
  Snap {
   widgetClassPath=&HorizontalBox
   parent=VerticalBox_64
   widget=Snap
   slot {class=VerticalBoxSlot, properties {padding=<ltrb 0 0 0 8>}}
  }
  HorizontalBox_7 {
   widgetClassPath=&HorizontalBox
   parent=Snap
   widget=HorizontalBox_7
   slot {class=HorizontalBoxSlot, properties {size {sizeRule=Fill, value=1}}}
  }
  Btn_Snap {
   widgetClassPath=&Button
   parent=HorizontalBox_7
   widget=Btn_Snap
   properties {widgetStyle {
     clickedSlateSound {resourceObject=null}
     disabled {
      bIsDynamicallyLoaded=false
      drawAs=RoundedBox
      imageType=NoImage
      mirroring=NoMirror
      resourceName=None
      resourceObject=null
      tiling=NoTile
      imageSize=<0 0>
      margin=<ltrb 0 0 0 0>
      outlineSettings {
       bUseBrushTransparency=false
       roundingType=FixedRadius
       width=1
       color {colorUseRule=UseColor_ColorTable, specifiedColor=<rgba 0 0 0 0>}
       cornerRadii=<4 0 0 4>
      }
      tintColor {colorUseRule=UseColor_Specified, specifiedColor=<rgba 0.0395 0.0395 0.0395 1>}
      uVRegion {bIsValid=false, max=<0 0>, min=<0 0>}
     }
     disabledForeground {colorUseRule=UseColor_ColorTable, specifiedColor=<rgba 0 0 0 0>}
     hovered {
      bIsDynamicallyLoaded=false
      drawAs=RoundedBox
      imageType=NoImage
      mirroring=NoMirror
      resourceName=None
      resourceObject=null
      tiling=NoTile
      imageSize=<32 32>
      margin=<ltrb 0 0 0 0>
      outlineSettings {
       bUseBrushTransparency=false
       roundingType=FixedRadius
       width=1
       color {colorUseRule=UseColor_ColorTable, specifiedColor=<rgba 0 0 0 0>}
       cornerRadii=<4 0 0 4>
      }
      tintColor {colorUseRule=UseColor_Specified, specifiedColor=<rgba 0.0953 0.0953 0.0953 1>}
      uVRegion {bIsValid=false, max=<0 0>, min=<0 0>}
     }
     hoveredForeground {colorUseRule=UseColor_ColorTable, specifiedColor=<rgba 0 0 0 0>}
     hoveredSlateSound {resourceObject=null}
     normal {
      bIsDynamicallyLoaded=false
      drawAs=RoundedBox
      imageType=NoImage
      mirroring=NoMirror
      resourceName=None
      resourceObject=null
      tiling=NoTile
      imageSize=<32 32>
      margin=<ltrb 0 0 0 0>
      outlineSettings {
       bUseBrushTransparency=false
       roundingType=FixedRadius
       width=1
       color {colorUseRule=UseColor_ColorTable, specifiedColor=<rgba 0 0 0 0>}
       cornerRadii=<4 0 0 4>
      }
      tintColor {colorUseRule=UseColor_Specified, specifiedColor=<rgba 0.0395 0.0395 0.0395 1>}
      uVRegion {bIsValid=false, max=<0 0>, min=<0 0>}
     }
     normalForeground {colorUseRule=UseColor_ColorTable, specifiedColor=<rgba 0 0 0 0>}
     normalPadding=<ltrb 12 1.5 12 1.5>
     pressed {
      bIsDynamicallyLoaded=false
      drawAs=RoundedBox
      imageType=NoImage
      mirroring=NoMirror
      resourceName=None
      resourceObject=null
      tiling=NoTile
      imageSize=<32 32>
      margin=<ltrb 0 0 0 0>
      outlineSettings {
       bUseBrushTransparency=false
       roundingType=FixedRadius
       width=1
       color {colorUseRule=UseColor_ColorTable, specifiedColor=<rgba 0 0 0 0>}
       cornerRadii=<4 0 0 4>
      }
      tintColor {colorUseRule=UseColor_Specified, specifiedColor=<rgba 0.0284 0.0284 0.0284 1>}
      uVRegion {bIsValid=false, max=<0 0>, min=<0 0>}
     }
     pressedForeground {colorUseRule=UseColor_ColorTable, specifiedColor=<rgba 0 0 0 0>}
     pressedPadding=<ltrb 12 2.5 12 0.5>
     pressedSlateSound {resourceObject=null}
    }}
   slot {class=HorizontalBoxSlot, properties {padding=<ltrb 10 0 10 0>, size {sizeRule=Fill, value=1}}}
  }
  CommonTextBlock_3 {
   widgetClassPath=&CommonTextBlock
   parent=Btn_Snap
   widget=CommonTextBlock_3
   properties {
    clipping=ClipToBounds
    style=&TextStyle_RobotoEngineLike_C
    text=Snap to Ground
    colorAndOpacity {colorUseRule=UseColor_Specified, specifiedColor=<rgba 0.5 0.5 0.5 1>}
    font {
     bForceMonospaced=false
     bMaterialIsStencil=false
     fontMaterial=null
     fontObject=&Roboto
     letterSpacing=0
     monospacedWidth=1
     size=8
     skewAmount=0
     typefaceFontName=Regular
     outlineSettings {
      bApplyOutlineToDropShadows=false
      bMiteredCorners=false
      bSeparateFillAlpha=false
      outlineMaterial=null
      outlineSize=0
      outlineColor=<rgba 0 0 0 1>
     }
    }
    shadowOffset=<0 0>
   }
   slot {class=ButtonSlot, properties {padding=<ltrb 0 2 0 2>}}
  }
  HorizontalBox_9 {
   widgetClassPath=&HorizontalBox
   parent=Snap
   widget=HorizontalBox_9
   slot {class=HorizontalBoxSlot, properties {size {sizeRule=Fill, value=1}}}
  }
  Btn_SnapContinuous {
   widgetClassPath=&Button
   parent=HorizontalBox_9
   widget=Btn_SnapContinuous
   properties {widgetStyle {
     clickedSlateSound {resourceObject=null}
     disabled {
      bIsDynamicallyLoaded=false
      drawAs=RoundedBox
      imageType=NoImage
      mirroring=NoMirror
      resourceName=None
      resourceObject=null
      tiling=NoTile
      imageSize=<0 0>
      margin=<ltrb 0 0 0 0>
      outlineSettings {
       bUseBrushTransparency=false
       roundingType=FixedRadius
       width=1
       color {colorUseRule=UseColor_ColorTable, specifiedColor=<rgba 0 0 0 0>}
       cornerRadii=<0 4 4 0>
      }
      tintColor {colorUseRule=UseColor_Specified, specifiedColor=<rgba 0.0395 0.0395 0.0395 1>}
      uVRegion {bIsValid=false, max=<0 0>, min=<0 0>}
     }
     disabledForeground {colorUseRule=UseColor_ColorTable, specifiedColor=<rgba 0 0 0 0>}
     hovered {
      bIsDynamicallyLoaded=false
      drawAs=RoundedBox
      imageType=NoImage
      mirroring=NoMirror
      resourceName=None
      resourceObject=null
      tiling=NoTile
      imageSize=<32 32>
      margin=<ltrb 0 0 0 0>
      outlineSettings {
       bUseBrushTransparency=false
       roundingType=FixedRadius
       width=1
       color {colorUseRule=UseColor_ColorTable, specifiedColor=<rgba 0 0 0 0>}
       cornerRadii=<0 4 4 0>
      }
      tintColor {colorUseRule=UseColor_Specified, specifiedColor=<rgba 0.0953 0.0953 0.0953 1>}
      uVRegion {bIsValid=false, max=<0 0>, min=<0 0>}
     }
     hoveredForeground {colorUseRule=UseColor_ColorTable, specifiedColor=<rgba 0 0 0 0>}
     hoveredSlateSound {resourceObject=null}
     normal {
      bIsDynamicallyLoaded=false
      drawAs=RoundedBox
      imageType=NoImage
      mirroring=NoMirror
      resourceName=None
      resourceObject=null
      tiling=NoTile
      imageSize=<32 32>
      margin=<ltrb 0 0 0 0>
      outlineSettings {
       bUseBrushTransparency=false
       roundingType=FixedRadius
       width=1
       color {colorUseRule=UseColor_ColorTable, specifiedColor=<rgba 0 0 0 0>}
       cornerRadii=<0 4 4 0>
      }
      tintColor {colorUseRule=UseColor_Specified, specifiedColor=<rgba 0.0395 0.0395 0.0395 1>}
      uVRegion {bIsValid=false, max=<0 0>, min=<0 0>}
     }
     normalForeground {colorUseRule=UseColor_ColorTable, specifiedColor=<rgba 0 0 0 0>}
     normalPadding=<ltrb 0 0 0 0>
     pressed {
      bIsDynamicallyLoaded=false
      drawAs=RoundedBox
      imageType=NoImage
      mirroring=NoMirror
      resourceName=None
      resourceObject=null
      tiling=NoTile
      imageSize=<32 32>
      margin=<ltrb 0 0 0 0>
      outlineSettings {
       bUseBrushTransparency=false
       roundingType=FixedRadius
       width=1
       color {colorUseRule=UseColor_ColorTable, specifiedColor=<rgba 0 0 0 0>}
       cornerRadii=<0 4 4 0>
      }
      tintColor {colorUseRule=UseColor_Specified, specifiedColor=<rgba 0.0284 0.0284 0.0284 1>}
      uVRegion {bIsValid=false, max=<0 0>, min=<0 0>}
     }
     pressedForeground {colorUseRule=UseColor_ColorTable, specifiedColor=<rgba 0 0 0 0>}
     pressedPadding=<ltrb 0 0 0 0>
     pressedSlateSound {resourceObject=null}
    }}
   slot {class=HorizontalBoxSlot, properties {padding=<ltrb -10 0 0 0>, size {sizeRule=Automatic, value=0.3}}}
  }
  processing_icon {
   widgetClassPath=&Image
   bIsVariable=true
   parent=Btn_SnapContinuous
   widget=processing_icon
   properties {brush {
     bIsDynamicallyLoaded=false
     drawAs=Image
     imageType=NoImage
     mirroring=NoMirror
     resourceName=None
     resourceObject=/Game/ArtTools/Tech_Art/InstancedMeshesCluster/Icon/processing_icon
     tiling=NoTile
     imageSize=<16 16>
     margin=<ltrb 0 0 0 0>
     outlineSettings {
      bUseBrushTransparency=false
      roundingType=HalfHeightRadius
      width=0
      color {colorUseRule=UseColor_Specified, specifiedColor=<rgba 0 0 0 0>}
      cornerRadii=<0 0 0 0>
     }
     tintColor {colorUseRule=UseColor_Specified, specifiedColor=<rgba 0.2588 0.5412 0.0667 1>}
     uVRegion {bIsValid=false, max=<0 0>, min=<0 0>}
    }}
   slot {class=ButtonSlot, properties {padding=<ltrb 4 0 4 0>}}
  }
  HorizontalBox_8 {
   widgetClassPath=&HorizontalBox
   parent=Snap
   widget=HorizontalBox_8
   slot {class=HorizontalBoxSlot, properties {size {sizeRule=Fill, value=1}}}
  }
  SnapSettings {widgetClassPath=&HorizontalBox, parent=VerticalBox_64, widget=SnapSettings, slot {class=VerticalBoxSlot}}
  HorizontalBox_17 {
   widgetClassPath=&HorizontalBox
   parent=SnapSettings
   widget=HorizontalBox_17
   slot {class=HorizontalBoxSlot, properties {size {sizeRule=Fill, value=0.5}}}
  }
  HorizontalBox_18 {
   widgetClassPath=&HorizontalBox
   parent=HorizontalBox_17
   widget=HorizontalBox_18
   slot {class=HorizontalBoxSlot, properties {horizontalAlignment=HAlign_Center, size {sizeRule=Fill, value=0.25}}}
  }
  CommonTextBlock_5 {
   widgetClassPath=&CommonTextBlock
   parent=HorizontalBox_18
   widget=CommonTextBlock_5
   properties {
    clipping=ClipToBounds
    style=&TextStyle_RobotoEngineLike_C
    text=Align to Surface
    colorAndOpacity {colorUseRule=UseColor_Specified, specifiedColor=<rgba 0.5 0.5 0.5 1>}
    font {
     bForceMonospaced=false
     bMaterialIsStencil=false
     fontMaterial=null
     fontObject=&Roboto
     letterSpacing=0
     monospacedWidth=1
     size=8
     skewAmount=0
     typefaceFontName=Regular
     outlineSettings {
      bApplyOutlineToDropShadows=false
      bMiteredCorners=false
      bSeparateFillAlpha=false
      outlineMaterial=null
      outlineSize=0
      outlineColor=<rgba 0 0 0 1>
     }
    }
    shadowOffset=<0 0>
   }
   slot {class=HorizontalBoxSlot, properties {verticalAlignment=VAlign_Center, padding=<ltrb 4 0 8 0>}}
  }
  Check_AlignToSurface {
   widgetClassPath=/Script/UMG.CheckBox
   bIsVariable=true
   parent=HorizontalBox_18
   widget=Check_AlignToSurface
   properties {
    checkedState=Checked
    widgetStyle {
     checkBoxType=CheckBox
     backgroundHoveredImage {
      bIsDynamicallyLoaded=false
      drawAs=RoundedBox
      imageType=NoImage
      mirroring=NoMirror
      resourceName=None
      resourceObject=null
      tiling=NoTile
      imageSize=<18 18>
      margin=<ltrb 0 0 0 0>
      outlineSettings {
       bUseBrushTransparency=false
       roundingType=FixedRadius
       width=1
       color {colorUseRule=UseColor_ColorTable, specifiedColor=<rgba 0 0 0 0>}
       cornerRadii=<3 3 3 3>
      }
      tintColor {colorUseRule=UseColor_Specified, specifiedColor=<rgba 0.0048 0.0048 0.0048 1>}
      uVRegion {bIsValid=false, max=<0 0>, min=<0 0>}
     }
     backgroundImage {
      bIsDynamicallyLoaded=false
      drawAs=RoundedBox
      imageType=NoImage
      mirroring=NoMirror
      resourceName=None
      resourceObject=null
      tiling=NoTile
      imageSize=<18 18>
      margin=<ltrb 0 0 0 0>
      outlineSettings {
       bUseBrushTransparency=false
       roundingType=FixedRadius
       width=1
       color {colorUseRule=UseColor_ColorTable, specifiedColor=<rgba 0 0 0 0>}
       cornerRadii=<3 3 3 3>
      }
      tintColor {colorUseRule=UseColor_Specified, specifiedColor=<rgba 0.0048 0.0048 0.0048 1>}
      uVRegion {bIsValid=false, max=<0 0>, min=<0 0>}
     }
     backgroundPressedImage {
      bIsDynamicallyLoaded=false
      drawAs=RoundedBox
      imageType=NoImage
      mirroring=NoMirror
      resourceName=None
      resourceObject=null
      tiling=NoTile
      imageSize=<18 18>
      margin=<ltrb 0 0 0 0>
      outlineSettings {
       bUseBrushTransparency=false
       roundingType=FixedRadius
       width=1
       color {colorUseRule=UseColor_ColorTable, specifiedColor=<rgba 0 0 0 0>}
       cornerRadii=<3 3 3 3>
      }
      tintColor {colorUseRule=UseColor_Specified, specifiedColor=<rgba 0.0048 0.0048 0.0048 1>}
      uVRegion {bIsValid=false, max=<0 0>, min=<0 0>}
     }
     borderBackgroundColor {colorUseRule=UseColor_Specified, specifiedColor=<rgba 1 1 1 1>}
     checkedForeground {colorUseRule=UseColor_Specified, specifiedColor=<rgba 1 1 1 1>}
     checkedHoveredForeground {colorUseRule=UseColor_Specified, specifiedColor=<rgba 1 1 1 1>}
     checkedHoveredImage {
      bIsDynamicallyLoaded=false
      drawAs=Image
      imageType=Vector
      mirroring=NoMirror
      resourceName=../../../Engine/Content/Slate/Starship/CoreWidgets/CheckBox/check.svg
      resourceObject=null
      tiling=NoTile
      imageSize=<16 16>
      margin=<ltrb 0 0 0 0>
      outlineSettings {
       bUseBrushTransparency=false
       roundingType=HalfHeightRadius
       width=0
       color {colorUseRule=UseColor_Specified, specifiedColor=<rgba 0 0 0 0>}
       cornerRadii=<0 0 0 0>
      }
      tintColor {colorUseRule=UseColor_Specified, specifiedColor=<rgba 0.0044 0.2384 1 1>}
      uVRegion {bIsValid=false, max=<0 0>, min=<0 0>}
     }
     checkedImage {
      bIsDynamicallyLoaded=false
      drawAs=Image
      imageType=Vector
      mirroring=NoMirror
      resourceName=../../../Engine/Content/Slate/Starship/CoreWidgets/CheckBox/check.svg
      resourceObject=null
      tiling=NoTile
      imageSize=<16 16>
      margin=<ltrb 0 0 0 0>
      outlineSettings {
       bUseBrushTransparency=false
       roundingType=HalfHeightRadius
       width=0
       color {colorUseRule=UseColor_Specified, specifiedColor=<rgba 0 0 0 0>}
       cornerRadii=<0 0 0 0>
      }
      tintColor {colorUseRule=UseColor_Specified, specifiedColor=<rgba 0 0.162 0.7454 1>}
      uVRegion {bIsValid=false, max=<0 0>, min=<0 0>}
     }
     checkedPressedForeground {colorUseRule=UseColor_Specified, specifiedColor=<rgba 1 1 1 1>}
     checkedPressedImage {
      bIsDynamicallyLoaded=false
      drawAs=Image
      imageType=Vector
      mirroring=NoMirror
      resourceName=../../../Engine/Content/Slate/Starship/CoreWidgets/CheckBox/check.svg
      resourceObject=null
      tiling=NoTile
      imageSize=<16 16>
      margin=<ltrb 0 0 0 0>
      outlineSettings {
       bUseBrushTransparency=false
       roundingType=HalfHeightRadius
       width=0
       color {colorUseRule=UseColor_Specified, specifiedColor=<rgba 0 0 0 0>}
       cornerRadii=<0 0 0 0>
      }
      tintColor {colorUseRule=UseColor_Specified, specifiedColor=<rgba 0 0.0802 0.3515 1>}
      uVRegion {bIsValid=false, max=<0 0>, min=<0 0>}
     }
     checkedSlateSound {resourceObject=null}
     foregroundColor {colorUseRule=UseColor_Specified, specifiedColor=<rgba 1 1 1 1>}
     hoveredForeground {colorUseRule=UseColor_Specified, specifiedColor=<rgba 1 1 1 1>}
     hoveredSlateSound {resourceObject=null}
     padding=<ltrb 2 0 0 0>
     pressedForeground {colorUseRule=UseColor_Specified, specifiedColor=<rgba 1 1 1 1>}
     uncheckedHoveredImage {
      bIsDynamicallyLoaded=false
      drawAs=NoDrawType
      imageType=NoImage
      mirroring=NoMirror
      resourceName=None
      resourceObject=null
      tiling=NoTile
      imageSize=<0 0>
      margin=<ltrb 0 0 0 0>
      outlineSettings {
       bUseBrushTransparency=false
       roundingType=HalfHeightRadius
       width=0
       color {colorUseRule=UseColor_Specified, specifiedColor=<rgba 0 0 0 0>}
       cornerRadii=<0 0 0 0>
      }
      tintColor {colorUseRule=UseColor_Specified, specifiedColor=<rgba 1 1 1 1>}
      uVRegion {bIsValid=false, max=<0 0>, min=<0 0>}
     }
     uncheckedImage {
      bIsDynamicallyLoaded=false
      drawAs=NoDrawType
      imageType=NoImage
      mirroring=NoMirror
      resourceName=None
      resourceObject=null
      tiling=NoTile
      imageSize=<0 0>
      margin=<ltrb 0 0 0 0>
      outlineSettings {
       bUseBrushTransparency=false
       roundingType=HalfHeightRadius
       width=0
       color {colorUseRule=UseColor_Specified, specifiedColor=<rgba 0 0 0 0>}
       cornerRadii=<0 0 0 0>
      }
      tintColor {colorUseRule=UseColor_Specified, specifiedColor=<rgba 1 1 1 1>}
      uVRegion {bIsValid=false, max=<0 0>, min=<0 0>}
     }
     uncheckedPressedImage {
      bIsDynamicallyLoaded=false
      drawAs=NoDrawType
      imageType=NoImage
      mirroring=NoMirror
      resourceName=None
      resourceObject=null
      tiling=NoTile
      imageSize=<0 0>
      margin=<ltrb 0 0 0 0>
      outlineSettings {
       bUseBrushTransparency=false
       roundingType=HalfHeightRadius
       width=0
       color {colorUseRule=UseColor_Specified, specifiedColor=<rgba 0 0 0 0>}
       cornerRadii=<0 0 0 0>
      }
      tintColor {colorUseRule=UseColor_Specified, specifiedColor=<rgba 1 1 1 1>}
      uVRegion {bIsValid=false, max=<0 0>, min=<0 0>}
     }
     uncheckedSlateSound {resourceObject=null}
     undeterminedForeground {colorUseRule=UseColor_Specified, specifiedColor=<rgba 1 1 1 1>}
     undeterminedHoveredImage {
      bIsDynamicallyLoaded=false
      drawAs=Image
      imageType=Vector
      mirroring=NoMirror
      resourceName=../../../Engine/Content/Slate/Starship/CoreWidgets/CheckBox/indeterminate.svg
      resourceObject=null
      tiling=NoTile
      imageSize=<16 16>
      margin=<ltrb 0 0 0 0>
      outlineSettings {
       bUseBrushTransparency=false
       roundingType=HalfHeightRadius
       width=0
       color {colorUseRule=UseColor_Specified, specifiedColor=<rgba 0 0 0 0>}
       cornerRadii=<0 0 0 0>
      }
      tintColor {colorUseRule=UseColor_Specified, specifiedColor=<rgba 0.0044 0.2384 1 1>}
      uVRegion {bIsValid=false, max=<0 0>, min=<0 0>}
     }
     undeterminedImage {
      bIsDynamicallyLoaded=false
      drawAs=Image
      imageType=Vector
      mirroring=NoMirror
      resourceName=../../../Engine/Content/Slate/Starship/CoreWidgets/CheckBox/indeterminate.svg
      resourceObject=null
      tiling=NoTile
      imageSize=<16 16>
      margin=<ltrb 0 0 0 0>
      outlineSettings {
       bUseBrushTransparency=false
       roundingType=HalfHeightRadius
       width=0
       color {colorUseRule=UseColor_Specified, specifiedColor=<rgba 0 0 0 0>}
       cornerRadii=<0 0 0 0>
      }
      tintColor {colorUseRule=UseColor_Specified, specifiedColor=<rgba 0 0.162 0.7454 1>}
      uVRegion {bIsValid=false, max=<0 0>, min=<0 0>}
     }
     undeterminedPressedImage {
      bIsDynamicallyLoaded=false
      drawAs=Image
      imageType=Vector
      mirroring=NoMirror
      resourceName=../../../Engine/Content/Slate/Starship/CoreWidgets/CheckBox/indeterminate.svg
      resourceObject=null
      tiling=NoTile
      imageSize=<16 16>
      margin=<ltrb 0 0 0 0>
      outlineSettings {
       bUseBrushTransparency=false
       roundingType=HalfHeightRadius
       width=0
       color {colorUseRule=UseColor_Specified, specifiedColor=<rgba 0 0 0 0>}
       cornerRadii=<0 0 0 0>
      }
      tintColor {colorUseRule=UseColor_Specified, specifiedColor=<rgba 0 0.0802 0.3515 1>}
      uVRegion {bIsValid=false, max=<0 0>, min=<0 0>}
     }
    }
   }
   slot {class=HorizontalBoxSlot, properties {padding=<ltrb 0 0 2 0>}}
  }
  HorizontalBox_19 {
   widgetClassPath=&HorizontalBox
   parent=SnapSettings
   widget=HorizontalBox_19
   slot {class=HorizontalBoxSlot, properties {size {sizeRule=Fill, value=1}}}
  }
  OrientUp {
   widgetClassPath=&SinglePropertyView
   parent=HorizontalBox_19
   widget=OrientUp
   properties {nameOverride=Up Bias, propertyName=UpBias}
   slot {class=HorizontalBoxSlot, properties {size {sizeRule=Fill, value=1}}}
  }
  HorizontalBox_20 {
   widgetClassPath=&HorizontalBox
   parent=HorizontalBox_19
   widget=HorizontalBox_20
   slot {class=HorizontalBoxSlot, properties {size {sizeRule=Fill, value=0.5}}}
  }
  F_Rotation {
   widgetClassPath=&HorizontalBox
   parent=VerticalBox_76
   widget=F_Rotation
   slot {class=VerticalBoxSlot, properties {padding=<ltrb 0 0 0 8>}}
  }
  CommonBorder_491 {
   widgetClassPath=&CommonBorder
   parent=F_Rotation
   widget=CommonBorder_491
   properties {
    background {
     bIsDynamicallyLoaded=false
     drawAs=Image
     imageType=NoImage
     mirroring=NoMirror
     resourceName=None
     resourceObject=&T_WhiteMask
     tiling=NoTile
     imageSize=<32 32>
     margin=<ltrb 0 0 0 0>
     outlineSettings {
      bUseBrushTransparency=false
      roundingType=HalfHeightRadius
      width=0
      color {colorUseRule=UseColor_Specified, specifiedColor=<rgba 0 0 0 0>}
      cornerRadii=<0 0 0 0>
     }
     tintColor {colorUseRule=UseColor_Specified, specifiedColor=<rgba 1 1 1 1>}
     uVRegion {bIsValid=false, max=<0 0>, min=<0 0>}
    }
    brushColor=<rgba 0.0103 0.0103 0.0103 1>
    padding=<ltrb 8 8 8 8>
   }
   slot {class=HorizontalBoxSlot, properties {padding=<ltrb 4 0 4 0>, size {sizeRule=Fill, value=1}}}
  }
  VerticalBox_360 {
   widgetClassPath=&VerticalBox
   parent=CommonBorder_491
   widget=VerticalBox_360
   slot {class=BorderSlot, properties {padding=<ltrb 8 8 8 8>}}
  }
  Rotation {
   widgetClassPath=&HorizontalBox
   parent=VerticalBox_360
   widget=Rotation
   slot {class=VerticalBoxSlot, properties {padding=<ltrb 0 0 0 9>, size {sizeRule=Fill, value=1}}}
  }
  HorizontalBox_10 {
   widgetClassPath=&HorizontalBox
   parent=Rotation
   widget=HorizontalBox_10
   slot {class=HorizontalBoxSlot, properties {size {sizeRule=Fill, value=1}}}
  }
  Btn_RandRot {
   widgetClassPath=&Button
   parent=HorizontalBox_10
   widget=Btn_RandRot
   properties {widgetStyle {
     clickedSlateSound {resourceObject=null}
     disabled {
      bIsDynamicallyLoaded=false
      drawAs=RoundedBox
      imageType=NoImage
      mirroring=NoMirror
      resourceName=None
      resourceObject=null
      tiling=NoTile
      imageSize=<0 0>
      margin=<ltrb 0 0 0 0>
      outlineSettings {
       bUseBrushTransparency=false
       roundingType=FixedRadius
       width=1
       color {colorUseRule=UseColor_ColorTable, specifiedColor=<rgba 0 0 0 0>}
       cornerRadii=<4 4 4 4>
      }
      tintColor {colorUseRule=UseColor_Specified, specifiedColor=<rgba 0.0395 0.0395 0.0395 1>}
      uVRegion {bIsValid=false, max=<0 0>, min=<0 0>}
     }
     disabledForeground {colorUseRule=UseColor_ColorTable, specifiedColor=<rgba 0 0 0 0>}
     hovered {
      bIsDynamicallyLoaded=false
      drawAs=RoundedBox
      imageType=NoImage
      mirroring=NoMirror
      resourceName=None
      resourceObject=null
      tiling=NoTile
      imageSize=<32 32>
      margin=<ltrb 0 0 0 0>
      outlineSettings {
       bUseBrushTransparency=false
       roundingType=FixedRadius
       width=1
       color {colorUseRule=UseColor_ColorTable, specifiedColor=<rgba 0 0 0 0>}
       cornerRadii=<4 4 4 4>
      }
      tintColor {colorUseRule=UseColor_Specified, specifiedColor=<rgba 0.0953 0.0953 0.0953 1>}
      uVRegion {bIsValid=false, max=<0 0>, min=<0 0>}
     }
     hoveredForeground {colorUseRule=UseColor_ColorTable, specifiedColor=<rgba 0 0 0 0>}
     hoveredSlateSound {resourceObject=null}
     normal {
      bIsDynamicallyLoaded=false
      drawAs=RoundedBox
      imageType=NoImage
      mirroring=NoMirror
      resourceName=None
      resourceObject=null
      tiling=NoTile
      imageSize=<32 32>
      margin=<ltrb 0 0 0 0>
      outlineSettings {
       bUseBrushTransparency=false
       roundingType=FixedRadius
       width=1
       color {colorUseRule=UseColor_ColorTable, specifiedColor=<rgba 0 0 0 0>}
       cornerRadii=<4 4 4 4>
      }
      tintColor {colorUseRule=UseColor_Specified, specifiedColor=<rgba 0.0395 0.0395 0.0395 1>}
      uVRegion {bIsValid=false, max=<0 0>, min=<0 0>}
     }
     normalForeground {colorUseRule=UseColor_ColorTable, specifiedColor=<rgba 0 0 0 0>}
     normalPadding=<ltrb 12 1.5 12 1.5>
     pressed {
      bIsDynamicallyLoaded=false
      drawAs=RoundedBox
      imageType=NoImage
      mirroring=NoMirror
      resourceName=None
      resourceObject=null
      tiling=NoTile
      imageSize=<32 32>
      margin=<ltrb 0 0 0 0>
      outlineSettings {
       bUseBrushTransparency=false
       roundingType=FixedRadius
       width=1
       color {colorUseRule=UseColor_ColorTable, specifiedColor=<rgba 0 0 0 0>}
       cornerRadii=<4 4 4 4>
      }
      tintColor {colorUseRule=UseColor_Specified, specifiedColor=<rgba 0.0284 0.0284 0.0284 1>}
      uVRegion {bIsValid=false, max=<0 0>, min=<0 0>}
     }
     pressedForeground {colorUseRule=UseColor_ColorTable, specifiedColor=<rgba 0 0 0 0>}
     pressedPadding=<ltrb 12 2.5 12 0.5>
     pressedSlateSound {resourceObject=null}
    }}
   slot {class=HorizontalBoxSlot, properties {padding=<ltrb 10 0 10 0>, size {sizeRule=Fill, value=1}}}
  }
  CommonTextBlock_1 {
   widgetClassPath=&CommonTextBlock
   parent=Btn_RandRot
   widget=CommonTextBlock_1
   properties {
    clipping=ClipToBounds
    style=&TextStyle_RobotoEngineLike_C
    text=Jitter Rotation
    colorAndOpacity {colorUseRule=UseColor_Specified, specifiedColor=<rgba 0.5 0.5 0.5 1>}
    font {
     bForceMonospaced=false
     bMaterialIsStencil=false
     fontMaterial=null
     fontObject=&Roboto
     letterSpacing=0
     monospacedWidth=1
     size=8
     skewAmount=0
     typefaceFontName=Regular
     outlineSettings {
      bApplyOutlineToDropShadows=false
      bMiteredCorners=false
      bSeparateFillAlpha=false
      outlineMaterial=null
      outlineSize=0
      outlineColor=<rgba 0 0 0 1>
     }
    }
    shadowOffset=<0 0>
   }
   slot {class=ButtonSlot}
  }
  HorizontalBox_11 {
   widgetClassPath=&HorizontalBox
   parent=Rotation
   widget=HorizontalBox_11
   slot {class=HorizontalBoxSlot, properties {size {sizeRule=Fill, value=1}}}
  }
  HorizontalBox_12 {
   widgetClassPath=&HorizontalBox
   parent=Rotation
   widget=HorizontalBox_12
   slot {class=HorizontalBoxSlot, properties {size {sizeRule=Fill, value=1}}}
  }
  Rotation_1 {
   widgetClassPath=&HorizontalBox
   parent=VerticalBox_360
   widget=Rotation_1
   slot {class=VerticalBoxSlot, properties {size {sizeRule=Fill, value=1}}}
  }
  HorizontalBox_13 {
   widgetClassPath=&HorizontalBox
   parent=Rotation_1
   widget=HorizontalBox_13
   slot {class=HorizontalBoxSlot, properties {size {sizeRule=Fill, value=0.5}}}
  }
  HorizontalBox_14 {
   widgetClassPath=&HorizontalBox
   parent=HorizontalBox_13
   widget=HorizontalBox_14
   slot {class=HorizontalBoxSlot, properties {horizontalAlignment=HAlign_Center, size {sizeRule=Fill, value=1}}}
  }
  CommonTextBlock_233 {
   widgetClassPath=&CommonTextBlock
   parent=HorizontalBox_14
   widget=CommonTextBlock_233
   properties {
    clipping=ClipToBounds
    style=&TextStyle_RobotoEngineLike_C
    text=Z Rotaion Only
    colorAndOpacity {colorUseRule=UseColor_Specified, specifiedColor=<rgba 0.5 0.5 0.5 1>}
    font {
     bForceMonospaced=false
     bMaterialIsStencil=false
     fontMaterial=null
     fontObject=&Roboto
     letterSpacing=0
     monospacedWidth=1
     size=8
     skewAmount=0
     typefaceFontName=Regular
     outlineSettings {
      bApplyOutlineToDropShadows=false
      bMiteredCorners=false
      bSeparateFillAlpha=false
      outlineMaterial=null
      outlineSize=0
      outlineColor=<rgba 0 0 0 1>
     }
    }
    shadowOffset=<0 0>
   }
   slot {class=HorizontalBoxSlot, properties {horizontalAlignment=HAlign_Right, verticalAlignment=VAlign_Center, padding=<ltrb 4 0 8 0>}}
  }
  Check_RotZOnly {
   widgetClassPath=&CheckBox
   bIsVariable=true
   parent=HorizontalBox_14
   widget=Check_RotZOnly
   properties {
    checkedState=Checked
    widgetStyle {
     checkBoxType=CheckBox
     backgroundHoveredImage {
      bIsDynamicallyLoaded=false
      drawAs=RoundedBox
      imageType=NoImage
      mirroring=NoMirror
      resourceName=None
      resourceObject=null
      tiling=NoTile
      imageSize=<18 18>
      margin=<ltrb 0 0 0 0>
      outlineSettings {
       bUseBrushTransparency=false
       roundingType=FixedRadius
       width=1
       color {colorUseRule=UseColor_ColorTable, specifiedColor=<rgba 0 0 0 0>}
       cornerRadii=<3 3 3 3>
      }
      tintColor {colorUseRule=UseColor_Specified, specifiedColor=<rgba 0.0048 0.0048 0.0048 1>}
      uVRegion {bIsValid=false, max=<0 0>, min=<0 0>}
     }
     backgroundImage {
      bIsDynamicallyLoaded=false
      drawAs=RoundedBox
      imageType=NoImage
      mirroring=NoMirror
      resourceName=None
      resourceObject=null
      tiling=NoTile
      imageSize=<18 18>
      margin=<ltrb 0 0 0 0>
      outlineSettings {
       bUseBrushTransparency=false
       roundingType=FixedRadius
       width=1
       color {colorUseRule=UseColor_ColorTable, specifiedColor=<rgba 0 0 0 0>}
       cornerRadii=<3 3 3 3>
      }
      tintColor {colorUseRule=UseColor_Specified, specifiedColor=<rgba 0.0048 0.0048 0.0048 1>}
      uVRegion {bIsValid=false, max=<0 0>, min=<0 0>}
     }
     backgroundPressedImage {
      bIsDynamicallyLoaded=false
      drawAs=RoundedBox
      imageType=NoImage
      mirroring=NoMirror
      resourceName=None
      resourceObject=null
      tiling=NoTile
      imageSize=<18 18>
      margin=<ltrb 0 0 0 0>
      outlineSettings {
       bUseBrushTransparency=false
       roundingType=FixedRadius
       width=1
       color {colorUseRule=UseColor_ColorTable, specifiedColor=<rgba 0 0 0 0>}
       cornerRadii=<3 3 3 3>
      }
      tintColor {colorUseRule=UseColor_Specified, specifiedColor=<rgba 0.0048 0.0048 0.0048 1>}
      uVRegion {bIsValid=false, max=<0 0>, min=<0 0>}
     }
     borderBackgroundColor {colorUseRule=UseColor_Specified, specifiedColor=<rgba 1 1 1 1>}
     checkedForeground {colorUseRule=UseColor_Specified, specifiedColor=<rgba 1 1 1 1>}
     checkedHoveredForeground {colorUseRule=UseColor_Specified, specifiedColor=<rgba 1 1 1 1>}
     checkedHoveredImage {
      bIsDynamicallyLoaded=false
      drawAs=Image
      imageType=Vector
      mirroring=NoMirror
      resourceName=../../../Engine/Content/Slate/Starship/CoreWidgets/CheckBox/check.svg
      resourceObject=null
      tiling=NoTile
      imageSize=<16 16>
      margin=<ltrb 0 0 0 0>
      outlineSettings {
       bUseBrushTransparency=false
       roundingType=HalfHeightRadius
       width=0
       color {colorUseRule=UseColor_Specified, specifiedColor=<rgba 0 0 0 0>}
       cornerRadii=<0 0 0 0>
      }
      tintColor {colorUseRule=UseColor_Specified, specifiedColor=<rgba 0.0044 0.2384 1 1>}
      uVRegion {bIsValid=false, max=<0 0>, min=<0 0>}
     }
     checkedImage {
      bIsDynamicallyLoaded=false
      drawAs=Image
      imageType=Vector
      mirroring=NoMirror
      resourceName=../../../Engine/Content/Slate/Starship/CoreWidgets/CheckBox/check.svg
      resourceObject=null
      tiling=NoTile
      imageSize=<16 16>
      margin=<ltrb 0 0 0 0>
      outlineSettings {
       bUseBrushTransparency=false
       roundingType=HalfHeightRadius
       width=0
       color {colorUseRule=UseColor_Specified, specifiedColor=<rgba 0 0 0 0>}
       cornerRadii=<0 0 0 0>
      }
      tintColor {colorUseRule=UseColor_Specified, specifiedColor=<rgba 0 0.162 0.7454 1>}
      uVRegion {bIsValid=false, max=<0 0>, min=<0 0>}
     }
     checkedPressedForeground {colorUseRule=UseColor_Specified, specifiedColor=<rgba 1 1 1 1>}
     checkedPressedImage {
      bIsDynamicallyLoaded=false
      drawAs=Image
      imageType=Vector
      mirroring=NoMirror
      resourceName=../../../Engine/Content/Slate/Starship/CoreWidgets/CheckBox/check.svg
      resourceObject=null
      tiling=NoTile
      imageSize=<16 16>
      margin=<ltrb 0 0 0 0>
      outlineSettings {
       bUseBrushTransparency=false
       roundingType=HalfHeightRadius
       width=0
       color {colorUseRule=UseColor_Specified, specifiedColor=<rgba 0 0 0 0>}
       cornerRadii=<0 0 0 0>
      }
      tintColor {colorUseRule=UseColor_Specified, specifiedColor=<rgba 0 0.0802 0.3515 1>}
      uVRegion {bIsValid=false, max=<0 0>, min=<0 0>}
     }
     checkedSlateSound {resourceObject=null}
     foregroundColor {colorUseRule=UseColor_Specified, specifiedColor=<rgba 1 1 1 1>}
     hoveredForeground {colorUseRule=UseColor_Specified, specifiedColor=<rgba 1 1 1 1>}
     hoveredSlateSound {resourceObject=null}
     padding=<ltrb 2 0 0 0>
     pressedForeground {colorUseRule=UseColor_Specified, specifiedColor=<rgba 1 1 1 1>}
     uncheckedHoveredImage {
      bIsDynamicallyLoaded=false
      drawAs=NoDrawType
      imageType=NoImage
      mirroring=NoMirror
      resourceName=None
      resourceObject=null
      tiling=NoTile
      imageSize=<0 0>
      margin=<ltrb 0 0 0 0>
      outlineSettings {
       bUseBrushTransparency=false
       roundingType=HalfHeightRadius
       width=0
       color {colorUseRule=UseColor_Specified, specifiedColor=<rgba 0 0 0 0>}
       cornerRadii=<0 0 0 0>
      }
      tintColor {colorUseRule=UseColor_Specified, specifiedColor=<rgba 1 1 1 1>}
      uVRegion {bIsValid=false, max=<0 0>, min=<0 0>}
     }
     uncheckedImage {
      bIsDynamicallyLoaded=false
      drawAs=NoDrawType
      imageType=NoImage
      mirroring=NoMirror
      resourceName=None
      resourceObject=null
      tiling=NoTile
      imageSize=<0 0>
      margin=<ltrb 0 0 0 0>
      outlineSettings {
       bUseBrushTransparency=false
       roundingType=HalfHeightRadius
       width=0
       color {colorUseRule=UseColor_Specified, specifiedColor=<rgba 0 0 0 0>}
       cornerRadii=<0 0 0 0>
      }
      tintColor {colorUseRule=UseColor_Specified, specifiedColor=<rgba 1 1 1 1>}
      uVRegion {bIsValid=false, max=<0 0>, min=<0 0>}
     }
     uncheckedPressedImage {
      bIsDynamicallyLoaded=false
      drawAs=NoDrawType
      imageType=NoImage
      mirroring=NoMirror
      resourceName=None
      resourceObject=null
      tiling=NoTile
      imageSize=<0 0>
      margin=<ltrb 0 0 0 0>
      outlineSettings {
       bUseBrushTransparency=false
       roundingType=HalfHeightRadius
       width=0
       color {colorUseRule=UseColor_Specified, specifiedColor=<rgba 0 0 0 0>}
       cornerRadii=<0 0 0 0>
      }
      tintColor {colorUseRule=UseColor_Specified, specifiedColor=<rgba 1 1 1 1>}
      uVRegion {bIsValid=false, max=<0 0>, min=<0 0>}
     }
     uncheckedSlateSound {resourceObject=null}
     undeterminedForeground {colorUseRule=UseColor_Specified, specifiedColor=<rgba 1 1 1 1>}
     undeterminedHoveredImage {
      bIsDynamicallyLoaded=false
      drawAs=Image
      imageType=Vector
      mirroring=NoMirror
      resourceName=../../../Engine/Content/Slate/Starship/CoreWidgets/CheckBox/indeterminate.svg
      resourceObject=null
      tiling=NoTile
      imageSize=<16 16>
      margin=<ltrb 0 0 0 0>
      outlineSettings {
       bUseBrushTransparency=false
       roundingType=HalfHeightRadius
       width=0
       color {colorUseRule=UseColor_Specified, specifiedColor=<rgba 0 0 0 0>}
       cornerRadii=<0 0 0 0>
      }
      tintColor {colorUseRule=UseColor_Specified, specifiedColor=<rgba 0.0044 0.2384 1 1>}
      uVRegion {bIsValid=false, max=<0 0>, min=<0 0>}
     }
     undeterminedImage {
      bIsDynamicallyLoaded=false
      drawAs=Image
      imageType=Vector
      mirroring=NoMirror
      resourceName=../../../Engine/Content/Slate/Starship/CoreWidgets/CheckBox/indeterminate.svg
      resourceObject=null
      tiling=NoTile
      imageSize=<16 16>
      margin=<ltrb 0 0 0 0>
      outlineSettings {
       bUseBrushTransparency=false
       roundingType=HalfHeightRadius
       width=0
       color {colorUseRule=UseColor_Specified, specifiedColor=<rgba 0 0 0 0>}
       cornerRadii=<0 0 0 0>
      }
      tintColor {colorUseRule=UseColor_Specified, specifiedColor=<rgba 0 0.162 0.7454 1>}
      uVRegion {bIsValid=false, max=<0 0>, min=<0 0>}
     }
     undeterminedPressedImage {
      bIsDynamicallyLoaded=false
      drawAs=Image
      imageType=Vector
      mirroring=NoMirror
      resourceName=../../../Engine/Content/Slate/Starship/CoreWidgets/CheckBox/indeterminate.svg
      resourceObject=null
      tiling=NoTile
      imageSize=<16 16>
      margin=<ltrb 0 0 0 0>
      outlineSettings {
       bUseBrushTransparency=false
       roundingType=HalfHeightRadius
       width=0
       color {colorUseRule=UseColor_Specified, specifiedColor=<rgba 0 0 0 0>}
       cornerRadii=<0 0 0 0>
      }
      tintColor {colorUseRule=UseColor_Specified, specifiedColor=<rgba 0 0.0802 0.3515 1>}
      uVRegion {bIsValid=false, max=<0 0>, min=<0 0>}
     }
    }
   }
   slot {class=HorizontalBoxSlot, properties {horizontalAlignment=HAlign_Center, verticalAlignment=VAlign_Center}}
  }
  HorizontalBox_15 {
   widgetClassPath=&HorizontalBox
   parent=Rotation_1
   widget=HorizontalBox_15
   slot {class=HorizontalBoxSlot, properties {size {sizeRule=Fill, value=1}}}
  }
  RandomRotationAmt {
   widgetClassPath=&SinglePropertyView
   parent=HorizontalBox_15
   widget=RandomRotationAmt
   properties {nameOverride=Random Strength, propertyName=RandomRotationBias}
   slot {class=HorizontalBoxSlot, properties {size {sizeRule=Fill, value=1}}}
  }
  HorizontalBox_16 {
   widgetClassPath=&HorizontalBox
   parent=HorizontalBox_15
   widget=HorizontalBox_16
   slot {class=HorizontalBoxSlot, properties {size {sizeRule=Fill, value=0.25}}}
  }
  F_Scale {
   widgetClassPath=&HorizontalBox
   parent=VerticalBox_76
   widget=F_Scale
   slot {class=VerticalBoxSlot, properties {padding=<ltrb 0 0 0 8>}}
  }
  CommonBorder_551 {
   widgetClassPath=&CommonBorder
   parent=F_Scale
   widget=CommonBorder_551
   properties {
    background {
     bIsDynamicallyLoaded=false
     drawAs=Image
     imageType=NoImage
     mirroring=NoMirror
     resourceName=None
     resourceObject=&T_WhiteMask
     tiling=NoTile
     imageSize=<32 32>
     margin=<ltrb 0 0 0 0>
     outlineSettings {
      bUseBrushTransparency=false
      roundingType=HalfHeightRadius
      width=0
      color {colorUseRule=UseColor_Specified, specifiedColor=<rgba 0 0 0 0>}
      cornerRadii=<0 0 0 0>
     }
     tintColor {colorUseRule=UseColor_Specified, specifiedColor=<rgba 1 1 1 1>}
     uVRegion {bIsValid=false, max=<0 0>, min=<0 0>}
    }
    brushColor=<rgba 0.0103 0.0103 0.0103 1>
    padding=<ltrb 8 8 8 8>
   }
   slot {class=HorizontalBoxSlot, properties {padding=<ltrb 4 0 4 0>, size {sizeRule=Fill, value=1}}}
  }
  VerticalBox_402 {
   widgetClassPath=&VerticalBox
   parent=CommonBorder_551
   widget=VerticalBox_402
   slot {class=BorderSlot, properties {padding=<ltrb 8 8 8 8>}}
  }
  Scale {
   widgetClassPath=&HorizontalBox
   parent=VerticalBox_402
   widget=Scale
   slot {class=VerticalBoxSlot, properties {padding=<ltrb 0 0 0 8>}}
  }
  HorizontalBox_949 {
   widgetClassPath=&HorizontalBox
   parent=Scale
   widget=HorizontalBox_949
   slot {class=HorizontalBoxSlot, properties {size {sizeRule=Fill, value=1}}}
  }
  Btn_RandScale {
   widgetClassPath=&Button
   parent=HorizontalBox_949
   widget=Btn_RandScale
   properties {widgetStyle {
     clickedSlateSound {resourceObject=null}
     disabled {
      bIsDynamicallyLoaded=false
      drawAs=RoundedBox
      imageType=NoImage
      mirroring=NoMirror
      resourceName=None
      resourceObject=null
      tiling=NoTile
      imageSize=<0 0>
      margin=<ltrb 0 0 0 0>
      outlineSettings {
       bUseBrushTransparency=false
       roundingType=FixedRadius
       width=1
       color {colorUseRule=UseColor_ColorTable, specifiedColor=<rgba 0 0 0 0>}
       cornerRadii=<4 4 4 4>
      }
      tintColor {colorUseRule=UseColor_Specified, specifiedColor=<rgba 0.0395 0.0395 0.0395 1>}
      uVRegion {bIsValid=false, max=<0 0>, min=<0 0>}
     }
     disabledForeground {colorUseRule=UseColor_ColorTable, specifiedColor=<rgba 0 0 0 0>}
     hovered {
      bIsDynamicallyLoaded=false
      drawAs=RoundedBox
      imageType=NoImage
      mirroring=NoMirror
      resourceName=None
      resourceObject=null
      tiling=NoTile
      imageSize=<32 32>
      margin=<ltrb 0 0 0 0>
      outlineSettings {
       bUseBrushTransparency=false
       roundingType=FixedRadius
       width=1
       color {colorUseRule=UseColor_ColorTable, specifiedColor=<rgba 0 0 0 0>}
       cornerRadii=<4 4 4 4>
      }
      tintColor {colorUseRule=UseColor_Specified, specifiedColor=<rgba 0.0953 0.0953 0.0953 1>}
      uVRegion {bIsValid=false, max=<0 0>, min=<0 0>}
     }
     hoveredForeground {colorUseRule=UseColor_ColorTable, specifiedColor=<rgba 0 0 0 0>}
     hoveredSlateSound {resourceObject=null}
     normal {
      bIsDynamicallyLoaded=false
      drawAs=RoundedBox
      imageType=NoImage
      mirroring=NoMirror
      resourceName=None
      resourceObject=null
      tiling=NoTile
      imageSize=<32 32>
      margin=<ltrb 0 0 0 0>
      outlineSettings {
       bUseBrushTransparency=false
       roundingType=FixedRadius
       width=1
       color {colorUseRule=UseColor_ColorTable, specifiedColor=<rgba 0 0 0 0>}
       cornerRadii=<4 4 4 4>
      }
      tintColor {colorUseRule=UseColor_Specified, specifiedColor=<rgba 0.0395 0.0395 0.0395 1>}
      uVRegion {bIsValid=false, max=<0 0>, min=<0 0>}
     }
     normalForeground {colorUseRule=UseColor_ColorTable, specifiedColor=<rgba 0 0 0 0>}
     normalPadding=<ltrb 12 1.5 12 1.5>
     pressed {
      bIsDynamicallyLoaded=false
      drawAs=RoundedBox
      imageType=NoImage
      mirroring=NoMirror
      resourceName=None
      resourceObject=null
      tiling=NoTile
      imageSize=<32 32>
      margin=<ltrb 0 0 0 0>
      outlineSettings {
       bUseBrushTransparency=false
       roundingType=FixedRadius
       width=1
       color {colorUseRule=UseColor_ColorTable, specifiedColor=<rgba 0 0 0 0>}
       cornerRadii=<4 4 4 4>
      }
      tintColor {colorUseRule=UseColor_Specified, specifiedColor=<rgba 0.0284 0.0284 0.0284 1>}
      uVRegion {bIsValid=false, max=<0 0>, min=<0 0>}
     }
     pressedForeground {colorUseRule=UseColor_ColorTable, specifiedColor=<rgba 0 0 0 0>}
     pressedPadding=<ltrb 12 2.5 12 0.5>
     pressedSlateSound {resourceObject=null}
    }}
   slot {class=HorizontalBoxSlot, properties {padding=<ltrb 10 0 10 0>, size {sizeRule=Fill, value=1}}}
  }
  CommonTextBlock_4 {
   widgetClassPath=&CommonTextBlock
   parent=Btn_RandScale
   widget=CommonTextBlock_4
   properties {
    clipping=ClipToBounds
    style=&TextStyle_RobotoEngineLike_C
    text=Jitter Scale
    colorAndOpacity {colorUseRule=UseColor_Specified, specifiedColor=<rgba 0.5 0.5 0.5 1>}
    font {
     bForceMonospaced=false
     bMaterialIsStencil=false
     fontMaterial=null
     fontObject=&Roboto
     letterSpacing=0
     monospacedWidth=1
     size=8
     skewAmount=0
     typefaceFontName=Regular
     outlineSettings {
      bApplyOutlineToDropShadows=false
      bMiteredCorners=false
      bSeparateFillAlpha=false
      outlineMaterial=null
      outlineSize=0
      outlineColor=<rgba 0 0 0 1>
     }
    }
    shadowOffset=<0 0>
   }
   slot {class=ButtonSlot}
  }
  HorizontalBox_2 {
   widgetClassPath=&HorizontalBox
   parent=Scale
   widget=HorizontalBox_2
   slot {class=HorizontalBoxSlot, properties {size {sizeRule=Fill, value=1}}}
  }
  Btn_ScaleMult {
   widgetClassPath=&Button
   parent=HorizontalBox_2
   widget=Btn_ScaleMult
   properties {widgetStyle {
     clickedSlateSound {resourceObject=null}
     disabled {
      bIsDynamicallyLoaded=false
      drawAs=RoundedBox
      imageType=NoImage
      mirroring=NoMirror
      resourceName=None
      resourceObject=null
      tiling=NoTile
      imageSize=<0 0>
      margin=<ltrb 0 0 0 0>
      outlineSettings {
       bUseBrushTransparency=false
       roundingType=FixedRadius
       width=1
       color {colorUseRule=UseColor_ColorTable, specifiedColor=<rgba 0 0 0 0>}
       cornerRadii=<4 4 4 4>
      }
      tintColor {colorUseRule=UseColor_Specified, specifiedColor=<rgba 0.0395 0.0395 0.0395 1>}
      uVRegion {bIsValid=false, max=<0 0>, min=<0 0>}
     }
     disabledForeground {colorUseRule=UseColor_ColorTable, specifiedColor=<rgba 0 0 0 0>}
     hovered {
      bIsDynamicallyLoaded=false
      drawAs=RoundedBox
      imageType=NoImage
      mirroring=NoMirror
      resourceName=None
      resourceObject=null
      tiling=NoTile
      imageSize=<32 32>
      margin=<ltrb 0 0 0 0>
      outlineSettings {
       bUseBrushTransparency=false
       roundingType=FixedRadius
       width=1
       color {colorUseRule=UseColor_ColorTable, specifiedColor=<rgba 0 0 0 0>}
       cornerRadii=<4 4 4 4>
      }
      tintColor {colorUseRule=UseColor_Specified, specifiedColor=<rgba 0.0953 0.0953 0.0953 1>}
      uVRegion {bIsValid=false, max=<0 0>, min=<0 0>}
     }
     hoveredForeground {colorUseRule=UseColor_ColorTable, specifiedColor=<rgba 0 0 0 0>}
     hoveredSlateSound {resourceObject=null}
     normal {
      bIsDynamicallyLoaded=false
      drawAs=RoundedBox
      imageType=NoImage
      mirroring=NoMirror
      resourceName=None
      resourceObject=null
      tiling=NoTile
      imageSize=<32 32>
      margin=<ltrb 0 0 0 0>
      outlineSettings {
       bUseBrushTransparency=false
       roundingType=FixedRadius
       width=1
       color {colorUseRule=UseColor_ColorTable, specifiedColor=<rgba 0 0 0 0>}
       cornerRadii=<4 4 4 4>
      }
      tintColor {colorUseRule=UseColor_Specified, specifiedColor=<rgba 0.0395 0.0395 0.0395 1>}
      uVRegion {bIsValid=false, max=<0 0>, min=<0 0>}
     }
     normalForeground {colorUseRule=UseColor_ColorTable, specifiedColor=<rgba 0 0 0 0>}
     normalPadding=<ltrb 12 1.5 12 1.5>
     pressed {
      bIsDynamicallyLoaded=false
      drawAs=RoundedBox
      imageType=NoImage
      mirroring=NoMirror
      resourceName=None
      resourceObject=null
      tiling=NoTile
      imageSize=<32 32>
      margin=<ltrb 0 0 0 0>
      outlineSettings {
       bUseBrushTransparency=false
       roundingType=FixedRadius
       width=1
       color {colorUseRule=UseColor_ColorTable, specifiedColor=<rgba 0 0 0 0>}
       cornerRadii=<4 4 4 4>
      }
      tintColor {colorUseRule=UseColor_Specified, specifiedColor=<rgba 0.0284 0.0284 0.0284 1>}
      uVRegion {bIsValid=false, max=<0 0>, min=<0 0>}
     }
     pressedForeground {colorUseRule=UseColor_ColorTable, specifiedColor=<rgba 0 0 0 0>}
     pressedPadding=<ltrb 12 2.5 12 0.5>
     pressedSlateSound {resourceObject=null}
    }}
   slot {class=HorizontalBoxSlot, properties {padding=<ltrb 10 0 10 0>, size {sizeRule=Fill, value=1}}}
  }
  CommonTextBlock_6 {
   widgetClassPath=&CommonTextBlock
   parent=Btn_ScaleMult
   widget=CommonTextBlock_6
   properties {
    clipping=ClipToBounds
    style=&TextStyle_RobotoEngineLike_C
    text=Multiply Scale
    colorAndOpacity {colorUseRule=UseColor_Specified, specifiedColor=<rgba 0.5 0.5 0.5 1>}
    font {
     bForceMonospaced=false
     bMaterialIsStencil=false
     fontMaterial=null
     fontObject=&Roboto
     letterSpacing=0
     monospacedWidth=1
     size=8
     skewAmount=0
     typefaceFontName=Regular
     outlineSettings {
      bApplyOutlineToDropShadows=false
      bMiteredCorners=false
      bSeparateFillAlpha=false
      outlineMaterial=null
      outlineSize=0
      outlineColor=<rgba 0 0 0 1>
     }
    }
    shadowOffset=<0 0>
   }
   slot {class=ButtonSlot}
  }
  HorizontalBox_3 {
   widgetClassPath=&HorizontalBox
   parent=Scale
   widget=HorizontalBox_3
   slot {class=HorizontalBoxSlot, properties {size {sizeRule=Fill, value=1}}}
  }
  HorizontalBox_584 {
   widgetClassPath=&HorizontalBox
   parent=VerticalBox_402
   widget=HorizontalBox_584
   slot {class=VerticalBoxSlot, properties {size {sizeRule=Fill, value=1}}}
  }
  HorizontalBox_947 {
   widgetClassPath=&HorizontalBox
   parent=HorizontalBox_584
   widget=HorizontalBox_947
   slot {class=HorizontalBoxSlot, properties {size {sizeRule=Fill, value=1}}}
  }
  HorizontalBox_953 {
   widgetClassPath=&HorizontalBox
   parent=HorizontalBox_584
   widget=HorizontalBox_953
   slot {class=HorizontalBoxSlot, properties {size {sizeRule=Fill, value=1}}}
  }
  PropScaleMultiplier {
   widgetClassPath=&SinglePropertyView
   parent=HorizontalBox_953
   widget=PropScaleMultiplier
   properties {nameOverride=" ", propertyName=ScaleMultiplier}
   slot {class=HorizontalBoxSlot, properties {size {sizeRule=Fill, value=1}}}
  }
  HorizontalBox_1 {
   widgetClassPath=&HorizontalBox
   parent=HorizontalBox_584
   widget=HorizontalBox_1
   slot {class=HorizontalBoxSlot, properties {size {sizeRule=Fill, value=1}}}
  }
  F_Location {
   widgetClassPath=&HorizontalBox
   parent=VerticalBox_76
   widget=F_Location
   slot {class=VerticalBoxSlot, properties {padding=<ltrb 0 0 0 8>}}
  }
  CommonBorder_586 {
   widgetClassPath=&CommonBorder
   parent=F_Location
   widget=CommonBorder_586
   properties {
    background {
     bIsDynamicallyLoaded=false
     drawAs=Image
     imageType=NoImage
     mirroring=NoMirror
     resourceName=None
     resourceObject=&T_WhiteMask
     tiling=NoTile
     imageSize=<32 32>
     margin=<ltrb 0 0 0 0>
     outlineSettings {
      bUseBrushTransparency=false
      roundingType=HalfHeightRadius
      width=0
      color {colorUseRule=UseColor_Specified, specifiedColor=<rgba 0 0 0 0>}
      cornerRadii=<0 0 0 0>
     }
     tintColor {colorUseRule=UseColor_Specified, specifiedColor=<rgba 1 1 1 1>}
     uVRegion {bIsValid=false, max=<0 0>, min=<0 0>}
    }
    brushColor=<rgba 0.0103 0.0103 0.0103 1>
    padding=<ltrb 0 0 0 0>
   }
   slot {class=HorizontalBoxSlot, properties {padding=<ltrb 4 0 4 0>, size {sizeRule=Fill, value=1}}}
  }
  VerticalBox_477 {
   widgetClassPath=&VerticalBox
   parent=CommonBorder_586
   widget=VerticalBox_477
   slot {class=BorderSlot, properties {padding=<ltrb 0 0 0 0>}}
  }
  Location {
   widgetClassPath=&HorizontalBox
   parent=VerticalBox_477
   widget=Location
   slot {class=VerticalBoxSlot, properties {padding=<ltrb 8 8 8 8>}}
  }
  HorizontalBox_950 {
   widgetClassPath=&HorizontalBox
   parent=Location
   widget=HorizontalBox_950
   slot {class=HorizontalBoxSlot, properties {size {sizeRule=Fill, value=1}}}
  }
  Btn_RandLoc {
   widgetClassPath=&Button
   parent=HorizontalBox_950
   widget=Btn_RandLoc
   properties {widgetStyle {
     clickedSlateSound {resourceObject=null}
     disabled {
      bIsDynamicallyLoaded=false
      drawAs=RoundedBox
      imageType=NoImage
      mirroring=NoMirror
      resourceName=None
      resourceObject=null
      tiling=NoTile
      imageSize=<0 0>
      margin=<ltrb 0 0 0 0>
      outlineSettings {
       bUseBrushTransparency=false
       roundingType=FixedRadius
       width=1
       color {colorUseRule=UseColor_ColorTable, specifiedColor=<rgba 0 0 0 0>}
       cornerRadii=<4 4 4 4>
      }
      tintColor {colorUseRule=UseColor_Specified, specifiedColor=<rgba 0.0395 0.0395 0.0395 1>}
      uVRegion {bIsValid=false, max=<0 0>, min=<0 0>}
     }
     disabledForeground {colorUseRule=UseColor_ColorTable, specifiedColor=<rgba 0 0 0 0>}
     hovered {
      bIsDynamicallyLoaded=false
      drawAs=RoundedBox
      imageType=NoImage
      mirroring=NoMirror
      resourceName=None
      resourceObject=null
      tiling=NoTile
      imageSize=<32 32>
      margin=<ltrb 0 0 0 0>
      outlineSettings {
       bUseBrushTransparency=false
       roundingType=FixedRadius
       width=1
       color {colorUseRule=UseColor_ColorTable, specifiedColor=<rgba 0 0 0 0>}
       cornerRadii=<4 4 4 4>
      }
      tintColor {colorUseRule=UseColor_Specified, specifiedColor=<rgba 0.0953 0.0953 0.0953 1>}
      uVRegion {bIsValid=false, max=<0 0>, min=<0 0>}
     }
     hoveredForeground {colorUseRule=UseColor_ColorTable, specifiedColor=<rgba 0 0 0 0>}
     hoveredSlateSound {resourceObject=null}
     normal {
      bIsDynamicallyLoaded=false
      drawAs=RoundedBox
      imageType=NoImage
      mirroring=NoMirror
      resourceName=None
      resourceObject=null
      tiling=NoTile
      imageSize=<32 32>
      margin=<ltrb 0 0 0 0>
      outlineSettings {
       bUseBrushTransparency=false
       roundingType=FixedRadius
       width=1
       color {colorUseRule=UseColor_ColorTable, specifiedColor=<rgba 0 0 0 0>}
       cornerRadii=<4 4 4 4>
      }
      tintColor {colorUseRule=UseColor_Specified, specifiedColor=<rgba 0.0395 0.0395 0.0395 1>}
      uVRegion {bIsValid=false, max=<0 0>, min=<0 0>}
     }
     normalForeground {colorUseRule=UseColor_ColorTable, specifiedColor=<rgba 0 0 0 0>}
     normalPadding=<ltrb 12 1.5 12 1.5>
     pressed {
      bIsDynamicallyLoaded=false
      drawAs=RoundedBox
      imageType=NoImage
      mirroring=NoMirror
      resourceName=None
      resourceObject=null
      tiling=NoTile
      imageSize=<32 32>
      margin=<ltrb 0 0 0 0>
      outlineSettings {
       bUseBrushTransparency=false
       roundingType=FixedRadius
       width=1
       color {colorUseRule=UseColor_ColorTable, specifiedColor=<rgba 0 0 0 0>}
       cornerRadii=<4 4 4 4>
      }
      tintColor {colorUseRule=UseColor_Specified, specifiedColor=<rgba 0.0284 0.0284 0.0284 1>}
      uVRegion {bIsValid=false, max=<0 0>, min=<0 0>}
     }
     pressedForeground {colorUseRule=UseColor_ColorTable, specifiedColor=<rgba 0 0 0 0>}
     pressedPadding=<ltrb 12 2.5 12 0.5>
     pressedSlateSound {resourceObject=null}
    }}
   slot {class=HorizontalBoxSlot, properties {padding=<ltrb 10 0 10 0>, size {sizeRule=Fill, value=1}}}
  }
  CommonTextBlock_2 {
   widgetClassPath=&CommonTextBlock
   parent=Btn_RandLoc
   widget=CommonTextBlock_2
   properties {
    clipping=ClipToBounds
    style=&TextStyle_RobotoEngineLike_C
    text=Jitter Position
    colorAndOpacity {colorUseRule=UseColor_Specified, specifiedColor=<rgba 0.5 0.5 0.5 1>}
    font {
     bForceMonospaced=false
     bMaterialIsStencil=false
     fontMaterial=null
     fontObject=&Roboto
     letterSpacing=0
     monospacedWidth=1
     size=8
     skewAmount=0
     typefaceFontName=Regular
     outlineSettings {
      bApplyOutlineToDropShadows=false
      bMiteredCorners=false
      bSeparateFillAlpha=false
      outlineMaterial=null
      outlineSize=0
      outlineColor=<rgba 0 0 0 1>
     }
    }
    shadowOffset=<0 0>
   }
   slot {class=ButtonSlot}
  }
  HorizontalBox_951 {
   widgetClassPath=&HorizontalBox
   parent=Location
   widget=HorizontalBox_951
   slot {class=HorizontalBoxSlot, properties {size {sizeRule=Fill, value=1}}}
  }
  Btn_RandLoc_1 {
   widgetClassPath=&Button
   parent=HorizontalBox_951
   widget=Btn_RandLoc_1
   properties {widgetStyle {
     clickedSlateSound {resourceObject=null}
     disabled {
      bIsDynamicallyLoaded=false
      drawAs=RoundedBox
      imageType=NoImage
      mirroring=NoMirror
      resourceName=None
      resourceObject=null
      tiling=NoTile
      imageSize=<0 0>
      margin=<ltrb 0 0 0 0>
      outlineSettings {
       bUseBrushTransparency=false
       roundingType=FixedRadius
       width=1
       color {colorUseRule=UseColor_ColorTable, specifiedColor=<rgba 0 0 0 0>}
       cornerRadii=<4 4 4 4>
      }
      tintColor {colorUseRule=UseColor_Specified, specifiedColor=<rgba 0.0395 0.0395 0.0395 1>}
      uVRegion {bIsValid=false, max=<0 0>, min=<0 0>}
     }
     disabledForeground {colorUseRule=UseColor_ColorTable, specifiedColor=<rgba 0 0 0 0>}
     hovered {
      bIsDynamicallyLoaded=false
      drawAs=RoundedBox
      imageType=NoImage
      mirroring=NoMirror
      resourceName=None
      resourceObject=null
      tiling=NoTile
      imageSize=<32 32>
      margin=<ltrb 0 0 0 0>
      outlineSettings {
       bUseBrushTransparency=false
       roundingType=FixedRadius
       width=1
       color {colorUseRule=UseColor_ColorTable, specifiedColor=<rgba 0 0 0 0>}
       cornerRadii=<4 4 4 4>
      }
      tintColor {colorUseRule=UseColor_Specified, specifiedColor=<rgba 0.0953 0.0953 0.0953 1>}
      uVRegion {bIsValid=false, max=<0 0>, min=<0 0>}
     }
     hoveredForeground {colorUseRule=UseColor_ColorTable, specifiedColor=<rgba 0 0 0 0>}
     hoveredSlateSound {resourceObject=null}
     normal {
      bIsDynamicallyLoaded=false
      drawAs=RoundedBox
      imageType=NoImage
      mirroring=NoMirror
      resourceName=None
      resourceObject=null
      tiling=NoTile
      imageSize=<32 32>
      margin=<ltrb 0 0 0 0>
      outlineSettings {
       bUseBrushTransparency=false
       roundingType=FixedRadius
       width=1
       color {colorUseRule=UseColor_ColorTable, specifiedColor=<rgba 0 0 0 0>}
       cornerRadii=<4 4 4 4>
      }
      tintColor {colorUseRule=UseColor_Specified, specifiedColor=<rgba 0.0395 0.0395 0.0395 1>}
      uVRegion {bIsValid=false, max=<0 0>, min=<0 0>}
     }
     normalForeground {colorUseRule=UseColor_ColorTable, specifiedColor=<rgba 0 0 0 0>}
     normalPadding=<ltrb 12 1.5 12 1.5>
     pressed {
      bIsDynamicallyLoaded=false
      drawAs=RoundedBox
      imageType=NoImage
      mirroring=NoMirror
      resourceName=None
      resourceObject=null
      tiling=NoTile
      imageSize=<32 32>
      margin=<ltrb 0 0 0 0>
      outlineSettings {
       bUseBrushTransparency=false
       roundingType=FixedRadius
       width=1
       color {colorUseRule=UseColor_ColorTable, specifiedColor=<rgba 0 0 0 0>}
       cornerRadii=<4 4 4 4>
      }
      tintColor {colorUseRule=UseColor_Specified, specifiedColor=<rgba 0.0284 0.0284 0.0284 1>}
      uVRegion {bIsValid=false, max=<0 0>, min=<0 0>}
     }
     pressedForeground {colorUseRule=UseColor_ColorTable, specifiedColor=<rgba 0 0 0 0>}
     pressedPadding=<ltrb 12 2.5 12 0.5>
     pressedSlateSound {resourceObject=null}
    }}
   slot {class=HorizontalBoxSlot, properties {padding=<ltrb 10 0 10 0>, size {sizeRule=Fill, value=1}}}
  }
  CommonTextBlock_8 {
   widgetClassPath=&CommonTextBlock
   parent=Btn_RandLoc_1
   widget=CommonTextBlock_8
   properties {
    clipping=ClipToBounds
    style=&TextStyle_RobotoEngineLike_C
    text=Random Position
    colorAndOpacity {colorUseRule=UseColor_Specified, specifiedColor=<rgba 0.5 0.5 0.5 1>}
    font {
     bForceMonospaced=false
     bMaterialIsStencil=false
     fontMaterial=null
     fontObject=&Roboto
     letterSpacing=0
     monospacedWidth=1
     size=8
     skewAmount=0
     typefaceFontName=Regular
     outlineSettings {
      bApplyOutlineToDropShadows=false
      bMiteredCorners=false
      bSeparateFillAlpha=false
      outlineMaterial=null
      outlineSize=0
      outlineColor=<rgba 0 0 0 1>
     }
    }
    shadowOffset=<0 0>
   }
   slot {class=ButtonSlot}
  }
  HorizontalBox_4 {
   widgetClassPath=&HorizontalBox
   parent=Location
   widget=HorizontalBox_4
   slot {class=HorizontalBoxSlot, properties {size {sizeRule=Fill, value=1}}}
  }
  PropSnap {
   widgetClassPath=&SinglePropertyView
   parent=HorizontalBox_4
   widget=PropSnap
   properties {nameOverride=Snap, propertyName=SnapToGroundPos}
   slot {class=HorizontalBoxSlot, properties {horizontalAlignment=HAlign_Center, size {sizeRule=Fill, value=1}}}
  }
  Location_1 {
   widgetClassPath=&HorizontalBox
   parent=VerticalBox_477
   widget=Location_1
   slot {class=VerticalBoxSlot, properties {padding=<ltrb 8 0 8 8>}}
  }
  PropJitterDistance {
   widgetClassPath=&SinglePropertyView
   parent=Location_1
   widget=PropJitterDistance
   properties {nameOverride=" ", propertyName=MaxRandomDistance}
   slot {class=HorizontalBoxSlot, properties {size {sizeRule=Fill, value=1}}}
  }
  HorizontalBox_5 {
   widgetClassPath=&HorizontalBox
   parent=Location_1
   widget=HorizontalBox_5
   slot {class=HorizontalBoxSlot, properties {size {sizeRule=Fill, value=1}}}
  }
  HorizontalBox_6 {
   widgetClassPath=&HorizontalBox
   parent=Location_1
   widget=HorizontalBox_6
   slot {class=HorizontalBoxSlot, properties {size {sizeRule=Fill, value=1}}}
  }
 ]
}
